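# Landscape material function that applies Jungle biome color overrides to Asteria terrain layers.
# It lerps incoming Grass, GravelBiomeColorDiv, MudBiomeColorDiv and SandBiomeColorDiv values toward Jungle values, using CustomBiomeA from MF_AsteriaTerrainMasks as alpha.
# Gravel and Mud targets are tunable vector parameters, while Grass and Sand targets are hardcoded constants.
# Two Frontier-named parameters and two white constants are left unconnected.
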MaterialFunction /Game/Athena/Asteria/Environments/Landscape/Materials/MaterialFunctions/MF_AsteriaCustomBiomes_Jungle
graph {nodes [
  Constant3Vector() {constant=<rgba 1 1 1 1>}
  Constant3Vector() {constant=<rgba 1 1 1 1>}
  n1 = MaterialFunctionCall!(UseLandscapeCoords) {materialFunction=&/MF_AsteriaTerrainMasks}
  FunctionOutput(None=LinearInterpolate(
    A=FunctionInput() {bUsePreviewValueAsDefault=true, inputName=Grass, sortPriority=-12, previewValue=<1 1 1 1>}
    B=Constant3Vector() {constant=<rgba 0.696 1.1457 5.8 1>}
    Alpha=@n1.CustomBiomeA
   )) {outputName=Grass, sortPriority=-12}
  n2 = FunctionInput() {bUsePreviewValueAsDefault=true, inputName=GravelBiomeColorDiv, sortPriority=7, previewValue=<1 1 1 1>}
  FunctionOutput(None=LinearInterpolate(A=@n2, B=VectorParameter() {parameterName=Jungle-GravelBiomeColorDiv, defaultValue=<rgba 1.5 1.5 1.5 1>}.RGB, Alpha=@n1.CustomBiomeA)) {outputName=GravelBiomeColorDiv, sortPriority=7}
  n3 = FunctionInput() {bUsePreviewValueAsDefault=true, inputName=MudBiomeColorDiv, sortPriority=9, previewValue=<1 1 1 1>}
  FunctionOutput(None=LinearInterpolate(A=@n3, B=VectorParameter() {parameterName=Jungle-MudBiomeColorDiv, defaultValue=<rgba 1 1 1 1>}.RGB, Alpha=@n1.CustomBiomeA)) {outputName=MudBiomeColorDiv, sortPriority=9}
  n4 = FunctionInput() {bUsePreviewValueAsDefault=true, inputName=SandBiomeColorDiv, sortPriority=12, previewValue=<1 1 1 1>}
  FunctionOutput(None=LinearInterpolate(A=@n4, B=Constant3Vector() {constant=<rgba 1 1 1 1>}, Alpha=@n1.CustomBiomeA)) {outputName=SandBiomeColorDiv, sortPriority=12}
  VectorParameter() {parameterName=Frontier-SandBiomeColorDiv, defaultValue=<rgba 1 1 1 1>}
  VectorParameter() {parameterName=Frontier-GrassEdge, defaultValue=<rgba 1 1 1 1>}
 ]}
