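# Landscape material function that styles an incoming set of material attributes as jungle mud and outputs a Height value.
# It tints base color with MudColorMult, zeroes metallic, and uses incoming specular and roughness as blend alphas between Min/Max parameters.
# Height comes from the alpha of a dirt color texture sampled on scaled landscape coordinates.
# The dirt normal and specular samples and a color constant connect to neither output.
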MaterialFunction /Game/Athena/Asteria/Environments/Landscape/Materials/MaterialFunctions/MF_JungleMud
graph {nodes [
  Constant3Vector() {constant=<rgba 0.9844 0.4957 0.2968 1>}
  n1 = GetMaterialAttributes(None=FunctionInput() {inputName=Material, inputType=FunctionInput_MaterialAttributes}) {attributeGetTypes [
    #1
    #2
    #3
    #4
   ]}
  n2 = SetMaterialAttributes!(
   MaterialAttributes=@n1.MaterialAttributes
   BaseColor=Multiply(A=VectorParameter() {parameterName=MudColorMult, defaultValue=<rgba 1 1 1 1>}.RGB, B=@n1.BaseColor)
   Metallic=Constant()
   Specular=LinearInterpolate(
    A=ScalarParameter() {parameterName=JungleMud_MinSpec}
    B=ScalarParameter() {defaultValue=1, parameterName=JungleMud_MaxSpec}
    Alpha=@n1.Specular
   )
   Roughness=LinearInterpolate(
    A=ScalarParameter() {parameterName=JungleMud_MinRoughness}
    B=ScalarParameter() {defaultValue=1, parameterName=JungleMud_MaxRoughness}
    Alpha=@n1.Roughness
   )
   Normal=@n1.MaterialAttributes
  ) {attributeSetTypes [
    #1
    #5
    #2
    #3
    #4
   ]}
  FunctionOutput(None=@n2) {outputName=Material}
  n3 = Multiply(A=LandscapeLayerCoords() {mappingScale=10}, B=ScalarParameter() {defaultValue=3, group=Forest, parameterName=Forest Scale})
  n4 = TextureSampleParameter2D(UVs=@n3) {
   group=Forest
   parameterName=JungleMud_Color
   samplerSource=SSM_Wrap_WorldGroupSettings
   texture=/Game/Environments/Apollo/Sets/Terrain/Textures/T_Apollo_Dirt_D_02
  }
  FunctionOutput(None=@n4.A) {outputName=Height, sortPriority=1}
  TextureSampleParameter2D!(UVs=@n3, Bias=ScalarParameter() {defaultValue=4, parameterName=NormalMipBioas}, Apply View MipBias) {
   mipValueMode=TMVM_MipBias
   parameterName=JungleMud_Normal
   samplerSource=SSM_Wrap_WorldGroupSettings
   samplerType=SAMPLERTYPE_Normal
   texture=/Game/Environments/Apollo/Sets/Terrain/Textures/T_Apollo_Dirt_N_02
  }
  TextureSampleParameter2D!(UVs=@n3, Apply View MipBias) {
   parameterName=JungleMud_Specular
   samplerSource=SSM_Wrap_WorldGroupSettings
   samplerType=SAMPLERTYPE_Masks
   texture=/Game/Environments/Apollo/Sets/Terrain/Textures/T_Apollo_Dirt_S_02
  }
 ]}
