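# Ice material for the Asteria landscape, building base color, roughness, specular and flattened normal values as named reroutes.
# MF_LandscapeTextureSample supplies crack, noise and normal samples at parameterized tiling.
# Base color blends two blue tints, then a white crack color; roughness lerps between low minimum and maximum values.
# It also declares three variation noise scales and several unconnected candidate textures, and shows no connections to material outputs.
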Material /Game/Athena/Asteria/Environments/Landscape/Materials/MaterialFunctions/M_Asteria_Ice
graph {
 properties {pixelDepthOffsetMode=PDOM_Legacy, usedShadingModels=Default Lit}
 nodes [
  Constant()
  NamedRerouteDeclaration(None=TextureObject() {texture=/Game/Athena/Environments/Meteor/Textures/Meteor_Cracks_M}) {name=CracksTexture, nodeColor=<rgba 0.5469 1 0 1>}
  n1 = MaterialFunctionCall!(
   Tiling=ScalarParameter() {defaultValue=4, parameterName=Ice_CracksScale}
   Texture=NamedRerouteUsage()
   LandscapeCoordinates=NamedRerouteUsage()
  ) {materialFunction=/Game/Athena/Asteria/Environments/Landscape/Materials/MaterialFunctions/UtilityFunctions/MF_LandscapeTextureSample}
  NamedRerouteDeclaration(None=Multiply(A=ComponentMask(None=@n1.Result) {r=true}, B=NamedRerouteUsage())) {name=Cracks_Mask, nodeColor=<rgba 1 0 0.3516 1>}
  NamedRerouteDeclaration(None=ScalarParameter() {defaultValue=1, parameterName=Ice_Specular}) {name=Ice_Specular, nodeColor=<rgba 0 1 1 1>}
  NamedRerouteDeclaration(None=LinearInterpolate(
    A=ScalarParameter() {defaultValue=0.1, parameterName=Ice_roughness_Min}
    B=ScalarParameter() {defaultValue=0.25, parameterName=Ice_roughness_Max}
    Alpha=NamedRerouteUsage()
   )) {name=Ice_Roughness, nodeColor=<rgba 0 1 1 1>}
  n2 = MaterialFunctionCall(
   Tiling=ScalarParameter() {defaultValue=2, parameterName=BaseColorNoise_Scale}
   Texture=NamedRerouteUsage()
   LandscapeCoordinates=NamedRerouteUsage()
  ) {materialFunction=&MF_LandscapeTextureSample}
  NamedRerouteDeclaration(None=ComponentMask(None=@n2.Result) {g=true}) {name=Roughness_Noise, nodeColor=<rgba 1 0 0.2812 1>}
  NamedRerouteDeclaration(None=TextureObject() {samplerType=SAMPLERTYPE_Normal, texture=/Game/Global/Textures/Default/Tileable_Patterns/T_Desert_Cracks_N}) {name=Ice_Cracks_Normal, nodeColor=<rgba 0.1484 0 1 1>}
  NamedRerouteDeclaration(None=TextureObject() {samplerType=SAMPLERTYPE_Normal, texture=/Game/Effects/Fort_Effects/Textures/Tile/T_GaussianNoise_02_N}) {name=Ice_Normal_01, nodeColor=<rgba 1 0.0703 0 1>}
  n3 = ScalarParameter() {defaultValue=2, parameterName=BaseNormal_Scale}
  n4 = NamedRerouteUsage()
  n5 = NamedRerouteUsage()
  NamedRerouteDeclaration(None=MaterialFunctionCall(
    Normal=MaterialFunctionCall(Tiling=@n3, Texture=@n4, LandscapeCoordinates=@n5) {materialFunction=&MF_LandscapeTextureSample}.Result
    Flatness=ScalarParameter() {defaultValue=0.8, parameterName=Ice_BaseNormalFlatness}
   ) {materialFunction=&FlattenNormal}.Result) {name=Ice_Normal, nodeColor=<rgba 0.5234 0 1 1>}
  NamedRerouteDeclaration(None=TextureObject() {samplerType=SAMPLERTYPE_Grayscale, texture=/FortWater/Textures/LowResBlurredNoise}) {name=GrayscaleNoise_Texture, nodeColor=<rgba 0 0.4844 1 1>}
  NamedRerouteDeclaration(None=MaterialFunctionCall(
    Tiling=ScalarParameter() {defaultValue=1, parameterName=Variation_NoiseScale_01}
    Texture=NamedRerouteUsage()
    LandscapeCoordinates=TextureCoordinate()
   ) {materialFunction=&MF_LandscapeTextureSample}.Result) {name=VariationNoise_01, nodeColor=<rgba 1 0 0.6562 1>}
  NamedRerouteDeclaration(None=MaterialFunctionCall(
    Tiling=ScalarParameter() {defaultValue=2, parameterName=Variation_NoiseScale_02}
    Texture=NamedRerouteUsage()
    LandscapeCoordinates=TextureCoordinate()
   ) {materialFunction=&MF_LandscapeTextureSample}.Result) {name=VariationNoise_02, nodeColor=<rgba 1 0 0.75 1>}
  NamedRerouteDeclaration(None=MaterialFunctionCall(
    Tiling=ScalarParameter() {defaultValue=4, parameterName=Variation_NoiseScale_03}
    Texture=NamedRerouteUsage()
    LandscapeCoordinates=TextureCoordinate()
   ) {materialFunction=&MF_LandscapeTextureSample}.Result) {name=VariationNoise_03, nodeColor=<rgba 0.7813 0 1 1>}
  NamedRerouteDeclaration(None=TextureObject() {texture=&T_NoisePack_01}) {name=RGBNoise_Texture, nodeColor=<rgba 0.7109 1 0 1>}
  NamedRerouteDeclaration(None=TextureCoordinate()) {name=Coordinates, nodeColor=<rgba 0 1 0.0859 1>}
  NamedRerouteDeclaration(None=ComponentMask(None=@n2.Result) {r=true}) {name=BaseColorNoise, nodeColor=<rgba 1 0 0.2812 1>}
  n6 = VectorParameter() {parameterName=BaseColorB, defaultValue=<rgba 0.0174 0.0891 0.3542 1>}
  n7 = VectorParameter() {parameterName=BaseColorA, defaultValue=<rgba 0.3247 0.6101 0.8906 1>}
  n8 = NamedRerouteUsage()
  NamedRerouteDeclaration(None=LinearInterpolate(
    A=LinearInterpolate(A=@n6.RGB, B=@n7.RGB, Alpha=@n8)
    B=VectorParameter() {parameterName=CracksColor, defaultValue=<rgba 1 1 1 1>}.RGB
    Alpha=NamedRerouteUsage()
   )) {name=Ice_BaseColor, nodeColor=<rgba 0 0.3672 1 1>}
  NamedRerouteUsage()
  NamedRerouteUsage()
  NamedRerouteUsage()
  NamedRerouteUsage()
  TextureObject() {samplerType=SAMPLERTYPE_Masks, texture=/Game/Athena/WinterArt/Textures/Cracks_Mask}
  TextureObject() {samplerType=SAMPLERTYPE_Normal, texture=/Game/Characters/Player/Female/Medium/Bodies/F_MED_Bistro_Astronaut/FX/T_GaussianNoise_Normal}
  TextureObject() {samplerType=SAMPLERTYPE_Masks, texture=/Game/Weapons/WeaponSkins/Wraps/Textures/IceBreaker/T_IceBreakerCracks}
  TextureObject() {samplerType=SAMPLERTYPE_Normal, texture=/Game/Environments/Sets/Arid/Textures/T_desert_cracks_N}
  TextureObject() {samplerType=SAMPLERTYPE_Normal, texture=/Game/Global/Textures/Default/Noise/T_Curie_Noise_Normal_Epic}
 ]
}
