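# FN_RvtWriteSwitch is a landscape material function that selects between two material attribute inputs by shading path.
# A Shading Path Switch passes HighEndMaterial on the default path and MobileMaterial on the mobile path to the single output.
# The name suggests it feeds runtime virtual texture writes, though the function itself contains only the switch.
# It is exposed to the material function library under the FN category.
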MaterialFunction /Game/Athena/Asteria/Environments/Landscape/Materials/MaterialFunctions/FN_RvtWriteSwitch
graph {
 properties {bExposeToLibrary=true, libraryCategoriesText [FN]}
 nodes [FunctionOutput(None=ShadingPathSwitch(
    Default=FunctionInput() {inputName=HighEndMaterial, inputType=FunctionInput_MaterialAttributes, sortPriority=1}
    Mobile=FunctionInput() {inputName=MobileMaterial, inputType=FunctionInput_MaterialAttributes}
   ))]
}
