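# MF_AltitudeMask is a landscape material function that outputs a 0 to 1 mask based on world-space altitude.
# It selects one world position channel through a channel mask parameter, Z (blue) by default, and adds the start height parameter, whose negative default offsets the ramp upward.
# The sum is divided by a fade distance parameter and clamped, giving a linear gradient that rises with height.
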MaterialFunction /Game/Athena/Asteria/Environments/Landscape/Materials/MaterialFunctions/MF_AltitudeMask
graph {nodes [
  n1 = Add(
   A=ChannelMaskParameter(None=WorldPosition().XYZ) {maskChannel=Blue, parameterName=AltitudeMaskDirection, defaultValue=<rgba 0 0 1 0>}.R
   B=ScalarParameter() {defaultValue=-5000, parameterName=AltitudeMask_StartHeight}
  )
  n2 = ScalarParameter() {defaultValue=2048, parameterName=AltitudeMaskFadeDistance}
  FunctionOutput(None=Clamp(None=Divide(A=@n1, B=@n2)))
 ]}
