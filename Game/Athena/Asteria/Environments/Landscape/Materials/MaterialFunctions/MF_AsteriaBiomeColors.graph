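# Landscape material function for Asteria that supplies per-biome Grass, GrassEdge, Gravel, Mud and Sand colors.
# Mountain, tropical and desert tint parameters are lerped over common values by biome masks, then passed through MF_AsteriaCustomBiomes.
# Gravel, Mud, Sand and GrassEdge are divided by those tints and saturated, behind static switches.
# Grass uses default, fall-biome or texture colors, while UseSimpleGrassColors substitutes decoded Helios grass colors for Grass and GrassEdge.
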MaterialFunction /Game/Athena/Asteria/Environments/Landscape/Materials/MaterialFunctions/MF_AsteriaBiomeColors
graph {nodes [
  Constant3Vector() {desc="Used for simple landscapes like\r\nCreative && Lobby Island", constant=<rgba 0.0932 0.15 0.03 1>}
  Constant3Vector() {desc=New tropical grass color?, constant=<rgba 0.1224 0.13 0.052 1>}
  Constant3Vector() {constant=<rgba 2.3 1.6585 1.1855 1>}
  Constant3Vector() {constant=<rgba 2.3 1.2548 0.645 1>}
  Constant3Vector() {constant=<rgba 4 2.6421 2.0761 1>}
  Constant3Vector() {desc=basic grass, constant=<rgba 0.696 1.1457 5.8 1>}
  Constant3Vector() {constant=<rgba 1.64 1.886 2.05 1>}
  Constant3Vector() {constant=<rgba 1.0192 1.2729 1.3 1>}
  Constant3Vector() {constant=<rgba 1 0.8963 0.3916 1>}
  Constant3Vector() {constant=<rgba 1 1 1 1>}
  Constant3Vector() {desc=swamp, constant=<rgba 0.8795 0.75 2.5 1>}
  Constant3Vector() {desc=New desert grass color, constant=<rgba 0.229 0.14 0.4 1>}
  Constant3Vector() {desc=Base, constant=<rgba 0.0932 0.15 0.03 1>}
  Constant3Vector() {desc=farm, constant=<rgba 0.6 1.0024 2 1>}
  Constant3Vector() {constant=<rgba 0.671 1.2367 1.1 1>}
  Constant3Vector() {desc=mountains, constant=<rgba 1.4468 1.158 9 1>}
  Constant3Vector() {constant=<rgba 0.543 0.804 0.41 1>}
  Constant() {r=1}
  FunctionInput(Preview=StaticBool() {value=true}) {bUsePreviewValueAsDefault=true, inputName=UseBiomeColorMask, inputType=FunctionInput_StaticBool, sortPriority=10, previewValue=<64 64 64 6>}
  n1 = FunctionInput() {bUsePreviewValueAsDefault=true, inputName=Gravel, sortPriority=-9, previewValue=<1 1 1 1>}
  n2 = LinearInterpolate(
   A=VectorParameter() {group=BiomeSettings, parameterName=Common-Gravel, defaultValue=<rgba 1 1 1 1>}.RGB
   B=Multiply(
    A=VectorParameter() {group=BiomeSettings, parameterName=Mountain-Gravel, defaultValue=<rgba 1 0.6299 0.3198 1>}.RGB
    B=ScalarParameter() {defaultValue=10, parameterName=Mountain-Gravel-Value}
   )
   Alpha=NamedRerouteUsage()
  )
  n3 = Multiply(
   A=VectorParameter() {group=BiomeSettings, parameterName=Tropical-Gravel, defaultValue=<rgba 0.749 0.6372 0.5992 1>}.RGB
   B=ScalarParameter() {defaultValue=0.75, parameterName=TropicalGravel_Val}
  )
  n4 = NamedRerouteUsage()
  n5 = VectorParameter() {desc=mountains, group=BiomeSettings, parameterName=Mountain-GrassEdge, defaultValue=<rgba 1.4468 1.158 9 1>}
  n6 = VectorParameter() {desc=basic grass, group=Grass, parameterName=Common-GrassEdgeColor, sortPriority=-998, defaultValue=<rgba 0.696 1.1457 5.8 1>}
  n7 = LinearInterpolate(A=NamedRerouteDeclaration(None=@n6.RGB) {name=Common-GrassEdge, nodeColor=<rgba 0.3828 0 1 1>}, B=@n5.RGB, Alpha=NamedRerouteUsage())
  n8 = VectorParameter() {desc=swamp, group=BiomeSettings, parameterName=Tropical-GrassEdge, defaultValue=<rgba 0.8795 0.75 2.5 1>}
  n9 = NamedRerouteUsage()
  n10 = VectorParameter() {desc=farm, group=BiomeSettings, parameterName=Desert-GrassEdge, defaultValue=<rgba 2 1.8394 1.5833 1>}
  n11 = LinearInterpolate(
   A=Constant() {r=1}
   B=Multiply(
    A=VectorParameter() {group=BiomeSettings, parameterName=Mountain-Mud, defaultValue=<rgba 1 0.6402 0.5 1>}.RGB
    B=ScalarParameter() {defaultValue=5, parameterName=Mountain-Mud-Value}
   )
   Alpha=NamedRerouteUsage()
  )
  n12 = Multiply(
   A=VectorParameter() {group=BiomeSettings, parameterName=Tropical-Mud, defaultValue=<rgba 0.749 0.5206 0.5505 1>}.RGB
   B=ScalarParameter() {defaultValue=0.9, parameterName=TropicalMud_Val}
  )
  n13 = NamedRerouteUsage()
  n14 = LinearInterpolate(
   A=Constant() {r=1}
   B=Multiply(
    A=VectorParameter() {group=BiomeSettings, parameterName=Mountain-Sand, defaultValue=<rgba 1 0.4638 0.3 1>}.RGB
    B=ScalarParameter() {defaultValue=15, parameterName=Mountain-SandValue}
   )
   Alpha=NamedRerouteUsage()
  )
  n15 = VectorParameter() {group=BiomeSettings, parameterName=Tropical-Sand, defaultValue=<rgba 1 1 1 1>}
  n16 = NamedRerouteUsage()
  n17 = MaterialFunctionCall!(
   Grass=LinearInterpolate(A=LinearInterpolate(A=@n7, B=@n8.RGB, Alpha=@n9), B=@n10.RGB, Alpha=NamedRerouteUsage())
   Gravel=LinearInterpolate(
    A=LinearInterpolate(A=@n2, B=@n3, Alpha=@n4)
    B=VectorParameter() {group=BiomeSettings, parameterName=Desert-Gravel, defaultValue=<rgba 1.5 1.0255 0.5406 1>}.RGB
    Alpha=NamedRerouteUsage()
   )
   Mud=LinearInterpolate(
    A=LinearInterpolate(A=@n11, B=@n12, Alpha=@n13)
    B=VectorParameter() {group=BiomeSettings, parameterName=Desert-Mud, defaultValue=<rgba 1.75 1.069 0.3181 1>}.RGB
    Alpha=NamedRerouteUsage()
   )
   Sand=LinearInterpolate(
    A=LinearInterpolate(A=@n14, B=@n15.RGB, Alpha=@n16)
    B=VectorParameter() {group=BiomeSettings, parameterName=Desert-Sand, defaultValue=<rgba 2.05 1.4503 1.1378 1>}.RGB
    Alpha=NamedRerouteUsage()
   )
  ) {materialFunction=&/MF_AsteriaCustomBiomes}
  n18 = Divide(A=@n1, B=@n17.Gravel)
  FunctionOutput(None=StaticSwitch(True=Saturate(None=@n18), False=@n1, Value=NamedRerouteUsage())) {desc=This is used in the mobile path, outputName=Gravel, sortPriority=-9}
  n19 = FunctionInput() {bUsePreviewValueAsDefault=true, inputName=Mud, sortPriority=-8, previewValue=<1 1 1 1>}
  n20 = Divide(A=@n19, B=@n17.Mud)
  FunctionOutput(None=StaticSwitch(True=Saturate(None=@n20), False=@n19, Value=NamedRerouteUsage())) {outputName=Mud, sortPriority=-8}
  n21 = FunctionInput() {bUsePreviewValueAsDefault=true, inputName=Sand, sortPriority=-7, previewValue=<1 1 1 1>}
  n22 = Divide(A=@n21, B=@n17.Sand)
  FunctionOutput(None=StaticSwitch(True=Saturate(None=@n22), False=@n21, Value=NamedRerouteUsage())) {outputName=Sand, sortPriority=-7}
  n23 = Desaturation(None=NamedRerouteUsage(), Fraction=Constant() {r=0.2})
  n24 = Constant3Vector() {constant=<rgba 0.8 0.8 1.2 0>}
  n25 = NamedRerouteUsage()
  n26 = Max(A=Desaturation(None=@n25, Fraction=Constant()), B=Constant() {r=0.13})
  n27 = NamedRerouteUsage()
  n28 = StaticSwitch(
   True=Saturate(None=Divide(A=@n26, B=@n17.Grass))
   False=Saturate(None=Divide(A=@n26, B=@n27) {desc=Default for the base grass edge color only})
   Value=NamedRerouteUsage()
  )
  FunctionOutput(None=StaticSwitchParameter(True=Multiply(A=@n23, B=@n24), False=@n28) {group=Grass, parameterName=UseSimpleGrassColors}) {outputName=GrassEdge, sortPriority=-10}
  n29 = SetMaterialAttributes!(MaterialAttributes, BaseColor=Constant() {r=1}) {attributeSetTypes [#1]}
  n30 = GetMaterialAttributes(None=MaterialFunctionCall!(SetMaterialAttributes=@n29) {materialFunction=/Game/Athena/Helios/Environments/Landscape/Materials/MaterialFunctions_Utilities/MF_Decode_Grass_Biome_Colors}.Blended Material) {attributeGetTypes [#1]}
  n31 = VectorParameter() {
   desc="Used for simple landscapes like\r\nCreative && Lobby Island"
   group=Grass
   parameterName=DefaultGrassColor_NoBiomes
   sortPriority=-999
   defaultValue=<rgba 0.75 0.75 0.75 1>
  }
  n32 = NamedRerouteUsage()
  n33 = VectorParameter() {
   desc="Used for simple landscapes like\r\nCreative && Lobby Island"
   group=BiomeSettings
   parameterName=FallBiome-DefaultGrassColor
   defaultValue=<rgba 1 1 1 1>
  }
  n34 = StaticSwitch(
   True=StaticSwitchParameter(True=NamedRerouteUsage(), False=LinearInterpolate(A=@n31.RGB, B=@n33.RGB, Alpha=@n32)) {group=BiomeSettings, parameterName=UseGrassColorFromTexture}
   False=@n31.RGB
   Value=NamedRerouteUsage()
  )
  FunctionOutput(None=StaticSwitchParameter(True=NamedRerouteDeclaration(None=@n30.BaseColor) {name=SimpleColorTInt, nodeColor=<rgba 1 0 0.9141 1>}, False=@n34) {group=Grass, parameterName=UseSimpleGrassColors}) {outputName=Grass, sortPriority=-11}
  LinearInterpolate(
   A=VectorParameter() {group=BiomeSettings, parameterName=TestNewGrassColor, defaultValue=<rgba 0.0932 0.15 0.03 1>}.RGB
   B=@n34
   Alpha=ScalarParameter() {group=BiomeSettings, parameterName=debug-MFBC-testgg}
  )
  LinearInterpolate(A=@n25, B=@n28, Alpha=ScalarParameter() {defaultValue=1, group=BiomeSettings, parameterName=debug-MFBC-test-edge-overrideA})
  MaterialFunctionCall!(SetMaterialAttributes=@n29) {materialFunction=/Game/Athena/Helios/Environments/Landscape/Materials/MaterialFunctions_Utilities/MF_Athena_Grass_Biome_Colors}
  Multiply(A=@n33.RGB, B=Desaturation(None=NamedRerouteUsage(), Fraction=Constant() {r=1}))
  Multiply(A=@n17.Mud, B=@n19)
  Multiply(A=@n17.Gravel, B=@n1)
  Multiply(A=@n17.Sand, B=@n21)
  n35 = MaterialFunctionCall!(UseLandscapeCoords) {materialFunction=&/MF_AsteriaTerrainMasks}
  NamedRerouteDeclaration(None=@n35.BiomeGrassColors) {name=BiomeGrassColors, nodeColor=<rgba 0.3828 0 1 1>}
  n36 = MaterialFunctionCall(Float4=FunctionInput(Preview=@n35.BiomeMaskRGBA) {bUsePreviewValueAsDefault=true, inputName=BiomeMask-RGBA, inputType=FunctionInput_Vector4}) {materialFunction=&BreakOutFloat4Components}
  NamedRerouteDeclaration(None=@n36.R) {name=Tundra, nodeColor=<rgba 0.3828 0 1 1>}
  NamedRerouteDeclaration(None=@n36.G) {name=GrassLand, nodeColor=<rgba 0.3828 0 1 1>}
  NamedRerouteDeclaration(None=@n36.B) {name=Desert, nodeColor=<rgba 0.3828 0 1 1>}
  NamedRerouteDeclaration(None=FunctionInput() {bUsePreviewValueAsDefault=true, inputName=Grass, sortPriority=-10, previewValue=<1 1 1 1>}) {name=Input_GrassColor, nodeColor=<rgba 0 1 0.7422 1>}
  n37 = StaticBoolParameter() {
   defaultValue=true
   desc="Leave this false for UEFN.\r\nTrue=Uses the BiomeTexture.\r\nFalse=Uses the base material properties"
   group=BiomeSettings
   parameterName=BiomeColors-UseDefault
   sortPriority=-500
  }
  NamedRerouteDeclaration(None=@n37) {name=UseBiomeColorMask, nodeColor=<rgba 0.3828 0 1 1>}
  n38 = FunctionInput(Preview=StaticBool()) {
   bUsePreviewValueAsDefault=true
   desc="Try leaving this unhooked for now - not sure if anything needs this && also it might be too easy to break stuff with it"
   inputName=UseLandscapeCoords
   inputType=FunctionInput_StaticBool
   sortPriority=11
  }
  Reroute(None=@n38)
 ]}
