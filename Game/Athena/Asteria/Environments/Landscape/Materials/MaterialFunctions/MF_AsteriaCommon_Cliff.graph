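# Material function for the Asteria landscape cliff layer, outputting material, height, cliff-face height and emissive mask, with displacement and scree mask left constant.
# Medium and large rock sets use world-aligned projection, with noise-driven height blending for their material and emissive blends.
# A rock detail texture modulates base color and specular before MF_CliffZColor applies the CustomBiomeMask input.
# Water streaks are switched off, and the altitude-darkening branch feeds no output.
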MaterialFunction /Game/Athena/Asteria/Environments/Landscape/Materials/MaterialFunctions/MF_AsteriaCommon_Cliff
graph {nodes [
  n1 = GetMaterialAttributes(None=NamedRerouteUsage()) {attributeGetTypes [
    #1
    #2
    #3
    #4
   ]}
  n2 = MaterialFunctionCall!(
   Diffuse Texture Object=TextureObject() {texture=/Game/Environments/Apollo/Rocks/Natural_Rocks/Textures/T_RockDetails_02D}
   Spec Texture Object
   TextureSize=Constant() {r=700}
   --------------------------
   Use seperate settings for Z projection
   Use seperate texture for Z projection
   Texture Size Z projection
   Z Plane Diffuse Texture Object
   Z Plane Spec Texture Object
   "--------------------------      "
   ProjectionTransitionContrast=Constant() {r=5}
   Use Texture Alpha Heightmap to Refine the Transistion
   HeightLerpTransitionContrast
   PickTheDiffuseHeightChannel
   TheDiffuseHeightChannel
   "--------------------------       "
   NormalTextureObject
   World Space Normals
   "--------------------------  "
   Export Float 4 Diffuse
   Export Float 4 Spec
   "--------------------------     "
   WorldPosition
   World Space Normal
  ) {materialFunction=/Engine/Functions/Engine_MaterialFunctions02/Texturing/WorldAlignedTextures_Complex}
  n3 = SquareRoot(None=@n2.XY Diffuse Texture)
  n4 = Saturate(None=Multiply(A=Multiply(A=@n1.BaseColor, B=@n3), B=Constant() {r=1.5}))
  n5 = ComponentMask(None=@n2.XY Diffuse Texture) {r=true}
  n6 = Add(A=Constant() {r=-1.1}, B=Multiply(A=Constant() {r=2.05}, B=@n5))
  n7 = LinearInterpolate(A=Constant() {r=7}, B=Constant() {r=-0.8}, Alpha=@n5)
  n8 = LinearInterpolate(A=Constant() {r=-0.25}, B=Constant() {r=1.5}, Alpha=@n1.Specular)
  n9 = Multiply(A=Multiply(A=Saturate(None=@n6), B=Saturate(None=@n7)), B=Clamp(None=Saturate(None=@n8), Max=Constant() {r=0.5}) {clampMode=CMODE_ClampMax})
  n10 = StaticSwitch(
   True=SetMaterialAttributes!(MaterialAttributes=@n1.MaterialAttributes, BaseColor=@n4, Specular=@n9) {attributeSetTypes [#4, #3]}
   False=NamedRerouteUsage()
   Value=StaticBool() {value=true}
  )
  n11 = FunctionInput() {bUsePreviewValueAsDefault=true, inputName=CustomBiomeMask, inputType=FunctionInput_Scalar, sortPriority=1, previewValue=<1 0 0 1>}
  FunctionOutput(None=MaterialFunctionCall!(In=@n10, CustomBiomeMask=@n11) {materialFunction=&/MF_CliffZColor}.Result) {outputName=Material, sortPriority=-1}
  FunctionOutput(None=Constant()) {outputName=Displacement, sortPriority=20}
  FunctionOutput(None=Constant()) {outputName=SlopeScreeMask, sortPriority=10}
  n12 = NamedRerouteUsage()
  n13 = Saturate(None=Multiply(A=OneMinus(None=@n12), B=Constant() {r=3}))
  n14 = MaterialFunctionCall(Transition Phase=@n13, Height Texture=NamedRerouteUsage()) {materialFunction=&HeightLerp}
  FunctionOutput(None=NamedRerouteDeclaration(None=Max(A=@n13, B=@n14.Lerp Alpha No Contrast) {desc="For BLEND MATERIAL, smallrockedge-mat should be in B"}) {name=OutputHeight, nodeColor=<rgba 1 0 1 1>}) {outputName=Height}
  n15 = TextureSample(UVs=MaterialFunctionCall!(Scale=Constant() {r=-10000}) {materialFunction=/Engine/Functions/Engine_MaterialFunctions01/Coordinates/WorldPosition-XY}.Result) {
   samplerSource=SSM_Wrap_WorldGroupSettings
   samplerType=SAMPLERTYPE_LinearGrayscale
   texture=/Game/Global/Textures/Default/Noise/T_SwirlNoise_N_height
  }
  n16 = MaterialFunctionCall!(
   Diffuse 1
   Diffuse 2
   Transistion Phase=@n15.R
   Height Texture 1=NamedRerouteUsage()
   Height Texture 2=NamedRerouteUsage()
   Contrast=Constant() {desc="Play with neg or pos ##", r=-5}
  ) {materialFunction=/Engine/Functions/Engine_MaterialFunctions02/Texturing/HeightLerpWithTwoHeightMaps}
  FunctionOutput(None=@n16.Resulting Height Map) {outputName=CliffFaceHeightMap}
  n17 = NamedRerouteUsage()
  n18 = NamedRerouteUsage()
  FunctionOutput(None=NamedRerouteDeclaration(None=LinearInterpolate(A=@n17, B=@n18, Alpha=@n16.Alpha)) {name=EmissiveMask_Result, nodeColor=<rgba 0 1 0.3906 1>}) {outputName=CliffEmissiveMask}
  n19 = BlendMaterialAttributes(A=NamedRerouteUsage(), B=NamedRerouteUsage(), Alpha=@n16.Alpha)
  NamedRerouteDeclaration(None=@n19) {name=BlendedMaterialBeforeDetailTexture, nodeColor=<rgba 1 0 1 1>}
  n20 = TextureObjectParameter() {
   group=Cliff
   parameterName=T_MediumRock_S
   samplerType=SAMPLERTYPE_Masks
   texture=/Game/Environments/Artemis/Sets/Terrain/Textures/T_Artemis_Cliff_03B_S
  }
  n21 = TextureObjectParameter() {
   group=Cliff
   parameterName=T_MediumRock_N
   samplerType=SAMPLERTYPE_Normal
   texture=/Game/Environments/Artemis/Sets/Terrain/Textures/T_Artemis_Cliff_03B_N
  }
  n22 = MaterialFunctionCall(
   Diffuse Texture Object=TextureObjectParameter() {group=Cliff, parameterName=T_MeduimRock_D, texture=/Game/Environments/Artemis/Sets/Terrain/Textures/T_Artemis_Cliff_03B_D}
   Spec Texture Object=@n20
   TextureSize=VectorParameter() {group=Cliff, parameterName=MediumRock_Scale, defaultValue=<rgba 3500 3500 3000 1>}.RGB
   ProjectionTransitionContrast=ScalarParameter() {defaultValue=0.5, group=Cliff, parameterName=MR-PTC}
   Use Texture Alpha Heightmap to Refine the Transistion=StaticBool() {value=true}
   HeightLerpTransitionContrast=ScalarParameter() {defaultValue=0.5, group=Cliff, parameterName=MR-HLTC}
   NormalTextureObject=@n21
   World Space Normals=StaticBool()
   Export Float 4 Diffuse=StaticBool() {value=true}
   WorldPosition=WorldPosition() {worldPositionShaderOffset=WPT_ExcludeAllShaderOffsets}.XYZ
  ) {materialFunction=&WorldAlignedTextures_Complex}
  NamedRerouteDeclaration(None=ComponentMask(None=@n22.XY Spec) {g=true}) {name=MediumRock_EmissiveMask, nodeColor=<rgba 0 0.3843 1 1>}
  n23 = SetMaterialAttributes!(
   MaterialAttributes
   BaseColor=ComponentMask(None=@n22.XY Diffuse Texture) {b=true, g=true, r=true}
   Specular=ComponentMask(None=@n22.XY Spec) {r=true}
   Roughness=ComponentMask(None=@n22.XY Spec) {b=true}
   Normal=@n22.Normal XY Texture
  ) {attributeSetTypes [
    #4
    #3
    #2
    #1
   ]}
  NamedRerouteDeclaration(None=@n23) {name=MediumRock-Material, nodeColor=<rgba 1 0 1 1>}
  n24 = TextureObjectParameter() {
   group=Cliff
   parameterName=T_LargeRock_S
   samplerType=SAMPLERTYPE_Masks
   texture=/Game/Environments/Artemis/Sets/Terrain/Textures/T_Artemis_Cliff_04B_S
  }
  n25 = TextureObjectParameter() {
   group=Cliff
   parameterName=T_LargeRock_N
   samplerType=SAMPLERTYPE_Normal
   texture=/Game/Environments/Artemis/Sets/Terrain/Textures/T_Artemis_Cliff_04B_N
  }
  n26 = MaterialFunctionCall(
   Diffuse Texture Object=TextureObjectParameter() {group=Cliff, parameterName=T_LargeRock_D, texture=/Game/Environments/Artemis/Sets/Terrain/Textures/T_Artemis_Cliff_04B_D}
   Spec Texture Object=@n24
   TextureSize=VectorParameter() {group=Cliff, parameterName=LargeRock_Scale, defaultValue=<rgba 3500 3500 3500 1>}.RGB
   ProjectionTransitionContrast=ScalarParameter() {defaultValue=0.5, group=Cliff, parameterName=LR-PTC}
   Use Texture Alpha Heightmap to Refine the Transistion=StaticBool() {value=true}
   HeightLerpTransitionContrast=ScalarParameter() {defaultValue=0.5, group=Cliff, parameterName=LR-HLTC}
   NormalTextureObject=@n25
   World Space Normals=StaticBool()
   Export Float 4 Diffuse=StaticBool() {value=true}
   WorldPosition=WorldPosition() {worldPositionShaderOffset=WPT_ExcludeAllShaderOffsets}.XYZ
  ) {materialFunction=&WorldAlignedTextures_Complex}
  n27 = SetMaterialAttributes(
   BaseColor=ComponentMask(None=@n26.XY Diffuse Texture) {b=true, g=true, r=true}
   Specular=ComponentMask(None=@n26.XY Spec) {r=true}
   Roughness=ComponentMask(None=@n26.XY Spec) {b=true}
   Normal=@n26.Normal XY Texture
  ) {attributeSetTypes [
    #4
    #3
    #2
    #1
   ]}
  NamedRerouteDeclaration(None=@n27) {name=LargeRock-Material, nodeColor=<rgba 1 0 1 1>}
  NamedRerouteDeclaration(None=ComponentMask(None=@n26.XY Diffuse Texture) {a=true}) {name=LargeRock-Height, nodeColor=<rgba 1 0 1 1>}
  NamedRerouteDeclaration(None=ComponentMask(None=@n22.XY Diffuse Texture) {a=true}) {name=MediumRock-Height, nodeColor=<rgba 1 0 1 1>}
  n28 = Multiply(A=NamedRerouteUsage(), B=Constant() {r=-185}) {constB=0.2}
  n29 = WorldPosition()
  n30 = ComponentMask(None=MaterialFunctionCall(
    Diffuse Texture Object=TextureObject() {samplerType=SAMPLERTYPE_LinearColor, texture=/Game/Environments/Apollo/Sets/Terrain/Textures/T_Apollo_Cliff_Drip}
    TextureSize=Constant3Vector() {constant=<rgba 6000 6000 4500 1>}
    ProjectionTransitionContrast=Constant() {r=5}
    WorldPosition=Add(A=@n28, B=@n29.XYZ)
   ) {materialFunction=&WorldAlignedTextures_Complex}.XY Diffuse Texture) {b=true}
  n31 = LinearInterpolate(A=Constant() {r=-3}, B=Constant() {r=1}, Alpha=@n30)
  n32 = Constant() {r=1}
  n33 = StaticBool() {desc=Use Water Streak}
  NamedRerouteDeclaration(None=StaticSwitch(True=Saturate(None=@n31), False=@n32, Value=@n33)) {name=WaterStreak-Spec, nodeColor=<rgba 1 0 1 1>}
  n34 = Constant3Vector() {constant=<rgba 1 1 1 1>}
  n35 = Constant() {r=1}
  n36 = Power(Base=@n30, Exp=Constant() {r=8})
  NamedRerouteDeclaration(None=StaticSwitch(True=LinearInterpolate(A=@n34, B=@n35, Alpha=@n36), False=@n32, Value=@n33)) {name=WaterStreak-Color, nodeColor=<rgba 1 0 1 1>}
  NamedRerouteDeclaration(None=ComponentMask(None=@n26.XY Spec) {g=true}) {name=LargeRock_EmissiveMask, nodeColor=<rgba 0 1 0.8594 1>}
  n37 = ComponentMask(None=VertexNormalWS()) {b=true}
  NamedRerouteDeclaration(None=FunctionInput(Preview=Abs(None=@n37)) {bUsePreviewValueAsDefault=true, inputName=Slope, inputType=FunctionInput_Scalar, sortPriority=4}) {name=BasicSlope, nodeColor=<rgba 1 0 1 1>}
  NamedRerouteDeclaration(None=Constant()) {
   name=SmallRockEdge-UnmodifiedHeight
   desc="This was previously using the alpha channel of T_Apollo_Dirt_D_02\r\nw/o any modifiers.. There are some very slight, but acceptable\r\ndifferences attributed to T_Apollo_Dirt_S_02 not being in sRGB space."
   nodeColor=<rgba 1 0 1 1>
  }
  n38 = TextureSample(UVs=MaterialFunctionCall(Scale=Constant() {r=1200}) {materialFunction=&WorldPosition-XY}.Result) {
   samplerSource=SSM_Wrap_WorldGroupSettings
   samplerType=SAMPLERTYPE_LinearGrayscale
   texture=&T_SwirlNoise_N_height
  }
  NamedRerouteDeclaration(None=OneMinus(None=@n38.A)) {
   name=SmallRockEdge-Height
   desc="This was previously using the alpha channel of T_Apollo_Dirt_D_02\r\nw/o any modifiers.. There are some very slight, but acceptable\r\ndifferences attributed to T_Apollo_Dirt_S_02 not being in sRGB space."
   nodeColor=<rgba 1 0 1 1>
  }
  n39 = Saturate(None=LinearInterpolate(A=Constant() {r=-6}, B=Constant() {r=2.5}, Alpha=NamedRerouteUsage()) {desc=Adjust the outer edge of the small rock})
  n40 = Constant() {r=5}
  n41 = Constant() {r=-2}
  n42 = NamedRerouteUsage()
  n43 = Add(A=Power(Base=NamedRerouteUsage(), Exp=Constant() {r=3}), B=Saturate(None=LinearInterpolate(A=@n40, B=@n41, Alpha=@n42)))
  n44 = Multiply(A=Power(Base=NamedRerouteUsage(), Exp=Constant() {r=5}), B=Constant() {r=3})
  n45 = OneMinus(None=MaterialFunctionCall(Transition Phase=Saturate(None=@n43), Height Texture=Saturate(None=@n44)) {materialFunction=&HeightLerp}.Lerp Alpha No Contrast)
  n46 = Constant() {r=-1}
  n47 = Constant() {r=1.15}
  n48 = NamedRerouteUsage()
  n49 = Multiply(A=Saturate(None=Add(A=@n39, B=@n45)), B=Saturate(None=LinearInterpolate(A=@n46, B=@n47, Alpha=@n48)))
  n50 = NamedRerouteUsage()
  n51 = Constant() {r=1}
  NamedRerouteDeclaration(None=Multiply(
    A=MaterialFunctionCall(Transition Phase=@n49, Height Texture=@n50, Contrast=@n51) {materialFunction=&HeightLerp}.Alpha
    B=Constant() {r=0.75}
   )) {name=SmallRockEdgeMask, nodeColor=<rgba 1 0 1 1>}
  NamedRerouteDeclaration(None=GetMaterialAttributes(None=@n19) {attributeGetTypes [#1]}.Normal) {name=MediumAndLargeRockNormals, nodeColor=<rgba 1 0 1 1>}
  n52 = Saturate(None=LinearInterpolate(A=Constant() {r=-5}, B=Constant() {r=1.4}, Alpha=@n1.Roughness))
  n53 = SetMaterialAttributes!(MaterialAttributes=@n1.MaterialAttributes, Normal=@n1.Normal, Roughness=@n52, Specular=@n9, BaseColor=@n4) {attributeSetTypes [
    #1
    #2
    #3
    #4
   ]}
  n54 = SetMaterialAttributes(MaterialAttributes=@n1.MaterialAttributes, Normal=@n1.Normal, Roughness=@n52, Specular=@n9, BaseColor=@n4) {attributeSetTypes [
    #1
    #2
    #3
    #4
   ]}
  QualitySwitch(Default=@n53, Epic=@n54)
  n55 = NamedRerouteUsage()
  Reroute(None=Saturate(None=Divide(A=@n4, B=@n55)))
  Reroute(None=Multiply(A=@n9, B=NamedRerouteUsage()))
  ScalarParameter() {defaultValue=3, group=Cliff, parameterName=6y4et87y3rtugg}
  n56 = MaterialFunctionCall!(DarknessAmount) {
   desc=Cliffs get darker with increasing altitude
   materialFunction=/Game/Athena/Apollo/Environments/Landscape/Materials/Biome/CommonMaterials/MF_HeightBasedDarkenCliff
  }
  n57 = Add(A=@n56.WorldZ, B=Constant() {r=3950})
  n58 = Constant() {r=250}
  n59 = Max(A=Saturate(None=Divide(A=@n57, B=@n58)), B=Constant() {r=0.5})
  n60 = GetMaterialAttributes(None=@n10) {attributeGetTypes [#4, #2]}
  SetMaterialAttributes!(
   MaterialAttributes=@n10
   BaseColor=Multiply(A=@n59, B=Multiply(A=@n56.Result, B=@n60.BaseColor))
   Roughness=Multiply(A=@n59, B=@n60.Roughness)
  ) {attributeSetTypes [#4, #2]}
 ]}
