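# Material function that builds grass color variation masks for the Asteria landscape from world-space noise.
# World XY position, excluding shader offsets by default, divided by scale parameters, samples one tiling Gaussian noise texture for three micro masks and one macro mask.
# Micro masks 1–3 are remapped by min/max lerps and saturated, with mask 3's position offset.
# Raw outputs include macro, unmodified mask 3 and mask 4, which reuses mask 2's sample.
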MaterialFunction /Game/Athena/Asteria/Environments/Landscape/Materials/MaterialFunctions/MF_GrassMicroColorMasks
graph {nodes [
  n1 = ScalarParameter() {group=Grass, parameterName=MicroColorMask2_Blend_Min, sortPriority=-44}
  n2 = ScalarParameter() {defaultValue=1, group=Grass, parameterName=MicroColorMask2_Blend_Max, sortPriority=-43}
  n3 = FunctionInput(Preview=WorldPosition() {worldPositionShaderOffset=WPT_ExcludeAllShaderOffsets}.XYZ) {bUsePreviewValueAsDefault=true, inputName=SamplingPosition}
  n4 = ComponentMask(None=@n3) {g=true, r=true}
  n5 = TextureSample(UVs=Divide(A=@n4, B=ScalarParameter() {defaultValue=12000, group=Grass, parameterName=MicroGrassMaskScale_02, sortPriority=-45})) {samplerSource=SSM_Wrap_WorldGroupSettings, texture=/Game/Effects/Fort_Effects/Textures/Tile/T_GaussianNoise_01_M}
  FunctionOutput(None=Saturate(None=LinearInterpolate(A=@n1, B=@n2, Alpha=@n5.G))) {outputName=GrassColorMask2, sortPriority=1}
  n6 = Constant() {r=500}
  n7 = TextureSample(UVs=Divide(A=Add(A=@n4, B=@n6), B=ScalarParameter() {defaultValue=10000, group=Grass, parameterName=MicroGrassMaskScale_03, sortPriority=-40})) {samplerSource=SSM_Wrap_WorldGroupSettings, texture=&T_GaussianNoise_01_M}
  FunctionOutput(None=@n7.G) {outputName=GrassColorMask3-NoMod, sortPriority=3}
  n8 = ComponentMask(None=@n3) {g=true, r=true}
  n9 = ScalarParameter() {defaultValue=25000, group=Grass, parameterName=MacroGrassColorMaskScale, sortPriority=-30}
  FunctionOutput(None=TextureSample(UVs=Divide(A=@n8, B=@n9)) {samplerSource=SSM_Wrap_WorldGroupSettings, texture=&T_GaussianNoise_01_M}.G) {outputName=GrassMacroColorMask, sortPriority=5}
  n10 = ScalarParameter() {defaultValue=5000, group=Grass, parameterName=MicroGrassMaskScale_01, sortPriority=-50}
  n11 = LinearInterpolate(
   A=ScalarParameter() {group=Grass, parameterName=MicroColorMask1_Blend_Min, sortPriority=-49}
   B=ScalarParameter() {defaultValue=1, group=Grass, parameterName=MicroColorMask1_Blend_Max, sortPriority=-48}
   Alpha=TextureSample(UVs=Divide(A=@n4, B=@n10)) {samplerSource=SSM_Wrap_WorldGroupSettings, texture=&T_GaussianNoise_01_M}.G
  )
  NamedRerouteDeclaration(None=FunctionOutput(None=Saturate(None=@n11)) {outputName=GrassColorMask1}) {name=GrassColorMask1, nodeColor=<rgba 1 0 0.8203 1>}
  n12 = LinearInterpolate(
   A=ScalarParameter() {group=Grass, parameterName=MicroColorMask3_Blend_Min, sortPriority=-39}
   B=ScalarParameter() {defaultValue=1, group=Grass, parameterName=MicroColorMask3_Blend_Max, sortPriority=-38}
   Alpha=@n7.G
  )
  NamedRerouteDeclaration(None=FunctionOutput(None=Saturate(None=@n12)) {outputName=GrassColorMask3, sortPriority=2}) {name=GrassColorMask3, nodeColor=<rgba 1 0 0.8203 1>}
  NamedRerouteDeclaration(None=FunctionOutput(None=@n5.G) {outputName=GrassColorMask4, sortPriority=4}) {name=GrassColorMask4, nodeColor=<rgba 1 0 0.8203 1>}
 ]}
