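# Material function that builds Asteria landscape grass colors, with base, Sakura and Fall palettes.
# MF_GrassMicroColorMasks drives layered micro and macro color lerps, with biome palette blending behind a static switch meant off for UEFN.
# The NORVT output adds biome tinting and optional tip color, with a camera-distance sphere-mask blend only on the default shading path.
# It also outputs color masks plus detail and height from a landscape texture sample.
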MaterialFunction /Game/Athena/Asteria/Environments/Landscape/Materials/MaterialFunctions/MF_AsteriaGrassColors
graph {nodes [
  n1 = FunctionInput() {bUsePreviewValueAsDefault=true, inputName=BiomeMask-RGBA, inputType=FunctionInput_Vector4}
  ComponentMask(None=@n1) {g=true}
  Constant()
  Constant() {desc="New: use neg ## to match with any world pos normals", r=800}
  n2 = VectorParameter() {group=Grass, parameterName=NewGrass-TipsColor, defaultValue=<rgba 0.2333 0.4427 0 1>}
  n3 = Constant() {r=0.01}
  n4 = NamedRerouteUsage()
  n5 = NamedRerouteUsage()
  n6 = NamedRerouteUsage()
  n7 = CameraPositionWS()
  n8 = Constant() {r=20000}
  n9 = Constant()
  n10 = ShadingPathSwitch(Default=LinearInterpolate(A=@n5, B=NamedRerouteUsage(), Alpha=SphereMask(A=@n6, B=@n7, Radius=@n8, Hardness=@n9)), Forward=@n5, Mobile=@n5)
  n11 = NamedRerouteUsage()
  n12 = Saturate(None=Multiply(
    A=@n10
    B=MaterialFunctionCall!(Grass=@n10, Gravel, Mud, Sand, BiomeMask-RGBA=@n11, UseBiomeColorMask, UseLandscapeCoords) {materialFunction=&/MF_AsteriaBiomeColors}.Grass
   ))
  n13 = LinearInterpolate(
   A=@n12
   B=MaterialFunctionCall(Grass=@n12, BiomeMask-RGBA=NamedRerouteUsage()) {materialFunction=&MF_AsteriaBiomeColors}.GrassEdge
   Alpha=ScalarParameter() {group=Grass, parameterName=GrassEdgeBlend}
  )
  n14 = Saturate(None=Add(A=Multiply(A=@n3, B=@n4), B=Multiply(A=@n4, B=@n13)))
  n15 = Saturate(None=Add(A=@n14, B=NamedRerouteDeclaration(None=@n2.RGB) {name=GrassTipColor, nodeColor=<rgba 1 0 0.8203 1>}))
  n16 = Power(Base=NamedRerouteUsage(), Exp=Constant() {r=3})
  n17 = StaticBool()
  FunctionOutput(None=StaticSwitch(
    True=@n12
    False=LinearInterpolate(A=@n14, B=@n15, Alpha=@n16)
    Value=FunctionInput(Preview=@n17) {bUsePreviewValueAsDefault=true, inputName=IsGrass, inputType=FunctionInput_StaticBool, sortPriority=201}
   )) {outputName=NORVT_GrassColors}
  n18 = VectorParameter() {group=Grass, parameterName=BaseGrass_Macro_Color1, defaultValue=<rgba 0.0078 0.102 0.0392 1>}
  n19 = LinearInterpolate(
   A=VectorParameter() {group=Grass, parameterName=BaseGrass_Color1, defaultValue=<rgba 0.0154 0.2 0.0769 1>}.RGB
   B=VectorParameter() {group=Grass, parameterName=BaseGrass_Color2, defaultValue=<rgba 0.0295 0.151 0.108 1>}.RGB
   Alpha=NamedRerouteUsage()
  )
  n20 = VectorParameter() {group=Grass, parameterName=BaseGrass_Color3, defaultValue=<rgba 0.0786 0.25 0.0729 1>}
  n21 = NamedRerouteUsage()
  n22 = Constant() {r=4}
  n23 = NamedRerouteUsage()
  n24 = LinearInterpolate(
   A=@n18.RGB
   B=LinearInterpolate(
    A=LinearInterpolate(A=@n19, B=@n20.RGB, Alpha=@n21)
    B=VectorParameter() {group=Grass, parameterName=BaseGrass_Color4, defaultValue=<rgba 0.0136 0.1719 0.0495 1>}.RGB
    Alpha=NamedRerouteUsage()
   )
   Alpha=Saturate(None=Multiply(A=@n23, B=@n22))
  )
  n25 = LinearInterpolate(A=LinearInterpolate(A=@n24, B=NamedRerouteUsage(), Alpha=NamedRerouteUsage()), B=NamedRerouteUsage(), Alpha=NamedRerouteUsage())
  n26 = StaticBoolParameter() {
   defaultValue=true
   desc="Leave this false for UEFN.\r\nTrue=Uses the BiomeTexture.\r\nFalse=Uses the base material properties"
   group=Grass
   parameterName=UseSpecialGrassBiomeColors
   sortPriority=-500
  }
  FunctionOutput(None=NamedRerouteDeclaration(None=StaticSwitch(True=@n25, False=@n24, Value=@n26)) {name=MicroGrassColors, nodeColor=<rgba 1 0 0.9844 1>}) {outputName=MicroGrassColors}
  n27 = MaterialFunctionCall!(SamplingPosition=NamedRerouteUsage()) {materialFunction=&/MF_GrassMicroColorMasks}
  FunctionOutput(None=@n27.GrassColorMask3-NoMod) {outputName=GrassColorMask3-NoMod, sortPriority=10}
  n28 = MaterialFunctionCall(Float3=@n1) {materialFunction=&BreakOutFloat3Components}
  NamedRerouteDeclaration(None=@n28.B) {name=SakuraBiomeBlend, nodeColor=<rgba 1 0 0.2578 1>}
  NamedRerouteDeclaration(None=@n1) {name=BiomeMask-RGBA, nodeColor=<rgba 0 0.4609 1 1>}
  NamedRerouteDeclaration(None=FunctionOutput(None=@n27.GrassColorMask3) {outputName=GrassColorMask3, sortPriority=10}) {name=GrassColorMask3, nodeColor=<rgba 1 0 0.8203 1>}
  NamedRerouteDeclaration(None=FunctionOutput(None=@n27.GrassColorMask4) {outputName=GrassColorMask4, sortPriority=11}) {name=GrassColorMask4, nodeColor=<rgba 1 0 0.8203 1>}
  NamedRerouteDeclaration(None=VectorParameter() {group=Grass, parameterName=Fall-MacroGrassColor, defaultValue=<rgba 0.2784 0.1686 0.0078 1>}.RGB) {
   name=Fall-MacroGrassColorBlend
   desc="Use the results of blending together the\r\ninput colors of this lerp to create the\r\n\"MacroGrassColorBlend\" value"
   nodeColor=<rgba 1 0 0.9844 1>
  }
  n29 = ComponentMask(None=NamedRerouteUsage()) {g=true, r=true}
  n30 = NamedRerouteDeclaration(None=ScalarParameter() {defaultValue=600, parameterName=Landscape_BaseScale}) {name=LandscapeCoordinate_BaseScale, nodeColor=<rgba 1 0 0.4688 1>}
  n31 = MaterialFunctionCall(Float4=MaterialFunctionCall!(
    Texture=TextureObject() {samplerType=SAMPLERTYPE_Masks, texture=/Game/Environments/Apollo/Sets/Terrain/Textures/T_Apollo_Grass_D_02}
    LandscapeCoordinates=Divide(A=@n29, B=@n30)
    LowEndTiling
    HexGridGradientSharpness
    HexGridSize
    HexGridUVInput
    MinScale
    MaxScale
    MinOffset
    MaxOffset
    "MaxRotation(Degrees)"
    "MinRotation(Degrees)"
    forceLowQuality
   ) {materialFunction=/Game/Athena/Asteria/Environments/Landscape/Materials/MaterialFunctions/UtilityFunctions/Landscape_TextureSample}.Result) {materialFunction=&BreakOutFloat4Components}
  NamedRerouteDeclaration(None=FunctionOutput(None=@n31.A) {outputName=Mobile-Height, sortPriority=-1}) {name=HeightFromTexture, nodeColor=<rgba 1 0.9844 0 1>}
  NamedRerouteDeclaration(None=@n27.GrassMacroColorMask) {name=GrassMacroColorMask, nodeColor=<rgba 1 0 0.8203 1>}
  NamedRerouteDeclaration(None=FunctionOutput(None=@n27.GrassColorMask2) {outputName=GrassColorMask2, sortPriority=9}) {name=GrassColorMask2, nodeColor=<rgba 1 0 0.8203 1>}
  NamedRerouteDeclaration(None=FunctionInput(Preview=WorldPosition() {worldPositionShaderOffset=WPT_ExcludeAllShaderOffsets}.XYZ) {bUsePreviewValueAsDefault=true, inputName=TextureSamplingPosition, sortPriority=100}) {name=Sampling Position, nodeColor=<rgba 1 0.8906 0 1>}
  NamedRerouteDeclaration(None=@n28.G) {name=TempBionmeBlend, nodeColor=<rgba 0 1 0.7656 1>}
  n32 = VectorParameter() {group=Grass, parameterName=Sakura-GrassMacroColorA, defaultValue=<rgba 0.0308 0.1458 0.1224 1>}
  n33 = ScalarParameter() {defaultValue=2, parameterName=Sakura-BlendValue}
  n34 = Multiply(A=NamedRerouteUsage(), B=@n33) {constB=4}
  n35 = LinearInterpolate(
   A=LinearInterpolate(
    A=VectorParameter() {group=Grass, parameterName=Sakura-SSC-1, defaultValue=<rgba 0.0109 0.1146 0.1009 1>}.RGB
    B=VectorParameter() {group=Grass, parameterName=Sakura-SSC-2, defaultValue=<rgba 0.0196 0.0729 0.0693 1>}.RGB
    Alpha=Clamp(None=@n34)
   )
   B=VectorParameter() {group=Grass, parameterName=Sakura-SSC-3, defaultValue=<rgba 0.0112 0.151 0.1114 1>}.RGB
   Alpha=NamedRerouteUsage()
  )
  n36 = VectorParameter() {parameterName=Sakura-yellowish, defaultValue=<rgba 0.0289 0.125 0.0852 1>}
  n37 = Clamp(None=Multiply(A=NamedRerouteUsage(), B=@n33) {constB=4})
  n38 = NamedRerouteUsage()
  n39 = Multiply(A=@n38, B=Constant() {r=4})
  NamedRerouteDeclaration(None=LinearInterpolate(A=@n32.RGB, B=LinearInterpolate(A=@n35, B=@n36.RGB, Alpha=@n37), Alpha=Saturate(None=@n39))) {name=Sakura-MicroGrassColors, nodeColor=<rgba 1 0 0.9844 1>}
  n40 = ScalarParameter() {group=Grass, parameterName=Sakura-MacroGrassBlend_Min}
  n41 = ScalarParameter() {defaultValue=1, group=Grass, parameterName=Sakura-MacroGrassBlend_Max}
  NamedRerouteDeclaration(None=LinearInterpolate(
    A=@n32.RGB
    B=VectorParameter() {group=Grass, parameterName=Sakura-GrassMacroColorB, defaultValue=<rgba 0.0206 0.096 0.099 1>}.RGB
    Alpha=LinearInterpolate(A=@n40, B=@n41, Alpha=@n38)
   )) {name=Sakura-MacroGrassColorLerp, nodeColor=<rgba 1 0 0.9844 1>}
  NamedRerouteDeclaration(None=VectorParameter() {group=Grass, parameterName=Sakura-MacroGrassColor, defaultValue=<rgba 0.0308 0.1667 0.1433 1>}.RGB) {
   name=Sakura-MacroGrassColorBlend
   desc="Use the results of blending together the\r\ninput colors of this lerp to create the\r\n\"MacroGrassColorBlend\" value"
   nodeColor=<rgba 1 0 0.9844 1>
  }
  n42 = VectorParameter() {group=Grass, parameterName=GrassMacroColorB, defaultValue=<rgba 0.0784 0.349 0.0667 1>}
  n43 = LinearInterpolate(
   A=ScalarParameter() {group=Grass, parameterName=MacroGrassBlend_Min}
   B=ScalarParameter() {defaultValue=1, group=Grass, parameterName=MacroGrassBlend_Max}
   Alpha=@n23
  )
  NamedRerouteDeclaration(None=LinearInterpolate(A=LinearInterpolate(A=@n18.RGB, B=@n42.RGB, Alpha=@n43), B=NamedRerouteUsage(), Alpha=NamedRerouteUsage())) {name=MacroGrassColorLerp, nodeColor=<rgba 1 0 0.9844 1>}
  n44 = VectorParameter() {group=Grass, parameterName=Fall-GrassMacroColorA, defaultValue=<rgba 0.2784 0.1686 0.0078 1>}
  n45 = ScalarParameter() {defaultValue=2, parameterName=Fall-BlendValue}
  n46 = Multiply(A=NamedRerouteUsage(), B=@n45) {constB=4}
  n47 = LinearInterpolate(
   A=LinearInterpolate(
    A=VectorParameter() {group=Grass, parameterName=Fall-SSC-1, defaultValue=<rgba 0.2784 0.1686 0.0078 1>}.RGB
    B=VectorParameter() {group=Grass, parameterName=Fall-SSC-2, defaultValue=<rgba 0.0353 0.102 0.0039 1>}.RGB
    Alpha=Clamp(None=@n46)
   )
   B=VectorParameter() {group=Grass, parameterName=Fall-SSC-3, defaultValue=<rgba 0.302 0.1569 0.0118 1>}.RGB
   Alpha=NamedRerouteUsage()
  )
  n48 = VectorParameter() {group=Grass, parameterName=Fall-yellowish, defaultValue=<rgba 0.0353 0.102 0.0039 1>}
  n49 = Clamp(None=Multiply(A=NamedRerouteUsage(), B=@n45) {constB=4})
  n50 = NamedRerouteUsage()
  n51 = Multiply(A=@n50, B=Constant() {r=4})
  NamedRerouteDeclaration(None=LinearInterpolate(A=@n44.RGB, B=LinearInterpolate(A=@n47, B=@n48.RGB, Alpha=@n49), Alpha=Saturate(None=@n51))) {name=Fall-MicroGrassColors, nodeColor=<rgba 1 0 0.9844 1>}
  n52 = VectorParameter() {group=Grass, parameterName=BaseGrass_Macro_Color2, defaultValue=<rgba 0.0078 0.102 0.0392 1>}
  n53 = LinearInterpolate(A=@n52.RGB, B=NamedRerouteUsage(), Alpha=NamedRerouteUsage())
  n54 = NamedRerouteUsage()
  n55 = NamedRerouteUsage()
  NamedRerouteDeclaration(None=StaticSwitch(True=LinearInterpolate(A=@n53, B=@n54, Alpha=@n55), False=@n52.RGB, Value=@n26)) {
   name=MacroGrassColorBlend
   desc="Use the results of blending together the\r\ninput colors of this lerp to create the\r\n\"MacroGrassColorBlend\" value"
   nodeColor=<rgba 1 0 0.9844 1>
  }
  NamedRerouteDeclaration(None=FunctionOutput(None=@n31.R) {outputName=BasecolorDetail, sortPriority=-2}) {name=BaseColorDetail, nodeColor=<rgba 1 0.0703 0 1>}
  n56 = ScalarParameter() {group=Grass, parameterName=Fall-MacroGrassBlend_Min}
  n57 = ScalarParameter() {defaultValue=1, group=Grass, parameterName=Fall-MacroGrassBlend_Max}
  NamedRerouteDeclaration(None=LinearInterpolate(
    A=@n44.RGB
    B=VectorParameter() {group=Grass, parameterName=Fall-GrassMacroColorB, defaultValue=<rgba 0.0353 0.102 0.0039 1>}.RGB
    Alpha=LinearInterpolate(A=@n56, B=@n57, Alpha=@n50)
   )) {name=Fall-MacroGrassColorLerp, nodeColor=<rgba 1 0 0.9844 1>}
  NamedRerouteDeclaration(None=FunctionOutput(None=@n27.GrassColorMask1) {outputName=GrassColorMask1, sortPriority=8}) {name=GrassColorMask1, nodeColor=<rgba 1 0 0.8203 1>}
  ScalarParameter() {defaultValue=20000, parameterName=DistanceNearEffect}
  VectorParameter() {group=Grass, parameterName=FallPaletteTest, defaultValue=<rgba 0.502 0.1569 0 1>}
  VectorParameter() {group=Grass, parameterName=FallPaletteTest2, defaultValue=<rgba 0.0392 0.2 0.0471 1>}
  VectorParameter() {group=Grass, parameterName=FallPaletteTEst, defaultValue=<rgba 0.502 0.1569 0 1>}
  WorldPosition() {worldPositionShaderOffset=WPT_ExcludeAllShaderOffsets}
 ]}
