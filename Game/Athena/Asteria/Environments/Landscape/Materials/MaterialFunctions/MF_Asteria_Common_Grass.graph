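# Material function for the Asteria common grass landscape layer, supplying material attributes plus separate color, normal, spec, height and mask outputs for landscape, static and grass meshes.
# MF_GrassMicroColorMasks drives micro and macro color lerps, with Fall and Sakura palettes, while MF_AsteriaBiomeColors applies biome colors.
# Distance masking switches between RVT mip level and camera sphere masks.
# Mobile colors get dedicated outputs, and Epic quality selects alternate tangent normals.
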MaterialFunction /Game/Athena/Asteria/Environments/Landscape/Materials/MaterialFunctions/MF_Asteria_Common_Grass
graph {nodes [
  n1 = Multiply(A=Saturate(None=NamedRerouteUsage()), B=Constant() {r=0.9})
  n2 = MaterialFunctionCall!(Grass=@n1, Gravel, Mud, Sand, BiomeMask-RGBA=NamedRerouteUsage(), UseBiomeColorMask=NamedRerouteUsage(), UseLandscapeCoords) {materialFunction=&/MF_AsteriaBiomeColors}
  n3 = StaticSwitchParameter(True=@n2.Grass, False=Multiply(A=@n1, B=@n2.Grass)) {group=Grass, parameterName=UseSimpleGrassColors}
  FunctionOutput(None=@n3) {desc=For spawned grass meshes, description=For the landscape grass meshes, outputName=BasicGrassMeshColor, sortPriority=-6}
  n4 = SetMaterialAttributes!(MaterialAttributes, BaseColor=NamedRerouteUsage(), Specular=NamedRerouteUsage(), Roughness=NamedRerouteUsage(), Normal=NamedRerouteUsage()) {attributeSetTypes [
    #1
    #2
    #3
    #4
   ]}
  FunctionOutput(None=@n4) {outputName=Default-LandscapeMaterial, sortPriority=-100}
  n5 = TextureObject() {samplerType=SAMPLERTYPE_Normal, texture=/Game/Athena/Apollo/Environments/Landscape/Materials/Biome/CommonMaterials/T_CommonGrass_N}
  n6 = Constant() {r=3500}
  FunctionOutput(None=MaterialFunctionCall(TextureObject=@n5, TextureSize=@n6, WorldSpace=StaticBool() {value=true}) {desc=World, materialFunction=&WorldAlignedNormal}.Z Texture) {outputName=WorldNormals, sortPriority=4}
  n7 = Subtract(A=LinearInterpolate(A=Constant() {r=0.5}, B=NamedRerouteUsage(), Alpha=NamedRerouteUsage()), B=Constant() {r=0.5})
  n8 = Constant() {r=25}
  FunctionOutput(None=Multiply(A=Multiply(A=@n7, B=@n8), B=Constant() {r=1})) {outputName=Displacement, sortPriority=12}
  n9 = Divide(A=ComponentMask(None=WorldPosition() {worldPositionShaderOffset=WPT_ExcludeAllShaderOffsets}.XYZ) {g=true, r=true}, B=NamedRerouteUsage())
  n10 = StaticBool() {value=true}
  n11 = MaterialFunctionCall(Float4=MaterialFunctionCall!(
    Texture=TextureObject() {samplerType=SAMPLERTYPE_Masks, texture=/Game/Environments/Apollo/Sets/Terrain/Textures/T_Apollo_Grass_D_02}
    LandscapeCoordinates=@n9
    LowEndTiling
    HexGridGradientSharpness
    HexGridSize
    HexGridUVInput
    MinScale
    MaxScale
    MinOffset
    MaxOffset
    "MaxRotation(Degrees)"
    "MinRotation(Degrees)"
    forceLowQuality=@n10
   ) {materialFunction=/Game/Athena/Asteria/Environments/Landscape/Materials/MaterialFunctions/UtilityFunctions/Landscape_TextureSample}.Result) {materialFunction=&BreakOutFloat4Components}
  FunctionOutput(None=@n11.G) {
   desc=This is the un-modified spec straight from the grass texture
   description=This is the un-modified spec straight from the grass texture
   outputName=Spec-Basic
   sortPriority=2
  }
  n12 = Constant() {r=0.01}
  n13 = Power(Base=NamedRerouteUsage(), Exp=Constant() {r=1.5})
  n14 = Saturate(None=Add(A=Multiply(A=@n12, B=@n13), B=Multiply(A=@n13, B=@n3)))
  n15 = Multiply(A=Divide(A=@n14, B=NamedRerouteUsage()), B=Constant() {r=1.25})
  n16 = NamedRerouteUsage()
  n17 = WorldPosition() {worldPositionShaderOffset=WPT_ExcludeAllShaderOffsets}
  n18 = TextureSample(UVs=Divide(A=ComponentMask(None=@n17.XYZ) {g=true, r=true}, B=Constant() {r=7500})) {
   desc="This was T_Athena_Snow_01_M\r\npre-19.00"
   samplerSource=SSM_Wrap_WorldGroupSettings
   samplerType=SAMPLERTYPE_LinearColor
   texture=/Game/Athena/Environments/Landscape/Textures/T_Athena_Snow_01_MOBILE_M
  }
  n19 = Constant() {r=1.5}
  n20 = MaterialFunctionCall(Transition Phase=Saturate(None=Multiply(A=@n18.A, B=@n19)), Height Texture=@n16) {materialFunction=&HeightLerp}
  n21 = CameraPositionWS()
  n22 = WorldPosition()
  n23 = Constant() {r=30000}
  n24 = Constant()
  n25 = Multiply(
   A=Multiply(A=Multiply(A=@n16, B=@n20.Lerp Alpha No Contrast), B=SphereMask(A=@n21, B=@n22.XYZ, Radius=@n23, Hardness=@n24))
   B=OneMinus(None=NamedRerouteUsage())
  )
  FunctionOutput(None=LinearInterpolate(A=@n14, B=Saturate(None=@n15), Alpha=@n25)) {outputName=Mobile-StaticMesh-Color, sortPriority=500}
  n26 = Constant() {r=0.01}
  n27 = NamedRerouteUsage()
  n28 = NamedRerouteUsage()
  n29 = NamedRerouteUsage()
  n30 = NamedRerouteUsage()
  n31 = Saturate(None=Multiply(
    A=@n28
    B=MaterialFunctionCall(Grass=@n28, BiomeMask-RGBA=@n29, UseBiomeColorMask=@n30) {materialFunction=&MF_AsteriaBiomeColors}.Grass
   )) {desc=new}
  n32 = Saturate(None=Add(A=Multiply(A=@n26, B=@n27), B=Multiply(A=@n27, B=@n31)))
  n33 = Divide(A=@n32, B=NamedRerouteUsage())
  n34 = NamedRerouteUsage()
  n35 = Constant() {r=3}
  FunctionOutput(None=LinearInterpolate(A=@n32, B=Saturate(None=@n33), Alpha=Power(Base=@n34, Exp=@n35))) {outputName=Default-StaticMesh-Color, sortPriority=500}
  FunctionOutput(None=NamedRerouteUsage()) {outputName=Default-LandscapeHeightmap, sortPriority=-99}
  FunctionOutput(None=@n31) {desc=For spawned grass meshes, description=For the landscape grass meshes, outputName=GrassMeshColor, sortPriority=-5}
  n36 = MaterialFunctionCall!(SamplingPosition) {materialFunction=&/MF_GrassMicroColorMasks}
  FunctionOutput(None=@n36.GrassColorMask3-NoMod) {outputName=GrassColorMask3-NoMod, sortPriority=10}
  n37 = VectorParameter() {group=Grass, parameterName=GrassMacroNoiseColor_01, sortPriority=-16, defaultValue=<rgba 0.0078 0.102 0.0392 1>}
  n38 = VectorParameter() {group=Grass, parameterName=GrassNoiseColor_01, sortPriority=-20, defaultValue=<rgba 0.0154 0.2 0.0769 1>}
  n39 = VectorParameter() {group=Grass, parameterName=GrassNoiseColor_02, sortPriority=-19, defaultValue=<rgba 0.0295 0.151 0.108 1>}
  n40 = LinearInterpolate(A=@n38.RGB, B=@n39.RGB, Alpha=NamedRerouteUsage())
  n41 = VectorParameter() {group=Grass, parameterName=GrassNoiseColor_03, sortPriority=-18, defaultValue=<rgba 0.0786 0.25 0.0729 1>}
  n42 = NamedRerouteUsage()
  n43 = VectorParameter() {group=Grass, parameterName=GrassNoiseColor_04, sortPriority=-17, defaultValue=<rgba 0.0136 0.1719 0.0495 1>}
  n44 = Constant() {r=4}
  n45 = NamedRerouteUsage()
  n46 = LinearInterpolate(
   A=@n37.RGB
   B=LinearInterpolate(A=LinearInterpolate(A=@n40, B=@n41.RGB, Alpha=@n42), B=@n43.RGB, Alpha=NamedRerouteUsage())
   Alpha=Saturate(None=Multiply(A=@n45, B=@n44))
  )
  n47 = LinearInterpolate(A=LinearInterpolate(A=@n46, B=NamedRerouteUsage(), Alpha=NamedRerouteUsage()), B=NamedRerouteUsage(), Alpha=NamedRerouteUsage())
  n48 = NamedRerouteUsage()
  FunctionOutput(None=NamedRerouteDeclaration(None=StaticSwitch(True=@n47, False=@n46, Value=@n48)) {name=MicroGrassColors, nodeColor=<rgba 1 0 0.9844 1>}) {outputName=MicroGrassColors}
  n49 = ViewProperty() {property=MEVP_RuntimeVirtualTextureOutputLevel}
  n50 = ComponentMask(None=@n49.InvProperty) {r=true}
  n51 = Constant() {r=2.5}
  n52 = LinearInterpolate(A=NamedRerouteUsage(), B=NamedRerouteUsage(), Alpha=Saturate(None=Multiply(A=@n50, B=@n51)))
  n53 = LinearInterpolate(
   A=NamedRerouteUsage()
   B=NamedRerouteUsage()
   Alpha=SphereMask(A=WorldPosition().XYZ, B=CameraPositionWS(), Radius=Constant() {r=20000}, Hardness=Constant())
  )
  FunctionOutput(None=NamedRerouteDeclaration(None=StaticSwitchParameter(True=@n52, False=@n53) {defaultValue=true, desc=Leave this false for UEFN, group=Grass, parameterName=UseRVT, sortPriority=1000}) {name=DistanceBasedMasking, nodeColor=<rgba 1 0 0.8203 1>}) {outputName=DistanceMasking}
  n54 = NamedRerouteUsage()
  n55 = MaterialFunctionCall(TextureObject=@n5, TextureSize=@n6) {desc=Tangent, materialFunction=&WorldAlignedNormal}
  FunctionOutput(None=NamedRerouteDeclaration(None=QualitySwitch(Default=@n55.Z Texture, Low=@n55.Z Texture, High=@n55.Z Texture, Medium=@n55.Z Texture, Epic=@n54)) {name=Output-Normals, nodeColor=<rgba 0 0.4141 1 1>}) {outputName=TangentNormals, sortPriority=4}
  n56 = Constant() {r=0.01}
  n57 = Power(Base=NamedRerouteUsage(), Exp=Constant() {r=1.5})
  n58 = Desaturation(None=@n3, Fraction=Constant() {r=0.5})
  n59 = Constant3Vector() {desc=This value was picked just to get a good result across all biomes, constant=<rgba 0.625 0.8354 2.5 1>}
  n60 = LinearInterpolate(A=@n3, B=Saturate(None=Divide(A=@n58, B=@n59)), Alpha=NamedRerouteUsage())
  n61 = Saturate(None=Add(A=Multiply(A=@n56, B=@n57), B=Multiply(A=@n57, B=@n60)))
  n62 = Saturate(None=Add(A=@n61, B=NamedRerouteUsage()))
  n63 = Power(Base=NamedRerouteUsage(), Exp=Constant() {r=3})
  FunctionOutput(None=LinearInterpolate(A=@n61, B=LinearInterpolate(A=@n61, B=@n62, Alpha=@n63), Alpha=@n25)) {outputName=Mobile-Landscape-Color, sortPriority=-4}
  n64 = NamedRerouteUsage()
  n65 = LinearInterpolate(A=Constant() {r=0.05}, B=Constant() {r=1.5}, Alpha=Multiply(A=OneMinus(None=@n64), B=@n11.G))
  n66 = Clamp(None=@n65, Max=Constant() {r=0.5}) {clampMode=CMODE_ClampMax}
  n67 = SphereMask(A=CameraPositionWS(), B=WorldPosition().XYZ, Radius=Constant() {r=2500}, Hardness=Constant())
  FunctionOutput(None=Multiply(A=@n66, B=OneMinus(None=@n67))) {
   desc="This is being used because the spec from the RVT has strange mipping\r\nissues when viewing glancing angles of the grass mesh. This should\r\nmatch the RVT_MipDistanceMidEffect distance as close as possible.\r\nThis is also used on the static mesh terrain."
   outputName=Spec-CameraDistance
   sortPriority=1
  }
  n68 = FunctionOutput(None=@n11.A) {outputName=Mobile-Height, sortPriority=-1}
  FunctionOutput(None=NamedRerouteDeclaration(None=Multiply(A=@n68, B=@n65) {constA=1}) {name=Output-Spec, nodeColor=<rgba 1 0.9844 0 1>}) {outputName=Default-Spec-RVTMipDistance}
  FunctionOutput(None=@n11.B) {outputName=Rough-Basic, sortPriority=3}
  n69 = ComponentMask(None=@n49.Property) {r=true}
  n70 = Constant() {r=18}
  LinearInterpolate(A=@n52, B=NamedRerouteUsage(), Alpha=Saturate(None=Divide(A=@n69, B=@n70)))
  n71 = CameraPositionWS()
  n72 = WorldPosition()
  n73 = Constant() {r=100000}
  n74 = Constant()
  LinearInterpolate(A=@n53, B=NamedRerouteUsage(), Alpha=OneMinus(None=SphereMask(A=@n71, B=@n72.XYZ, Radius=@n73, Hardness=@n74)))
  n75 = VectorParameter() {parameterName=TestNewColors, defaultValue=<rgba 0.1157 0.15 0.0332 0>}
  n76 = Saturate(None=@n75.RGB)
  LinearInterpolate(A=Saturate(None=Multiply(A=@n76, B=@n28)), B=@n31, Alpha=@n75.A)
  Multiply(A=Max(A=Constant() {r=0.4}, B=NamedRerouteUsage()), B=@n66)
  n77 = Multiply(A=LandscapeLayerCoords() {mappingScale=1}, B=Constant() {r=0.6})
  n78 = NamedRerouteDeclaration(None=TextureSample(UVs=NamedRerouteDeclaration(None=@n77) {name=Epic_GrassACoords, nodeColor=<rgba 1 0.8672 0 1>}) {samplerSource=SSM_Wrap_WorldGroupSettings, texture=/Game/Environments/Apollo/Sets/Terrain/Textures/T_Apollo_GrassHeight_D}.A) {name=Epic_Noise1, nodeColor=<rgba 1 0.8672 0 1>}
  Multiply(A=ScalarParameter() {defaultValue=1, group=Grass, parameterName=EPIC_GrassA-HeightMod}, B=@n78)
  NamedRerouteDeclaration(None=FunctionOutput(None=@n11.R) {outputName=BasecolorDetail, sortPriority=-2}) {name=BaseColorDetail, nodeColor=<rgba 1 0.0703 0 1>}
  NamedRerouteDeclaration(None=FunctionOutput(None=@n36.GrassColorMask3) {outputName=GrassColorMask3, sortPriority=10}) {name=GrassColorMask3, nodeColor=<rgba 1 0 0.8203 1>}
  NamedRerouteDeclaration(None=FunctionOutput(None=@n36.GrassColorMask4) {outputName=GrassColorMask4, sortPriority=11}) {name=GrassColorMask4, nodeColor=<rgba 1 0 0.8203 1>}
  n79 = NamedRerouteUsage()
  n80 = TextureSample!(UVs=Multiply(A=LandscapeLayerCoords() {mappingScale=1}, B=Constant() {r=0.08}), Tex, Bias=Constant() {r=2}, Apply View MipBias) {
   mipValueMode=TMVM_MipBias
   samplerSource=SSM_Wrap_WorldGroupSettings
   samplerType=SAMPLERTYPE_Normal
   texture=/Game/Environments/Apollo/Sets/Terrain/Textures/T_Apollo_DirtHeight_N
  }
  n81 = LinearInterpolate(
   A=@n79
   B=MaterialFunctionCall(BaseNormal=@n79, AdditionalNormal=@n80.RGB) {materialFunction=&BlendAngleCorrectedNormals}.Result
   Alpha=Power(Base=NamedRerouteUsage(), Exp=Constant() {r=1.5})
  )
  n82 = TextureSample(UVs=NamedRerouteUsage(), Bias=Constant() {r=3}) {
   mipValueMode=TMVM_MipBias
   samplerSource=SSM_Wrap_WorldGroupSettings
   samplerType=SAMPLERTYPE_Normal
   texture=/Game/Environments/Apollo/Sets/Terrain/Textures/T_Apollo_GrassHeight_N
  }
  n83 = MaterialFunctionCall(BaseNormal=@n82.RGB, AdditionalNormal=@n81) {materialFunction=&BlendAngleCorrectedNormals}
  n84 = NamedRerouteUsage()
  NamedRerouteDeclaration(None=MaterialFunctionCall(BaseNormal=LinearInterpolate(A=@n81, B=@n83.Result, Alpha=@n84), AdditionalNormal=NamedRerouteUsage()) {materialFunction=&BlendAngleCorrectedNormals}.Result) {name=Epic-Normals, nodeColor=<rgba 1 0.8672 0 1>}
  NamedRerouteDeclaration(None=@n68) {name=HeightFromTexture, nodeColor=<rgba 1 0.9844 0 1>}
  n85 = Constant() {r=3}
  NamedRerouteDeclaration(None=Saturate(None=Divide(A=@n69, B=@n85))) {name=RVT_MipDistanceMid, nodeColor=<rgba 1 0 0.8203 1>}
  NamedRerouteDeclaration(None=ScalarParameter() {defaultValue=600, group=Grass, parameterName=LandscapeGrass_BaseScale, sortPriority=-1000}) {name=LandscapeCoordinate_BaseScale, nodeColor=<rgba 1 0 0.4687 1>}
  n86 = MaterialFunctionCall!(
   Material=SetMaterialAttributes!(MaterialAttributes)
   Heightmap
   NewSlopeMaskMask
   Displacement
   AsphaltLayer
   UseAlternateCliffMat
   AltCliffHeight
   AltCliffMat
  ) {materialFunction=/Game/Athena/Apollo/Environments/Landscape/Materials/Biome/CommonMaterials/MF_SlopeBlendCliff}
  n87 = Power(Base=@n86.HeightmapForGrass, Exp=Constant() {r=3})
  n88 = Constant() {r=6}
  n89 = OneMinus(None=Saturate(None=Multiply(A=@n87, B=@n88)))
  n90 = NamedRerouteUsage()
  n91 = Constant() {r=0.75}
  n92 = Constant()
  n93 = NamedRerouteUsage()
  NamedRerouteDeclaration(None=LinearInterpolate(A=Multiply(A=@n89, B=@n90), B=LinearInterpolate(A=@n91, B=@n92, Alpha=@n93), Alpha=Constant() {r=0.5}) {desc=This makes the transition softer}) {name=ModifiedHeightmap, nodeColor=<rgba 1 0 0.8203 1>}
  NamedRerouteDeclaration(None=@n55.Z Texture) {name=BaseTangentNormals, nodeColor=<rgba 0 0.6016 1 1>}
  n94 = VectorParameter() {group=Grass, parameterName=Grass_TipsColor, sortPriority=-10, defaultValue=<rgba 0.2333 0.4427 0 1>}
  n95 = Constant() {r=0.01}
  n96 = NamedRerouteUsage()
  n97 = LinearInterpolate(
   A=@n31
   B=MaterialFunctionCall(Grass=@n31, BiomeMask-RGBA=NamedRerouteUsage(), UseBiomeColorMask=NamedRerouteUsage()) {materialFunction=&MF_AsteriaBiomeColors}.GrassEdge
   Alpha=NamedRerouteDeclaration(None=FunctionInput() {bUsePreviewValueAsDefault=true, inputName=CombinedLandscapeLayersMask, inputType=FunctionInput_Scalar}) {name=CombinedLS-LayersMask, nodeColor=<rgba 1 0 0.9844 1>}
  )
  n98 = VectorParameter() {group=Grass, parameterName=Grass_Localised_Tint, sortPriority=-10, defaultValue=<rgba 1 1 1 1>}
  n99 = FunctionInput() {bUsePreviewValueAsDefault=true, inputName=Grass Localised Tint Mask, inputType=FunctionInput_Scalar, previewValue=<1 0 0 1>}
  n100 = StaticSwitchParameter(True=LinearInterpolate(A=@n97, B=Multiply(A=@n97, B=@n98.RGB), Alpha=@n99), False=@n97) {group=Grass, parameterName=UseGrassLocalisedTint}
  n101 = Saturate(None=Add(A=Multiply(A=@n95, B=@n96), B=Multiply(A=@n96, B=@n100)))
  n102 = Saturate(None=Add(A=@n101, B=NamedRerouteDeclaration(None=@n94.RGB) {name=GrassTipColor, nodeColor=<rgba 1 0 0.8203 1>}))
  n103 = Power(Base=NamedRerouteUsage(), Exp=Constant() {r=3})
  NamedRerouteDeclaration(None=FunctionOutput(None=LinearInterpolate(A=@n101, B=@n102, Alpha=@n103)) {outputName=Default-Landscape-Color, sortPriority=-3}) {name=Output-Basecolor, desc=I think this is only used on landscape materials, nodeColor=<rgba 0 0.8359 1 1>}
  n104 = FunctionInput(Preview=StaticBool() {value=true}) {
   bUsePreviewValueAsDefault=true
   description="Default=True"
   inputName=BiomeColors-UseDefault
   inputType=FunctionInput_StaticBool
   sortPriority=10
   previewValue=<64 64 64 6>
  }
  NamedRerouteDeclaration(None=@n104) {name=UseBiomeColors, nodeColor=<rgba 1 0 0.9844 1>}
  NamedRerouteDeclaration(None=@n36.GrassMacroColorMask) {name=GrassMacroColorMask, nodeColor=<rgba 1 0 0.8203 1>}
  NamedRerouteDeclaration(None=FunctionOutput(None=@n36.GrassColorMask2) {outputName=GrassColorMask2, sortPriority=9}) {name=GrassColorMask2, nodeColor=<rgba 1 0 0.8203 1>}
  n105 = VectorParameter() {group=Grass, parameterName=Fall-GrassMacroColorA, defaultValue=<rgba 0.2784 0.1686 0.0078 1>}
  n106 = ScalarParameter() {defaultValue=2, group=Grass, parameterName=Fall-BlendValue}
  n107 = Multiply(A=NamedRerouteUsage(), B=@n106) {constB=4}
  n108 = LinearInterpolate(
   A=LinearInterpolate(
    A=VectorParameter() {group=Grass, parameterName=Fall-GrassNoiseColor_01, defaultValue=<rgba 0.2784 0.1686 0.0078 1>}.RGB
    B=VectorParameter() {group=Grass, parameterName=Fall-GrassNoiseColor_02, defaultValue=<rgba 0.0353 0.102 0.0039 1>}.RGB
    Alpha=Clamp(None=@n107)
   )
   B=VectorParameter() {group=Grass, parameterName=Fall-GrassNoiseColor_03, defaultValue=<rgba 0.302 0.1569 0.0118 1>}.RGB
   Alpha=NamedRerouteUsage()
  )
  n109 = VectorParameter() {group=Grass, parameterName=Fall-GrassNoiseColor_04, defaultValue=<rgba 0.0353 0.102 0.0039 1>}
  n110 = Clamp(None=Multiply(A=NamedRerouteUsage(), B=@n106) {constB=4})
  n111 = NamedRerouteUsage()
  n112 = Multiply(A=@n111, B=Constant() {r=4})
  NamedRerouteDeclaration(None=LinearInterpolate(A=@n105.RGB, B=LinearInterpolate(A=@n108, B=@n109.RGB, Alpha=@n110), Alpha=Saturate(None=@n112))) {name=Fall-MicroGrassColors, nodeColor=<rgba 1 0 0.9844 1>}
  n113 = ScalarParameter() {group=Grass, parameterName=Fall-MacroGrassBlend_Min}
  n114 = ScalarParameter() {defaultValue=1, group=Grass, parameterName=Fall-MacroGrassBlend_Max}
  NamedRerouteDeclaration(None=LinearInterpolate(
    A=@n105.RGB
    B=VectorParameter() {group=Grass, parameterName=Fall-GrassMacroColorB, defaultValue=<rgba 0.0353 0.102 0.0039 1>}.RGB
    Alpha=LinearInterpolate(A=@n113, B=@n114, Alpha=@n111)
   )) {name=Fall-MacroGrassColorLerp, nodeColor=<rgba 1 0 0.9844 1>}
  n115 = VectorParameter() {desc=This is disabled in UEFN, group=Grass, parameterName=Fall-MacroGrassColor, sortPriority=900, defaultValue=<rgba 0.2784 0.1686 0.0078 1>}
  NamedRerouteDeclaration(None=@n115.RGB) {
   name=Fall-MacroGrassColorBlend
   desc="Use the results of blending together the\r\ninput colors of this lerp to create the\r\n\"MacroGrassColorBlend\" value"
   nodeColor=<rgba 1 0 0.9844 1>
  }
  NamedRerouteDeclaration(None=Constant()) {name=BiomeGrassColors, nodeColor=<rgba 0.3828 0 1 1>}
  n116 = NamedRerouteDeclaration(None=FunctionInput(Preview=Constant()) {bUsePreviewValueAsDefault=true, inputName=BiomeMask-RGBA, inputType=FunctionInput_Vector4}) {name=BiomeMask-RGBA, nodeColor=<rgba 0 0.7472 1 1>}
  NamedRerouteDeclaration(None=ChannelMaskParameter(None=@n116) {desc=This is disabled in UEFN, group=Grass, parameterName=Tundra, sortPriority=800}.R) {name=Tundra, nodeColor=<rgba 0.3828 0 1 1>}
  n117 = VectorParameter() {group=Grass, parameterName=GrassMacroNoiseColor_02, sortPriority=-15, defaultValue=<rgba 0.0784 0.349 0.0667 1>}
  n118 = LinearInterpolate(
   A=@n37.RGB
   B=@n117.RGB
   Alpha=LinearInterpolate(
    A=ScalarParameter() {group=Grass, parameterName=MacroGrassBlend_Min, sortPriority=-28}
    B=ScalarParameter() {defaultValue=1, group=Grass, parameterName=MacroGrassBlend_Max, sortPriority=-27}
    Alpha=@n45
   )
  )
  n119 = LinearInterpolate(A=@n118, B=NamedRerouteUsage(), Alpha=NamedRerouteUsage())
  n120 = NamedRerouteUsage()
  n121 = NamedRerouteUsage()
  NamedRerouteDeclaration(None=StaticSwitch(True=LinearInterpolate(A=@n119, B=@n120, Alpha=@n121), False=@n118, Value=NamedRerouteUsage())) {name=MacroGrassColorLerp, nodeColor=<rgba 1 0 0.9844 1>}
  n122 = ChannelMaskParameter(None=@n116) {desc=This is disabled in UEFN, group=Grass, maskChannel=Green, parameterName=Fall, sortPriority=800, defaultValue=<rgba 0 1 0 0>}
  NamedRerouteDeclaration(None=NamedRerouteDeclaration(None=@n122.R) {name=Fall, nodeColor=<rgba 0.3828 0 1 1>}) {name=TempBionmeBlend, nodeColor=<rgba 0 1 0.7656 1>}
  n123 = TextureObject() {samplerType=SAMPLERTYPE_Normal, texture=/Game/Environments/Apollo/Sets/Terrain/Textures/T_Apollo_Dirt_N_02b}
  NamedRerouteDeclaration(None=ComponentMask(None=MaterialFunctionCall!(
     Texture=@n123
     UseMipBias
     MipBias
     LowEndTiling
     LandscapeCoordinates=@n9
     HexGridSize
     HexGridUVInput
     HexGridGradientSharpness
     MaxScale
     MinScale
     MaxOffset
     MinOffset
     "MaxRotation(Degrees)"
     "MinRotation(Degrees)"
     forceLowQuality=@n10
    ) {materialFunction=/Game/Athena/Asteria/Environments/Landscape/Materials/MaterialFunctions/UtilityFunctions/Landscape_NormalTextureSample}.Result) {b=true, g=true, r=true}) {name=BladeNormals, nodeColor=<rgba 1 0.5156 0 1>}
  n124 = ChannelMaskParameter(None=@n116) {desc=This is disabled in UEFN, group=Grass, maskChannel=Blue, parameterName=Desert, sortPriority=800, defaultValue=<rgba 0 0 1 0>}
  NamedRerouteDeclaration(None=NamedRerouteDeclaration(None=@n124.R) {name=Desert, nodeColor=<rgba 0.3828 0 1 1>}) {name=BlueBiomeBlend, nodeColor=<rgba 0 1 0.2031 1>}
  n125 = VectorParameter() {group=Grass, parameterName=Sakura-GrassMacroColorA, defaultValue=<rgba 0.0308 0.1458 0.1224 1>}
  n126 = ScalarParameter() {defaultValue=2, group=Grass, parameterName=Sakura-BlendValue}
  n127 = Multiply(A=NamedRerouteUsage(), B=@n126) {constB=4}
  n128 = LinearInterpolate(
   A=LinearInterpolate(
    A=VectorParameter() {group=Grass, parameterName=Sakura-GrassNoiseColor_01, defaultValue=<rgba 0.0109 0.1146 0.1009 1>}.RGB
    B=VectorParameter() {group=Grass, parameterName=Sakura-GrassNoiseColor_02, defaultValue=<rgba 0.0196 0.0729 0.0693 1>}.RGB
    Alpha=Clamp(None=@n127)
   )
   B=VectorParameter() {group=Grass, parameterName=Sakura-GrassNoiseColor_03, defaultValue=<rgba 0 0.151 0.134 1>}.RGB
   Alpha=NamedRerouteUsage()
  )
  n129 = VectorParameter() {group=Grass, parameterName=Sakura-GrassNoiseColor_04, defaultValue=<rgba 0.0239 0.125 0.088 1>}
  n130 = Clamp(None=Multiply(A=NamedRerouteUsage(), B=@n126) {constB=4})
  n131 = NamedRerouteUsage()
  n132 = Multiply(A=@n131, B=Constant() {r=4})
  NamedRerouteDeclaration(None=LinearInterpolate(A=@n125.RGB, B=LinearInterpolate(A=@n128, B=@n129.RGB, Alpha=@n130), Alpha=Saturate(None=@n132))) {name=Sakura-MicroGrassColors, nodeColor=<rgba 1 0 0.9844 1>}
  n133 = ScalarParameter() {group=Grass, parameterName=Sakura-MacroGrassBlend_Min}
  n134 = ScalarParameter() {defaultValue=1, group=Grass, parameterName=Sakura-MacroGrassBlend_Max}
  NamedRerouteDeclaration(None=LinearInterpolate(
    A=@n125.RGB
    B=VectorParameter() {group=Grass, parameterName=Sakura-GrassMacroColorB, defaultValue=<rgba 0.0082 0.0801 0.099 1>}.RGB
    Alpha=LinearInterpolate(A=@n133, B=@n134, Alpha=@n131)
   )) {name=Sakura-MacroGrassColorLerp, nodeColor=<rgba 1 0 0.9844 1>}
  n135 = VectorParameter() {
   desc=This is disabled in UEFN
   group=Grass
   parameterName=Sakura-MacroGrassColor
   sortPriority=899
   defaultValue=<rgba 0.0308 0.1667 0.1433 1>
  }
  NamedRerouteDeclaration(None=@n135.RGB) {
   name=Sakura-MacroGrassColorBlend
   desc="Use the results of blending together the\r\ninput colors of this lerp to create the\r\n\"MacroGrassColorBlend\" value"
   nodeColor=<rgba 1 0 0.9844 1>
  }
  n136 = TextureSample!(
   UVs=MaterialFunctionCall!(Scale=Constant() {r=3000}) {materialFunction=/Engine/Functions/Engine_MaterialFunctions01/Coordinates/WorldPosition-XY}.Result
   Tex
   Apply View MipBias
  ) {
   samplerSource=SSM_Wrap_WorldGroupSettings
   samplerType=SAMPLERTYPE_LinearColor
   texture=/Game/Characters/Player/Male/Medium/Bodies/M_MED_SRDrift/Textures/T_ElectricNoise
  }
  n137 = Add(
   A=Power(Base=@n136.R, Exp=Constant() {r=1.5})
   B=Multiply(A=NamedRerouteUsage(), B=Constant() {r=0.5}) {desc="This may just need to be the gravel && mud layer mask"}
  )
  NamedRerouteDeclaration(None=MaterialFunctionCall(Transition Phase=Saturate(None=@n137), Height Texture=Saturate(None=@n78)) {materialFunction=&HeightLerp}.Lerp Alpha No Contrast) {name=Epic_NoiseBlend, nodeColor=<rgba 1 0.8672 0 1>}
  n138 = VectorParameter() {
   desc=This also serves as the grass color for mobile.
   group=Grass
   parameterName=MacroGrassColor
   sortPriority=-14
   defaultValue=<rgba 0.0078 0.102 0.0392 1>
  }
  n139 = LinearInterpolate(A=@n138.RGB, B=NamedRerouteUsage(), Alpha=NamedRerouteUsage())
  n140 = NamedRerouteUsage()
  n141 = NamedRerouteUsage()
  NamedRerouteDeclaration(None=StaticSwitch(True=LinearInterpolate(A=@n139, B=@n140, Alpha=@n141), False=@n138.RGB, Value=NamedRerouteUsage())) {
   name=MacroGrassColorBlend
   desc="Use the results of blending together the\r\ninput colors of this lerp to create the\r\n\"MacroGrassColorBlend\" value"
   nodeColor=<rgba 1 0 0.9844 1>
  }
  n142 = LinearInterpolate(Alpha=@n68) {constA=0.5, constB=-0.2}
  NamedRerouteDeclaration(None=LinearInterpolate(Alpha=Saturate(None=@n142)) {constA=0.5}) {name=Output-Rough, nodeColor=<rgba 0 1 0.6719 1>}
  NamedRerouteDeclaration(None=FunctionOutput(None=@n36.GrassColorMask1) {outputName=GrassColorMask1, sortPriority=8}) {name=GrassColorMask1, nodeColor=<rgba 1 0 0.8203 1>}
  n143 = NamedRerouteUsage()
  n144 = Constant() {r=0.2}
  n145 = Saturate(None=Multiply(A=Subtract(A=@n143, B=@n144), B=Constant() {r=2}) {constB=2})
  n146 = NamedRerouteUsage()
  Reroute(None=LinearInterpolate(A=@n146, B=Max(A=@n145, B=@n146), Alpha=Constant()))
  ScalarParameter() {defaultValue=0.6, group=Grass, parameterName=EPIC_GrassA-Scale}
  ScalarParameter() {defaultValue=3000, group=Grass, parameterName=EPIC_GrassNoise-Scale}
  ScalarParameter() {defaultValue=1.5, group=Grass, parameterName=EPIC_GrassNoise-Pow}
  ScalarParameter() {defaultValue=3, group=Grass, parameterName=EPIC_GrassA-Normal-MipBias}
  ScalarParameter() {defaultValue=0.5, group=Grass, parameterName=EPIC_HD_GrassGravelLayerCont}
  n147 = Constant() {r=2}
  n148 = Saturate(None=Multiply(A=Power(Base=@n86.HeightmapForGrass, Exp=@n147), B=Constant() {r=12}))
  StaticSwitch(True=Multiply(A=NamedRerouteUsage(), B=OneMinus(None=@n148)), False=NamedRerouteUsage(), Value=StaticBool()) {desc="This looks better, but exposes more rock"}
 ]}
