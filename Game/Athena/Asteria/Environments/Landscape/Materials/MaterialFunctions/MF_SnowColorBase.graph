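# Material function that supplies the base snow albedo color for Asteria landscape materials, named as a pre-RVT value.
# Three color parameters feed the main SnowAlbedo output through a shading-path switch, with Mobile using its own snow color.
# On other paths, a quality switch uses the darker Medium color for Low and Medium and the lighter color otherwise.
# Separate outputs also expose the Medium/Low and Mobile colors directly.
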MaterialFunction /Game/Athena/Asteria/Environments/Landscape/Materials/MaterialFunctions/MF_SnowColorBase
graph {nodes [
  n1 = VectorParameter() {parameterName=PreRVTSnowAlbedo, defaultValue=<rgba 0.5569 0.5922 0.651 1>}
  n2 = VectorParameter() {parameterName=PreRVTSnowAlbedo-Medium, defaultValue=<rgba 0.2431 0.2549 0.302 1>}
  n3 = VectorParameter() {parameterName=Mobile-SnowColor, defaultValue=<rgba 0.251 0.2706 0.302 1>}
  n4 = QualitySwitch(Default=@n3.RGB, Low=@n3.RGB, Medium=@n3.RGB)
  FunctionOutput(None=ShadingPathSwitch(Default=QualitySwitch(Default=@n1.RGB, Low=@n2.RGB, Medium=@n2.RGB, Epic=@n1.RGB), Mobile=@n4)) {outputName=SnowAlbedo}
  FunctionOutput(None=@n2.RGB) {outputName=SnowAlbedo_MediumLow, sortPriority=4}
  FunctionOutput(None=@n4) {outputName=SnowAlbedo_Mobile, sortPriority=4}
 ]}
