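# MF_SnowBuildup is a landscape material function that adds snow to a diffuse color within the Tundra biome mask.
# A SphereMask on the Height input, blended with its inverse by FlipHeightMap, forms the snow mask.
# That mask blends Diffuse toward MF_SnowColorBase albedo, itself mixed with Diffuse by SnowColorBlend.
# SnowCoverageMask outputs 1 outside Tundra and the height mask inside; the G_Snow_Color parameter is unconnected.
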MaterialFunction /Game/Athena/Asteria/Environments/Landscape/Materials/MaterialFunctions/MF_SnowBuildup
graph {nodes [
  n1 = FunctionInput(Preview=Constant3Vector()) {bUsePreviewValueAsDefault=true, inputName=Diffuse}
  n2 = LinearInterpolate(
   A=MaterialFunctionCall() {materialFunction=&/MF_SnowColorBase}.SnowAlbedo
   B=@n1
   Alpha=FunctionInput(Preview=ScalarParameter() {defaultValue=0.25, parameterName=G_SnowColorBlend}) {bUsePreviewValueAsDefault=true, inputName=SnowColorBlend, inputType=FunctionInput_Scalar}
  )
  n3 = SphereMask(
   A=FunctionInput(Preview=Constant()) {bUsePreviewValueAsDefault=true, inputName=Height, inputType=FunctionInput_Scalar}
   B=Constant()
   Radius=FunctionInput(Preview=ScalarParameter() {defaultValue=1, parameterName=HeightMapSampleRadius}) {bUsePreviewValueAsDefault=true, inputName=HeightSampleRadius, inputType=FunctionInput_Scalar}
   Hardness=FunctionInput(Preview=ScalarParameter() {defaultValue=0.5, parameterName=HeightMapSampleHardness}) {bUsePreviewValueAsDefault=true, inputName=HeightSampleHardness, inputType=FunctionInput_Scalar}
  )
  n4 = LinearInterpolate(
   A=OneMinus(None=@n3)
   B=@n3
   Alpha=FunctionInput(Preview=ScalarParameter() {parameterName=FlipHeightMap}) {bUsePreviewValueAsDefault=true, inputName=FlipHeightMap, inputType=FunctionInput_Scalar}
  )
  n5 = FunctionInput() {bUsePreviewValueAsDefault=true, inputName=BiomeMask-Tundra, inputType=FunctionInput_Scalar}
  FunctionOutput(None=LinearInterpolate(A=@n1, B=LinearInterpolate(A=@n1, B=@n2, Alpha=@n4), Alpha=@n5)) {outputName=Diffuse}
  FunctionOutput(None=LinearInterpolate(B=@n4, Alpha=@n5) {constA=1}) {outputName=SnowCoverageMask, sortPriority=1}
  VectorParameter() {parameterName=G_Snow_Color, defaultValue=<rgba 0.6437 0.6862 0.75 1>}
 ]}
