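# Gravel layer material function for the Asteria landscape, outputting material attributes plus height, displacement and individual channels.
# Dirt and gravel textures blend by layer weight and variance mask, then receive gravel color tints.
# Epic quality adds macro-masked height-lerped dirt detail and blended normals through quality switches; that height also drives displacement.
# MF_AsteriaBiomeColors adjusts base color per biome, and a static switch adds snow buildup masked by the red (tundra) biome channel.
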MaterialFunction /Game/Athena/Asteria/Environments/Landscape/Materials/MaterialFunctions/MF_Asteria_Gravel
graph {nodes [
  Constant() {r=1.5}
  Constant() {r=0.75}
  Constant() {r=0.07}
  Constant() {r=0.85}
  Constant() {r=0.1}
  Constant() {r=1.25}
  Constant() {r=0.5}
  Constant() {r=30}
  Constant() {r=1}
  Constant()
  Constant() {r=1}
  n1 = NamedRerouteUsage()
  n2 = NamedRerouteUsage()
  n3 = FunctionInput(Preview=Constant()) {bUsePreviewValueAsDefault=true, inputName=BiomeMask-RGBA, inputType=FunctionInput_Vector4}
  n4 = FunctionInput(Preview=StaticBool() {value=true}) {
   bUsePreviewValueAsDefault=true
   description="Default=True"
   inputName=BiomeColors-UseDefault
   inputType=FunctionInput_StaticBool
   sortPriority=10
   previewValue=<64 64 64 6>
  }
  n5 = NamedRerouteUsage()
  n6 = SetMaterialAttributes!(
   MaterialAttributes
   BaseColor=MaterialFunctionCall!(
    Grass
    Gravel=QualitySwitch(Default=@n2, Low=@n2, High=@n2, Medium=@n2, Epic=@n1)
    Mud
    Sand
    BiomeMask-RGBA=@n3
    UseBiomeColorMask=@n4
    UseLandscapeCoords
   ) {materialFunction=&/MF_AsteriaBiomeColors}.Gravel
   Specular=NamedRerouteUsage()
   Roughness=NamedRerouteUsage()
   Normal=QualitySwitch(Default=@n5, Low=@n5, High=@n5, Medium=@n5, Epic=NamedRerouteUsage())
  ) {attributeSetTypes [
    #1
    #2
    #3
    #4
   ]}
  FunctionOutput(None=@n6) {outputName=GravelOnly, sortPriority=10}
  n7 = Multiply(A=LandscapeLayerCoords() {mappingScale=4.5}, B=ScalarParameter() {defaultValue=1, group=Mud, parameterName=MudScale})
  n8 = TextureSampleParameter2D(UVs=@n7) {
   group=Gravel
   parameterName=T_Dirt_D_01
   samplerSource=SSM_Wrap_WorldGroupSettings
   texture=/Game/Environments/Apollo/Sets/Terrain/Textures/T_Apollo_Dirt_D_01
  }
  FunctionOutput(None=@n8.RGB) {outputName=Diffuse, sortPriority=1}
  n9 = TextureSampleParameter2D(UVs=@n7) {
   group=Gravel
   parameterName=T_Dirt_N_01
   samplerSource=SSM_Wrap_WorldGroupSettings
   samplerType=SAMPLERTYPE_Normal
   texture=/Game/Environments/Apollo/Sets/Terrain/Textures/T_Apollo_Dirt_N_01
  }
  n10 = Multiply(A=@n9.RGB, B=Constant3Vector() {constant=<rgba 1.5 1.5 1 1>})
  FunctionOutput(None=@n10) {outputName=Normals, sortPriority=1}
  n11 = NamedRerouteUsage()
  FunctionOutput(None=QualitySwitch(Default=@n11, Low=@n11, High=@n11, Medium=@n11, Epic=NamedRerouteUsage())) {outputName=Height-W/Variance, sortPriority=1}
  n12 = Multiply(A=LandscapeLayerCoords() {mappingScale=1}, B=ScalarParameter() {defaultValue=0.1, group=Gravel, parameterName=EPIC_GravelA-Scale})
  n13 = TextureSampleParameter2D(UVs=@n12) {
   group=Gravel
   parameterName=T_DirtHeight_D_01
   samplerSource=SSM_Wrap_WorldGroupSettings
   texture=/Game/Environments/Apollo/Sets/Terrain/Textures/T_Apollo_DirtHeight_D
  }
  n14 = Multiply(A=LandscapeLayerCoords() {mappingScale=1}, B=ScalarParameter() {defaultValue=0.07, group=Gravel, parameterName=EPIC_GravelMacro-Scale})
  n15 = TextureSample(UVs=@n14) {
   samplerSource=SSM_Wrap_WorldGroupSettings
   samplerType=SAMPLERTYPE_LinearColor
   texture=/Game/Environments/Apollo/Sets/Terrain/Textures/T_Apollo_DirtHeight_M
  }
  n16 = Power(Base=@n15.R, Exp=ScalarParameter() {defaultValue=1.5, group=Gravel, parameterName=EPIC_GravelMacroHeight-Pow})
  n17 = ScalarParameter() {defaultValue=0.75, group=Gravel, parameterName=EPIC_GravelMacroHeight-Mod}
  n18 = MaterialFunctionCall!(
   Diffuse 1=NamedRerouteUsage()
   Diffuse 2=Multiply(A=ScalarParameter() {defaultValue=0.85, group=Gravel, parameterName=EPIC_GravelA-DiffuseMod}, B=@n13.RGB)
   Transistion Phase=Saturate(None=Multiply(A=@n16, B=@n17) {desc="Can use an additional mask to affect the multiply amount for variation!"})
   Height Texture 1=NamedRerouteUsage()
   Height Texture 2=@n13.A
   Contrast
  ) {materialFunction=/Engine/Functions/Engine_MaterialFunctions02/Texturing/HeightLerpWithTwoHeightMaps}
  n19 = NamedRerouteDeclaration(None=@n18.Resulting Height Map) {name=Epic-Height-W/Variance_1, nodeColor=<rgba 0.5 0 1 1>}
  n20 = Multiply(
   A=Subtract(A=@n19, B=ScalarParameter() {defaultValue=0.5, group=Gravel, parameterName=HD_Gravel_Sub1})
   B=ScalarParameter() {defaultValue=30, group=Gravel, parameterName=HD_Gravel_Mod1}
  )
  n21 = ScalarParameter() {group=Gravel, parameterName=HD_Gravel_Add1}
  FunctionOutput(None=Multiply(A=Add(A=@n20, B=@n21), B=ScalarParameter() {defaultValue=1, group=Gravel, parameterName=HD_Gravel_Mod2})) {outputName=Displacement, sortPriority=2}
  n22 = GetMaterialAttributes(None=@n6) {attributeGetTypes [#1]}
  n23 = ComponentMask(None=@n3) {r=true}
  n24 = LinearInterpolate(
   A=@n22.BaseColor
   B=MaterialFunctionCall!(
    Height=NamedRerouteUsage()
    HeightSampleRadius
    HeightSampleHardness
    FlipHeightMap
    SnowColorBlend
    BiomeMask-Tundra=@n23
    Diffuse=@n22.BaseColor
   ) {materialFunction=&/MF_SnowBuildup}.Diffuse
   Alpha=@n23
  )
  FunctionOutput(None=StaticSwitchParameter(
    True=SetMaterialAttributes!(MaterialAttributes=@n22.MaterialAttributes, BaseColor=@n24) {attributeSetTypes [#1]}
    False=@n6
   ) {parameterName=UseSnowGravelBuildUp}) {outputName=Material}
  FunctionOutput(None=@n8.A) {outputName=Height, sortPriority=1}
  LinearInterpolate(A=NamedRerouteUsage(), B=@n19, Alpha=ScalarParameter() {defaultValue=1, group=Gravel, parameterName=HD_Gravel_DebugAlteredHeight})
  NamedRerouteDeclaration(None=ScalarParameter() {group=Gravel, parameterName=Gravel_Desaturation_Amt}) {name=Desaturation_Amt, nodeColor=<rgba 0 0.2031 1 1>}
  NamedRerouteDeclaration(None=VectorParameter() {group=Gravel, parameterName=Gravel_ColorVal, defaultValue=<rgba 1 1 1 1>}.RGB) {name=Gravel_RecolorValue, nodeColor=<rgba 1 0.4219 0 1>}
  n25 = TextureSampleParameter2D(UVs=@n12) {
   group=Gravel
   parameterName=T_DirtHeight_N_01
   samplerSource=SSM_Wrap_WorldGroupSettings
   samplerType=SAMPLERTYPE_Normal
   texture=/Game/Environments/Apollo/Sets/Terrain/Textures/T_Apollo_DirtHeight_N
  }
  n26 = LinearInterpolate(A=@n25.RGB, B=NamedRerouteUsage(), Alpha=@n18.Lerp Alpha No Contrast)
  n27 = TextureSample(UVs=@n14) {
   samplerSource=SSM_Wrap_WorldGroupSettings
   samplerType=SAMPLERTYPE_Normal
   texture=/Game/Global/Textures/Default/Tileable_Patterns/T_DirtMask_N
  }
  n28 = Subtract(A=Constant() {r=1}, B=@n18.Lerp Alpha No Contrast)
  n29 = Multiply(
   A=Multiply(A=Abs(None=@n28), B=@n18.Lerp Alpha No Contrast)
   B=ScalarParameter() {defaultValue=1.25, group=Gravel, parameterName=EPIC_GravelMacroMask-Mod}
  )
  NamedRerouteDeclaration(None=LinearInterpolate(
    A=@n26
    B=MaterialFunctionCall(BaseNormal=@n27.RGB, AdditionalNormal=@n26) {materialFunction=&BlendAngleCorrectedNormals}.Result
    Alpha=Saturate(None=@n29)
   )) {name=Epic-Normals, nodeColor=<rgba 0.5 0 1 1>}
  NamedRerouteDeclaration(None=@n18.Results) {name=Epic-Diffuse, nodeColor=<rgba 0.5 0 1 1>}
  n30 = TextureSampleParameter2D(UVs=@n7) {
   group=Gravel
   parameterName=T_DIrt_S_01
   samplerSource=SSM_Wrap_WorldGroupSettings
   samplerType=SAMPLERTYPE_Masks
   texture=/Game/Environments/Apollo/Sets/Terrain/Textures/T_Apollo_Dirt_S_01
  }
  NamedRerouteDeclaration(None=FunctionOutput(None=@n30.R) {outputName=Spec, sortPriority=1}) {name=Spec, nodeColor=<rgba 0.5 0 1 1>}
  NamedRerouteDeclaration(None=FunctionOutput(None=@n30.B) {outputName=Rough, sortPriority=1}) {name=Rough, nodeColor=<rgba 0.5 0 1 1>}
  n31 = Constant4Vector() {constant=<rgba 0.7 0.7 0.7 1>}
  n32 = Multiply(A=LandscapeLayerCoords() {mappingScale=4}, B=ScalarParameter() {defaultValue=1, group=Gravel, parameterName=GravelScale})
  n33 = TextureSampleParameter2D(UVs=@n32) {
   group=Gravel
   parameterName=T_Gravel_D_01
   samplerSource=SSM_Wrap_WorldGroupSettings
   texture=/Game/Environments/Apollo/Sets/Terrain/Textures/T_Apollo_Gravel_D_01
  }
  n34 = FunctionInput() {inputName=LayerWeight, inputType=FunctionInput_Scalar, sortPriority=-10}
  n35 = FunctionInput() {bUsePreviewValueAsDefault=true, inputName=VarianceMask, inputType=FunctionInput_Scalar, previewValue=<1 1 1 1>}
  n36 = Constant() {r=1.65}
  n37 = Multiply(A=@n34, B=Saturate(None=Multiply(A=@n35, B=@n36)))
  n38 = LinearInterpolate(A=@n8.RGBA, B=Multiply(A=Max(A=@n31.RGBA, B=@n33.A), B=@n33.RGBA), Alpha=@n37)
  NamedRerouteDeclaration(None=ComponentMask(None=@n38) {a=true}) {name=Height-W/Variance, nodeColor=<rgba 0.5 0 1 1>}
  n39 = TextureSampleParameter2D(UVs=@n32) {
   group=Gravel
   parameterName=T_Gravel_Normal_01
   samplerSource=SSM_Wrap_WorldGroupSettings
   samplerType=SAMPLERTYPE_Normal
   texture=/Game/Environments/Apollo/Sets/Terrain/Textures/T_Apollo_Gravel_N_01
  }
  NamedRerouteDeclaration(None=LinearInterpolate(A=@n10, B=@n39.RGB, Alpha=@n37)) {name=Normals, nodeColor=<rgba 0.5 0 1 1>}
  n40 = MaterialFunctionCall() {materialFunction=/Game/Athena/Apollo/Environments/Landscape/Materials/Biome/CommonMaterials/MF_GravelAndMudColors}
  n41 = Multiply(A=@n40.GravelColorMod, B=ComponentMask(None=@n38) {b=true, g=true, r=true})
  NamedRerouteDeclaration(None=LinearInterpolate(A=Multiply(A=@n40.GravelEdgeColorMod, B=@n41), B=@n41, Alpha=@n34)) {name=Diffuse, nodeColor=<rgba 0.5 0 1 1>}
 ]}
