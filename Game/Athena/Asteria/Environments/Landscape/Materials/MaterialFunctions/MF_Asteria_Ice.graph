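# Landscape material function for Asteria's ice layer, blending ice with an incoming material and outputting the result, ice base color, heightmap and mask.
# Noise textures, a three-color lerp, fresnel, slope and light-direction masks build the ice color and emissive, with a flattened tiled normal.
# A desaturated, tinted copy of the input forms a frozen-cliff variant.
# Mobile swaps in a simplified flat-color ice material through a shading path switch.
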MaterialFunction /Game/Athena/Asteria/Environments/Landscape/Materials/MaterialFunctions/MF_Asteria_Ice
graph {nodes [
  Constant()
  n1 = MakeMaterialAttributes!(
   BaseColor=NamedRerouteUsage()
   Metallic=Constant()
   Specular=NamedRerouteUsage()
   Roughness=NamedRerouteUsage()
   Anisotropy
   EmissiveColor=NamedRerouteUsage()
   Opacity
   OpacityMask
   Normal=NamedRerouteUsage()
   Tangent
   WorldPositionOffset
   SubsurfaceColor
   ClearCoat
   ClearCoatRoughness
   AmbientOcclusion
   Refraction
   CustomizedUVs_0
   CustomizedUVs_1
   CustomizedUVs_2
   CustomizedUVs_3
   CustomizedUVs_4
   CustomizedUVs_5
   CustomizedUVs_6
   CustomizedUVs_7
   PixelDepthOffset
   ShadingModel
   Displacement
  )
  n2 = NamedRerouteUsage()
  n3 = MaterialFunctionCall!(
   Material 1=NamedRerouteUsage()
   Height Texture 1=NamedRerouteUsage()
   Material 2=ShadingPathSwitch(Default=QualitySwitch(Default=@n1), Mobile=QualitySwitch(Default=@n2))
   Height Texture 2=NamedRerouteUsage()
   Transistion Phase=NamedRerouteUsage()
   UseContrast
   Contrast
   AddLayerMask
  ) {materialFunction=/Game/Athena/Apollo/Environments/Landscape/Materials/Biome/CommonMaterials/MF_BlendLandscapeLayers}
  FunctionOutput(None=@n3.BlendedMaterial)
  FunctionOutput(None=NamedRerouteUsage()) {outputName=Ice_BaseColor}
  FunctionOutput(None=@n3.Resulting Height Map) {outputName=Heightmap}
  FunctionOutput(None=@n3.Alpha) {outputName=IceMask}
  LandscapeLayerCoords()
  n4 = VectorParameter() {parameterName=BaseColorC, defaultValue=<rgba 0.0174 0.0891 0.3542 1>}
  n5 = VectorParameter() {parameterName=BaseColorA, defaultValue=<rgba 0.0128 0.3708 0.891 1>}
  n6 = NamedRerouteUsage()
  LinearInterpolate(A=@n4.RGB, B=@n5.RGB, Alpha=@n6)
  n7 = MaterialFunctionCall(A=VectorParameter() {parameterName=BaseColorB, defaultValue=<rgba 0.0223 0.0944 0.714 1>}.RGB, B=@n4.RGB, C=@n5.RGB, Alpha=@n6) {materialFunction=&Lerp_3Color}
  LinearInterpolate(A=@n7.Result, B=VectorParameter() {parameterName=CracksColor, defaultValue=<rgba 0.6406 0.6406 0.6406 1>}.RGB, Alpha=NamedRerouteUsage())
  n8 = NamedRerouteUsage()
  n9 = ScalarParameter() {defaultValue=800, parameterName=BaseNormal_Scale}
  n10 = MaterialFunctionCall(
   Normal=MaterialFunctionCall!(
    Texture=NamedRerouteUsage()
    LandscapeCoordinates=Divide(A=@n8, B=@n9)
    LowEndTiling
    HexGridGradientSharpness
    HexGridSize
    HexGridUVInput
    MinScale
    MaxScale
    MinOffset
    MaxOffset
    "MaxRotation(Degrees)"
    "MinRotation(Degrees)"
    forceLowQuality
   ) {materialFunction=/Game/Athena/Asteria/Environments/Landscape/Materials/MaterialFunctions/UtilityFunctions/Landscape_TextureSample}.Result
   Flatness=ScalarParameter() {defaultValue=0.8, parameterName=Ice_BaseNormalFlatness}
  ) {materialFunction=&FlattenNormal}
  n11 = ScalarParameter() {defaultValue=1, parameterName=Cracks_Flatness_Max}
  n12 = ScalarParameter() {defaultValue=0.75, parameterName=Cracks_Flatness_Min}
  n13 = NamedRerouteUsage()
  MaterialFunctionCall(
   BaseNormal=@n10.Result
   AdditionalNormal=MaterialFunctionCall(Normal=NamedRerouteUsage(), Flatness=LinearInterpolate(A=@n11, B=@n12, Alpha=@n13)) {materialFunction=&FlattenNormal}.Result
  ) {materialFunction=&BlendAngleCorrectedNormals}
  n14 = NamedRerouteUsage()
  Multiply(A=@n14)
  NamedRerouteDeclaration(None=TextureObject() {texture=/Game/Athena/Environments/Meteor/Textures/Meteor_Cracks_M}) {name=CracksTexture, nodeColor=<rgba 0.5469 1 0 1>}
  n15 = MaterialFunctionCall(
   Texture=NamedRerouteUsage()
   LandscapeCoordinates=Divide(A=NamedRerouteUsage(), B=ScalarParameter() {defaultValue=1, parameterName=Ice_CracksScale})
  ) {materialFunction=&Landscape_TextureSample}
  NamedRerouteDeclaration(None=Multiply(A=ComponentMask(None=@n15.Result) {r=true}, B=NamedRerouteUsage())) {name=Cracks_Mask, nodeColor=<rgba 1 0 0.3516 1>}
  NamedRerouteDeclaration(None=ScalarParameter() {defaultValue=1, parameterName=Ice_Specular}) {name=Ice_Specular, nodeColor=<rgba 0 1 1 1>}
  NamedRerouteDeclaration(None=LinearInterpolate(
    A=ScalarParameter() {defaultValue=0.1, parameterName=Ice_roughness_Min}
    B=ScalarParameter() {defaultValue=0.15, parameterName=Ice_roughness_Max}
    Alpha=NamedRerouteUsage()
   )) {name=Ice_Roughness, nodeColor=<rgba 0 1 1 1>}
  n16 = MaterialFunctionCall(
   Texture=NamedRerouteUsage()
   LandscapeCoordinates=Divide(A=NamedRerouteUsage(), B=ScalarParameter() {defaultValue=2048, parameterName=SCale_01})
  ) {materialFunction=&Landscape_TextureSample}
  NamedRerouteDeclaration(None=ComponentMask(None=@n16.Result) {g=true}) {name=Roughness_Noise, nodeColor=<rgba 1 0 0.2812 1>}
  NamedRerouteDeclaration(None=TextureObject() {samplerType=SAMPLERTYPE_Normal, texture=/Game/Global/Textures/Default/Tileable_Patterns/T_Desert_Cracks_N}) {name=Ice_Cracks_Normal, nodeColor=<rgba 0.1484 0 1 1>}
  NamedRerouteDeclaration(None=TextureObject() {samplerType=SAMPLERTYPE_Normal, texture=/Game/Environments/Asteria/Terrain/Textures/T_Asteria_Ice_ChunkyWall_01_N}) {name=Ice_Normal_01, nodeColor=<rgba 1 0.0703 0 1>}
  NamedRerouteDeclaration(None=@n10.Result) {name=Ice_Normal, nodeColor=<rgba 0.5234 0 1 1>}
  n17 = VertexNormalWS()
  n18 = PixelNormalWS()
  n19 = ScalarParameter() {defaultValue=0.5, parameterName=BlendPixelNormal}
  n20 = DotProduct(
   A=CameraVectorWS()
   B=NamedRerouteDeclaration(None=LinearInterpolate(A=@n17, B=@n18, Alpha=@n19)) {name=VertexNormal, nodeColor=<rgba 0 1 0.7188 1>}
  )
  NamedRerouteDeclaration(None=Power(Base=OneMinus(None=@n20), Exp=ScalarParameter() {defaultValue=12, parameterName=FresnelPower})) {name=Ice_Fresnel, nodeColor=<rgba 0 0.6484 1 1>}
  n21 = Abs(None=ComponentMask(None=NamedRerouteUsage()) {b=true})
  NamedRerouteDeclaration(None=Clamp(None=Power(Base=@n21) {constExponent=32})) {name=Ice_SlopwMask, nodeColor=<rgba 1 0.9375 0 1>}
  n22 = DotProduct(A=AtmosphericLightVector(), B=NamedRerouteUsage())
  n23 = Multiply(A=@n22, B=ScalarParameter() {defaultValue=2, parameterName=SS_MaskPower})
  NamedRerouteDeclaration(None=Clamp(None=Power(Base=@n23))) {name=SS_Mask, nodeColor=<rgba 0 0.3672 1 1>}
  NamedRerouteDeclaration(None=TextureObject() {samplerType=SAMPLERTYPE_Masks, texture=/Game/Environments/Asteria/Terrain/Textures/T_Asteria_Ice_Noise_M}) {name=GrayscaleNoise_Texture, nodeColor=<rgba 0 0.4844 1 1>}
  n24 = VectorParameter() {parameterName=Ice_SlopeColor, defaultValue=<rgba 0.222 0.66 1 1>}
  n25 = NamedRerouteUsage()
  NamedRerouteDeclaration(None=Multiply(A=Multiply(A=@n24.RGB, B=@n25), B=ScalarParameter() {defaultValue=5, parameterName=IceSlopeColorStrength}) {constB=4}) {name=IceEmissive, nodeColor=<rgba 1 0.0703 0 1>}
  NamedRerouteDeclaration(None=MaterialFunctionCall(Texture=NamedRerouteUsage(), LandscapeCoordinates=@n14) {materialFunction=&Landscape_TextureSample}.Result) {name=Crack_normal, nodeColor=<rgba 0.0547 1 0 1>}
  n26 = GetMaterialAttributes(None=FunctionInput() {inputName=inputMaterial, inputType=FunctionInput_MaterialAttributes}) {attributeGetTypes [#1]}
  n27 = Multiply(
   A=Desaturation(None=@n26.BaseColor, Fraction=Constant() {r=1})
   B=Multiply(A=VectorParameter() {parameterName=Cliff_IceColor, defaultValue=<rgba 1 7 4 1>}.RGB, B=NamedRerouteUsage())
  )
  n28 = NamedRerouteUsage()
  n29 = NamedRerouteUsage()
  n30 = NamedRerouteUsage()
  NamedRerouteDeclaration(None=MaterialFunctionCall(
    Material 1=SetMaterialAttributes!(MaterialAttributes=@n26.MaterialAttributes, BaseColor=@n27, Roughness=@n28, Specular=@n29) {attributeSetTypes [#1, #2, #3]}
    Height Texture 1=@n30
    Material 2=NamedRerouteDeclaration(None=@n26.MaterialAttributes) {name=Input_Material, nodeColor=<rgba 0.7578 1 0 1>}
    Height Texture 2=@n30
    Transistion Phase=NamedRerouteUsage()
   ) {materialFunction=&MF_BlendLandscapeLayers}.BlendedMaterial) {name=FrozenCliff_Material, nodeColor=<rgba 1 0.4453 0 1>}
  NamedRerouteDeclaration(None=FunctionInput(Preview=Constant() {r=1}) {bUsePreviewValueAsDefault=true, inputName=LayerMask, inputType=FunctionInput_Scalar}) {name=Ice_Layer_Mask, nodeColor=<rgba 0 0.9766 1 1>}
  NamedRerouteDeclaration(None=MakeMaterialAttributes(
    BaseColor=VectorParameter() {parameterName=BaseColorA, defaultValue=<rgba 0.0271 0.1778 0.5521 1>}.RGB
    Metallic=Constant()
    Specular=NamedRerouteUsage()
    Roughness=NamedRerouteUsage()
   )) {name=Mobile-Material, nodeColor=<rgba 0 1 0.3906 1>}
  NamedRerouteDeclaration(None=MaterialFunctionCall() {materialFunction=&/MF_AltitudeMask}.Result) {name=AltitudeMask, nodeColor=<rgba 0 1 0.5781 1>}
  NamedRerouteDeclaration(None=FunctionInput(Preview=Constant()) {bUsePreviewValueAsDefault=true, inputName=InputHeightMap, inputType=FunctionInput_Scalar}) {name=InputHeightMap, nodeColor=<rgba 0 1 0.6484 1>}
  NamedRerouteDeclaration(None=MaterialFunctionCall(Texture=NamedRerouteUsage(), LandscapeCoordinates=NamedRerouteUsage()) {materialFunction=&Landscape_TextureSample}.Result) {name=VariationNoise_01, nodeColor=<rgba 1 0 0.6562 1>}
  NamedRerouteDeclaration(None=MaterialFunctionCall(Texture=NamedRerouteUsage(), LandscapeCoordinates=NamedRerouteUsage()) {materialFunction=&Landscape_TextureSample}.Result) {name=VariationNoise_02, nodeColor=<rgba 1 0 0.75 1>}
  NamedRerouteDeclaration(None=MaterialFunctionCall(Texture=NamedRerouteUsage(), LandscapeCoordinates=NamedRerouteUsage()) {materialFunction=&Landscape_TextureSample}.Result) {name=VariationNoise_03, nodeColor=<rgba 0.7812 0 1 1>}
  NamedRerouteDeclaration(None=TextureObject() {samplerType=SAMPLERTYPE_Masks, texture=&T_Asteria_Ice_Noise_M}) {name=RGBNoise_Texture, nodeColor=<rgba 0.7109 1 0 1>}
  NamedRerouteDeclaration(None=ComponentMask(None=WorldPosition().XYZ) {g=true, r=true}) {name=Coordinates, nodeColor=<rgba 0 1 0.0859 1>}
  NamedRerouteDeclaration(None=ComponentMask(None=@n16.Result) {r=true}) {name=BaseColorNoise, nodeColor=<rgba 1 0 0.2812 1>}
  n31 = Multiply(A=@n7.Result, B=@n24.RGB)
  NamedRerouteDeclaration(None=LinearInterpolate(
    A=LinearInterpolate(A=@n7.Result, B=@n31, Alpha=@n25)
    B=VectorParameter() {parameterName=Ice_FresnelColor, defaultValue=<rgba 1 1 1 1>}.RGB
    Alpha=NamedRerouteUsage()
   )) {name=Ice_BaseColor, nodeColor=<rgba 0 0.3672 1 1>}
  OneMinus(None=@n22)
  ScalarParameter() {defaultValue=0.05, parameterName=Variation_NoiseScale_01}
  ScalarParameter() {defaultValue=2, parameterName=Variation_NoiseScale_02}
  ScalarParameter() {defaultValue=4, parameterName=Variation_NoiseScale_03}
  ScalarParameter() {defaultValue=0.1, parameterName=BaseColorNoise_Scale}
  TextureCoordinate()
  TextureObject() {samplerType=SAMPLERTYPE_Masks, texture=/Game/Athena/WinterArt/Textures/Cracks_Mask}
  TextureObject() {samplerType=SAMPLERTYPE_Normal, texture=/Game/Characters/Player/Female/Medium/Bodies/F_MED_Bistro_Astronaut/FX/T_GaussianNoise_Normal}
  TextureObject() {samplerType=SAMPLERTYPE_Masks, texture=/Game/Weapons/WeaponSkins/Wraps/Textures/IceBreaker/T_IceBreakerCracks}
  TextureObject() {samplerType=SAMPLERTYPE_Normal, texture=/Game/Environments/Sets/Arid/Textures/T_desert_cracks_N}
  TextureObject() {samplerType=SAMPLERTYPE_Normal, texture=/Game/Global/Textures/Default/Noise/T_Curie_Noise_Normal_Epic}
 ]}
