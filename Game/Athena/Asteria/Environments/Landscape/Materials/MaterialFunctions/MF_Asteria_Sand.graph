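# Landscape material function that builds the Asteria terrain's sand layer, outputting a Material, an UnderOceanSandMaterial variant and a Height from diffuse alpha.
# Shared sand textures are tinted by large-scale noise and biome colors, with UVs switched between landscape and underwater mapping.
# World Z against a water level offset drives height-blended masks for wet-sand darkening and surface response.
# Epic quality adds world-space noise to normals, reduced on slopes, and rescales specular/roughness UVs.
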MaterialFunction /Game/Athena/Asteria/Environments/Landscape/Materials/MaterialFunctions/MF_Asteria_Sand
graph {
 properties {description=-3909}
 nodes [
  Constant() {desc=apollo setting, r=3909}
  Constant() {desc=apollo setting, r=4100}
  n1 = NamedRerouteUsage()
  n2 = Constant() {desc=Make underwater sand darker, r=0.5}
  n3 = LinearInterpolate(A=Constant() {r=1}, B=ScalarParameter() {defaultValue=0.6, parameterName=NewSandNoiseLB}, Alpha=NamedRerouteUsage())
  n4 = TextureSample(UVs=NamedRerouteUsage()) {samplerSource=SSM_Wrap_WorldGroupSettings, texture=/Game/Environments/Apollo/Sets/Terrain/Textures/T_Apollo_Sand_D_01}
  n5 = FunctionInput(Preview=StaticBool() {value=true}) {
   bUsePreviewValueAsDefault=true
   description="Default=True"
   inputName=BiomeColors-UseDefault
   inputType=FunctionInput_StaticBool
   previewValue=<64 64 64 6>
  }
  n6 = MaterialFunctionCall!(
   Grass
   Gravel
   Mud
   Sand=NamedRerouteDeclaration(None=Multiply(A=@n3, B=@n4.RGB)) {name=DefaultSandDiffuse, nodeColor=<rgba 1 0 1 1>}
   BiomeMask-RGBA=FunctionInput(Preview=Constant()) {bUsePreviewValueAsDefault=true, inputName=BiomeMask-RGBA, inputType=FunctionInput_Vector4}
   UseBiomeColorMask=@n5
   UseLandscapeCoords
  ) {materialFunction=&/MF_AsteriaBiomeColors}
  n7 = MaterialFunctionCall!(WorldPosition) {materialFunction=/Game/Athena/Environments/Landscape/MaterialFunctions/Utilities/MF_Apollo_Terrain_Coords}
  n8 = Constant() {r=1}
  n9 = SetMaterialAttributes!(
   MaterialAttributes
   BaseColor=LinearInterpolate(
    A=Multiply(A=@n1, B=@n2)
    B=Multiply(A=@n6.Sand, B=@n2)
    Alpha=MaterialFunctionCall(UVs=@n7.Adjustable, Texture Scale=@n8) {materialFunction=&ScaleUVsByCenter}.0-1 mask
   )
   Specular=NamedRerouteUsage()
   Roughness=Constant() {r=0.6}
   Normal=Transform(None=NamedRerouteUsage())
  ) {attributeSetTypes [
    #1
    #2
    #3
    #4
   ]}
  FunctionOutput(None=@n9) {outputName=UnderOceanSandMaterial}
  n10 = NamedRerouteDeclaration(None=@n4.A) {name=DefaultSandHeight, nodeColor=<rgba 1 0 1 1>}
  FunctionOutput(None=Power(Base=@n10, Exp=Constant() {r=8})) {outputName=Height}
  n11 = NamedRerouteUsage()
  n12 = ScalarParameter() {defaultValue=1.5, group=Sand, parameterName=Sand-SpecRoughScaleMod}
  n13 = TextureSample(UVs=QualitySwitch(Default=@n11, Epic=Multiply(A=@n11, B=@n12))) {
   samplerSource=SSM_Wrap_WorldGroupSettings
   samplerType=SAMPLERTYPE_Masks
   texture=/Game/Environments/Apollo/Sets/Terrain/Textures/T_Apollo_Sand_S_01
  }
  n14 = Multiply(A=@n13.B, B=Constant() {r=1.75})
  n15 = Constant() {r=-1.5}
  n16 = Multiply(A=Saturate(None=Add(A=@n14, B=@n15)), B=@n13.R)
  n17 = Constant() {r=0.3}
  n18 = NamedRerouteUsage()
  n19 = NamedRerouteUsage()
  n20 = Add(A=Multiply(A=OneMinus(None=@n19), B=Constant() {r=4.5}), B=Constant() {r=-3.9})
  n21 = NamedRerouteUsage()
  n22 = Constant() {r=2}
  n23 = Saturate(None=Add(A=Multiply(A=@n21, B=@n22), B=Constant() {r=-1}))
  n24 = ComponentMask(None=WorldPosition() {worldPositionShaderOffset=WPT_ExcludeAllShaderOffsets}.XYZ) {b=true}
  n25 = Add(A=NamedRerouteUsage(), B=Constant() {r=191})
  n26 = Saturate(None=Divide(A=Add(A=@n24, B=@n25), B=Constant() {r=200}))
  n27 = SetMaterialAttributes(
   BaseColor=NamedRerouteUsage()
   Specular=LinearInterpolate(A=@n16, B=@n13.R, Alpha=NamedRerouteUsage())
   Roughness=Max(A=LinearInterpolate(A=@n17, B=@n13.B, Alpha=@n18), B=Saturate(None=@n20))
   Normal=LinearInterpolate(A=NamedRerouteUsage(), B=Constant3Vector() {constant=<rgba 0 0 1 0>}, Alpha=Multiply(A=@n23, B=@n26))
  ) {attributeSetTypes [
    #1
    #2
    #3
    #4
   ]}
  FunctionOutput(None=@n27) {outputName=Material}
  LinearInterpolate(A=Constant() {r=-0.5}, B=Constant() {r=1.2}, Alpha=@n13.B)
  n28 = Add(
   A=ComponentMask(None=WorldPosition() {worldPositionShaderOffset=WPT_ExcludeAllShaderOffsets}.XYZ) {b=true}
   B=NamedRerouteDeclaration(None=Constant() {r=65}) {name=WaterLevelOffset, nodeColor=<rgba 1 0 1 1>}
  )
  n29 = Divide(A=@n28, B=Constant() {r=25})
  n30 = OneMinus(None=@n10)
  n31 = Multiply(A=@n30, B=Constant() {r=4})
  n32 = Constant() {r=-0.5}
  NamedRerouteDeclaration(None=MaterialFunctionCall(Transition Phase=Saturate(None=@n29), Height Texture=Add(A=@n31, B=@n32)) {materialFunction=&HeightLerp}.Lerp Alpha No Contrast) {name=WaterLevel-HeightMask, nodeColor=<rgba 1 0 1 1>}
  n33 = WorldPosition() {worldPositionShaderOffset=WPT_ExcludeAllShaderOffsets}
  n34 = TextureSample(UVs=Divide(A=ComponentMask(None=@n33.XYZ) {g=true, r=true}, B=Constant() {r=8192}) {constB=8192}) {
   samplerSource=SSM_Wrap_WorldGroupSettings
   samplerType=SAMPLERTYPE_Grayscale
   texture=/Game/Athena/Apollo/Environments/Landscape/Materials/Biome/CommonMaterials/T_ApolloGrass_ColorMask
  }
  NamedRerouteDeclaration(None=@n34.R) {name=SandNoise, nodeColor=<rgba 1 0 1 1>}
  n35 = TextureSample(UVs=NamedRerouteUsage()) {
   samplerSource=SSM_Wrap_WorldGroupSettings
   samplerType=SAMPLERTYPE_Normal
   texture=/Game/Environments/Apollo/Sets/Terrain/Textures/T_Apollo_Sand_N_01
  }
  n36 = Constant3Vector() {constant=<rgba 0 0 1 0>}
  n37 = WorldPosition() {worldPositionShaderOffset=WPT_ExcludeAllShaderOffsets}
  n38 = TextureSample(UVs=Divide(A=ComponentMask(None=@n37.XYZ) {g=true, r=true}, B=ScalarParameter() {defaultValue=85, parameterName=NR-testsandnoise}) {constB=8192}) {samplerSource=SSM_Wrap_WorldGroupSettings, samplerType=SAMPLERTYPE_LinearColor, texture=/Game/Global/Textures/Default/Noise/T_Apollo_TCNoise}
  n39 = Subtract(A=@n38.RGB, B=Constant() {r=0.5})
  n40 = Constant3Vector() {constant=<rgba 0.6 0.6 0 1>}
  n41 = Constant() {r=-5}
  n42 = Constant() {r=1.5}
  n43 = NamedRerouteUsage()
  n44 = LinearInterpolate(
   A=Add(A=Multiply(A=@n39, B=@n40), B=Constant3Vector() {constant=<rgba 0 0 1 0>})
   B=Constant3Vector() {constant=<rgba 0 0 1 0>}
   Alpha=Saturate(None=LinearInterpolate(A=@n41, B=@n42, Alpha=@n43))
  )
  n45 = ComponentMask(None=VertexNormalWS()) {b=true}
  n46 = Saturate(None=LinearInterpolate(A=Constant() {r=-3}, B=Constant() {r=1}, Alpha=Abs(None=@n45))) {desc=Less on slopes}
  n47 = Multiply(A=Normalize(VectorInput=LinearInterpolate(A=@n36, B=@n44, Alpha=@n46)), B=Constant3Vector() {constant=<rgba 0.5 0.5 0 0>})
  NamedRerouteDeclaration(None=QualitySwitch(Default=@n35.RGB, Epic=Add(A=@n35.RGB, B=@n47))) {name=Normals, nodeColor=<rgba 1 0 1 1>}
  NamedRerouteDeclaration(None=@n16) {name=UnderOcean-Spec, nodeColor=<rgba 1 0 1 1>}
  n48 = Add(A=@n28, B=Constant() {desc=slightly wet sand height offset, r=10})
  n49 = Constant() {r=125}
  NamedRerouteDeclaration(None=Saturate(None=Divide(A=@n48, B=@n49))) {name=WaterLevel, nodeColor=<rgba 1 0 1 1>}
  n50 = Add(A=NamedRerouteUsage(), B=Constant() {r=0.9})
  n51 = MaterialFunctionCall!(Diffuse 1, Diffuse 2, Transistion Phase=NamedRerouteUsage(), Height Texture 1=@n30, Height Texture 2=NamedRerouteUsage(), Contrast) {materialFunction=/Engine/Functions/Engine_MaterialFunctions02/Texturing/HeightLerpWithTwoHeightMaps}
  n52 = Saturate(None=Add(A=@n2, B=OneMinus(None=@n51.Lerp Alpha No Contrast)))
  NamedRerouteDeclaration(None=Multiply(A=Saturate(None=@n50), B=Multiply(A=@n6.Sand, B=@n52))) {name=BaseColor, nodeColor=<rgba 1 0 1 1>}
  NamedRerouteDeclaration(None=@n51.Lerp Alpha No Contrast) {name=WaterLevel-HeightMask2, nodeColor=<rgba 1 0 1 1>}
  n53 = Constant() {r=1000}
  n54 = TextureCoordinate()
  n55 = StaticBool()
  NamedRerouteDeclaration(None=StaticSwitch(
    True=Multiply(A=@n53, B=@n54)
    False=LandscapeLayerCoords() {mappingScale=7}
    Value=FunctionInput(Preview=@n55) {bUsePreviewValueAsDefault=true, inputName=UseUnderOceanSandMaterial, inputType=FunctionInput_StaticBool, sortPriority=10}
   )) {name=SandUVs, nodeColor=<rgba 1 0 1 1>}
  Saturate(None=LinearInterpolate(A=ScalarParameter() {parameterName=SSSndA}, B=ScalarParameter() {defaultValue=1, parameterName=SSSndB}, Alpha=@n13.B))
  VertexNormalWS()
 ]
}
