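# Material function that shades snow after the RVT stage, rebuilding incoming material attributes as a subsurface material with snow color, roughness, specular and sparkle.
# Mask and slippery-snow inputs accompany quality and shading-path switches that simplify the look below Epic and on mobile.
# Epic-only sparkle emissive samples the skylight through view-dependent glints, ping-ponging circle textures between depth-based scales.
# Separate outputs expose albedo, height and sparkle masks, plus deprecated constants.
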MaterialFunction /Game/Athena/Asteria/Environments/Landscape/Materials/MaterialFunctions/MF_SnowPostRVT
graph {nodes [
  Constant3Vector() {constant=<rgba 1 0 0 0>}
  Constant()
  Constant() {r=1}
  n1 = SetMaterialAttributes!(
   MaterialAttributes=NamedRerouteUsage()
   BaseColor=NamedRerouteUsage()
   Roughness=NamedRerouteUsage()
   Specular=NamedRerouteUsage()
   EmissiveColor=NamedRerouteUsage()
   ShadingModel=ShadingModel() {shadingModel=MSM_Subsurface}
   Opacity=NamedRerouteUsage()
   SubsurfaceColor=NamedRerouteUsage()
  ) {attributeSetTypes [
    #1
    #2
    #3
    #4
    #5
    #6
    #7
   ]}
  FunctionOutput(None=@n1) {outputName=SnowMaterial, sortPriority=1}
  FunctionOutput(None=Constant()) {outputName=Sparkle-RoughnessDEPRICATED}
  FunctionOutput(None=Constant() {r=1}) {outputName=SparkleMaskDEPRICATED}
  FunctionOutput(None=NamedRerouteUsage()) {outputName=SparkleEmissive}
  n2 = TextureSample(UVs=MaterialFunctionCall!(Scale=Constant() {r=600}) {materialFunction=/Game/Athena/Apollo/Environments/Landscape/Materials/Biome/CommonMaterials/MF_SimpleWorldUV}.Result) {
   desc=T_Athena_Snow_01_MOBILE_M
   samplerSource=SSM_Wrap_WorldGroupSettings
   samplerType=SAMPLERTYPE_LinearColor
   texture=/Game/Athena/Environments/Landscape/Textures/T_Athena_Snow_01_MOBILE_M
  }
  n3 = Multiply(A=@n2.R, B=@n2.R)
  n4 = Multiply(A=@n2.R, B=Multiply(A=@n3, B=@n3))
  n5 = Multiply(A=@n4, B=ScalarParameter() {defaultValue=0.5, group=SnowProgression, parameterName=SPK-modroughx})
  FunctionOutput(None=NamedRerouteDeclaration(None=Saturate(None=@n5)) {name=Sparkle-Roughness, nodeColor=<rgba 1 0 1 1>}) {outputName=Sparkle-Roughness}
  FunctionOutput(None=MaterialFunctionCall() {materialFunction=&/MF_SnowColorBase}.SnowAlbedo) {outputName=PreRVT_SnowAlbedo}
  n6 = NamedRerouteUsage()
  n7 = QualitySwitch(Default=@n6, Epic=Multiply(A=@n6, B=ScalarParameter() {defaultValue=0.8, group=PostRVTSnow, parameterName=SnowBaseColorBrightness}))
  FunctionOutput(None=@n7) {outputName=OutputSnowAlbedo}
  n8 = NamedRerouteUsage()
  n9 = MaterialFunctionCall!(
   Texture=NamedRerouteUsage()
   LandscapeCoordinates=@n8
   LowEndTiling
   HexGridGradientSharpness
   HexGridSize
   HexGridUVInput=Multiply(A=@n8, B=ScalarParameter() {defaultValue=0.5, parameterName=HexGridTile})
   MinScale=ScalarParameter() {defaultValue=1, parameterName=MinScale}
   MaxScale=ScalarParameter() {defaultValue=1, parameterName=MaxScale}
   MinOffset
   MaxOffset
   "MaxRotation(Degrees)"=ScalarParameter() {parameterName=MaxRotation}
   "MinRotation(Degrees)"=ScalarParameter() {parameterName=MinRotation}
   forceLowQuality
  ) {materialFunction=/Game/Athena/Asteria/Environments/Landscape/Materials/MaterialFunctions/UtilityFunctions/Landscape_TextureSample}
  FunctionOutput(None=Multiply(A=ComponentMask(None=@n9.Result) {a=true}, B=ScalarParameter() {defaultValue=0.5, parameterName=SnowHeightMapBias})) {outputName=HeightOutput, sortPriority=1}
  n10 = ScalarParameter() {defaultValue=1, parameterName=SAPK_modSpecx}
  FunctionOutput(None=NamedRerouteDeclaration(None=Multiply(A=@n4, B=@n10)) {name=SparkleMask, nodeColor=<rgba 0.8281 1 0 1>}) {outputName=SparkleMask}
  n11 = VectorParameter() {group=PostRVTSnow, parameterName=SnowSSSColor, defaultValue=<rgba 0.0677 0.2393 0.5 1>}
  n12 = ConstantBiasScale(None=DotProduct(A=VertexNormalWS(), B=AtmosphericLightVector()))
  Multiply(
   A=@n11.RGB
   B=Power(Base=Saturate(None=@n12), Exp=ScalarParameter() {defaultValue=2, group=PostRVTSnow, parameterName=SnowSSSSunMaskExponent})
  )
  n13 = NamedRerouteUsage()
  n14 = LinearInterpolate(
   A=ScalarParameter() {defaultValue=0.2, group=PostRVTSnow, parameterName=SnowRoughnessMinEpic}
   B=ScalarParameter() {defaultValue=1, group=PostRVTSnow, parameterName=SnowRoughnessMaxEpic}
   Alpha=@n13
  )
  Multiply(A=@n14, B=NamedRerouteUsage())
  n15 = Constant() {r=400}
  n16 = Divide(A=CameraPositionWS(), B=Constant() {r=10000})
  n17 = TextureSample(UVs=Add(
    A=MaterialFunctionCall(Scale=@n15) {materialFunction=&MF_SimpleWorldUV}.Result
    B=ComponentMask(None=@n16) {g=true, r=true}
   )) {
   desc=T_Athena_Snow_01_MOBILE_M
   samplerSource=SSM_Wrap_WorldGroupSettings
   samplerType=SAMPLERTYPE_LinearColor
   texture=&T_Athena_Snow_01_MOBILE_M
  }
  n18 = Multiply(A=@n17.R, B=@n17.R)
  n19 = Multiply(A=@n18, B=@n18)
  n20 = Saturate(None=Multiply(A=Multiply(A=@n17.R, B=@n19), B=Constant() {r=1.5}))
  Multiply(B=@n20)
  n21 = Saturate(None=DotProduct(A=CameraVectorWS(), B=PixelNormalWS()))
  Multiply(A=LinearInterpolate(Alpha=Power(Base=@n21)) {constA=1.1, constB=0.6}, B=@n6)
  n22 = QualitySwitch(
   Default=LinearInterpolate(
    A=ScalarParameter() {defaultValue=0.5, group=PostRVTSnow, parameterName=SnowRoughnessMin}
    B=ScalarParameter() {defaultValue=1, group=PostRVTSnow, parameterName=SnowRoughnessMax}
    Alpha=@n13
   )
   Epic=@n14
  )
  n23 = ScalarParameter() {defaultValue=0.5, group=PostRVTSnow, parameterName=SnowRoughnessMobile}
  NamedRerouteDeclaration(None=LinearInterpolate(A=ShadingPathSwitch(Default=@n22, Forward=@n23, Mobile=@n23), B=NamedRerouteUsage(), Alpha=NamedRerouteUsage())) {name=SnowRoughness, nodeColor=<rgba 1 0.0938 0 1>}
  n24 = NamedRerouteUsage()
  n25 = Saturate(None=Power(Base=OneMinus(None=@n24), Exp=ScalarParameter() {defaultValue=4, group=PostRVTSnow, parameterName=SnowSpecularSelfShadowing}) {constExponent=3})
  n26 = Multiply(
   A=NamedRerouteUsage()
   B=LinearInterpolate(
    A=ScalarParameter() {defaultValue=0.3, group=PostRVTSnow, parameterName=SnowSpecularMinEpic}
    B=ScalarParameter() {defaultValue=0.5, group=PostRVTSnow, parameterName=SnowSpecularMaxEpic}
    Alpha=@n25
   ) {constA=1}
  )
  n27 = ScalarParameter() {defaultValue=0.5, group=PostRVTSnow, parameterName=SnowSpecular}
  NamedRerouteDeclaration(None=LinearInterpolate(A=QualitySwitch(Default=@n27, Epic=@n26), B=NamedRerouteUsage(), Alpha=NamedRerouteUsage())) {name=SnowSpecular, nodeColor=<rgba 0 1 0.9766 1>}
  n28 = Divide(A=PixelDepth().R, B=ScalarParameter() {defaultValue=1000, group=PostRVTSnow, parameterName=SnowSparklePingPongDistance}) {constB=5000}
  n29 = Multiply(A=NamedRerouteUsage(), B=Divide(B=Ceil(None=@n28)))
  n30 = TextureSample(UVs=@n29) {
   samplerSource=SSM_Wrap_WorldGroupSettings
   samplerType=SAMPLERTYPE_LinearColor
   texture=/Game/Environments/Asteria/Terrain/Textures/T_SnowSparkleCircles_M
  }
  n31 = Add(A=@n28) {constB=0.5}
  n32 = Multiply(A=NamedRerouteUsage(), B=Divide(B=Ceil(None=@n31)))
  n33 = TextureSample(UVs=@n32) {
   samplerSource=SSM_Wrap_WorldGroupSettings
   samplerType=SAMPLERTYPE_LinearColor
   texture=&T_SnowSparkleCircles_M
  }
  n34 = LinearInterpolate(A=@n30.RGB, B=@n33.RGB, Alpha=NamedRerouteUsage())
  NamedRerouteDeclaration(None=ComponentMask(None=@n34) {g=true}) {name=SparkleSoftSpecular, nodeColor=<rgba 0 0.9531 1 1>}
  NamedRerouteDeclaration(None=@n21) {name=FacingAngleModification, nodeColor=<rgba 1 0.5391 0 1>}
  n35 = Constant3Vector() {constant=<rgba 0.5294 0.5922 0.702 1>}
  n36 = GetMaterialAttributes(None=FunctionInput(Preview=MakeMaterialAttributes!(
     BaseColor=@n35
     Metallic
     Specular
     Roughness
     Anisotropy
     EmissiveColor
     Opacity
     OpacityMask
     Normal
     Tangent
     WorldPositionOffset
     SubsurfaceColor
     ClearCoat
     ClearCoatRoughness
     AmbientOcclusion
     Refraction
     CustomizedUVs_0
     CustomizedUVs_1
     CustomizedUVs_2
     CustomizedUVs_3
     CustomizedUVs_4
     CustomizedUVs_5
     CustomizedUVs_6
     CustomizedUVs_7
     PixelDepthOffset
     ShadingModel
     Displacement
    )) {bUsePreviewValueAsDefault=true, inputName="Material;", inputType=FunctionInput_MaterialAttributes}) {attributeGetTypes [
    #1
    #8
    #3
    #2
   ]}
  NamedRerouteDeclaration(None=@n36.MaterialAttributes) {name=originalMaterial, nodeColor=<rgba 0 1 0.2031 1>}
  NamedRerouteDeclaration(None=@n2.R) {name=SnowDetailMask-R, nodeColor=<rgba 1 0 1 1>}
  NamedRerouteDeclaration(None=@n7) {name=snowBaseColor, nodeColor=<rgba 1 0 0.8437 1>}
  NamedRerouteDeclaration(None=LinearInterpolate(B=ScalarParameter() {defaultValue=0.95, group=PostRVTSnow, parameterName=SnowSSSOpacity}, Alpha=NamedRerouteUsage()) {constA=1}) {name=snowOpacity, nodeColor=<rgba 1 0.5859 0 1>}
  n37 = LinearInterpolate(A=ScalarParameter() {defaultValue=0.5, group=PostRVTSnow, parameterName=SnowSSSFacingExponent}, Alpha=NamedRerouteUsage())
  NamedRerouteDeclaration(None=Multiply(A=Multiply(A=@n11.RGB, B=@n37), B=NamedRerouteUsage())) {name=snowSubsurfaceColor, nodeColor=<rgba 0.0781 1 0 1>}
  n38 = Multiply(A=Add(A=FunctionInput() {bUsePreviewValueAsDefault=true, inputName=Mask, inputType=FunctionInput_Scalar}) {constB=-0.5}) {constB=2}
  NamedRerouteDeclaration(None=Power(Base=Saturate(None=@n38)) {constExponent=1}) {name=SnowMaskModified, nodeColor=<rgba 0.125 0 1 1>}
  NamedRerouteDeclaration(None=FunctionInput(Preview=Constant()) {bUsePreviewValueAsDefault=true, inputName=SlipperySnow_Mask, inputType=FunctionInput_Scalar}) {name=SL_Snow_Mask, nodeColor=<rgba 1 0 0.4922 1>}
  NamedRerouteDeclaration(None=@n36.Specular) {name=OriginalSpecular, nodeColor=<rgba 0.3828 1 0 1>}
  n39 = TextureSample(UVs=@n29) {
   samplerSource=SSM_Wrap_WorldGroupSettings
   samplerType=SAMPLERTYPE_Normal
   texture=/Game/Environments/Asteria/Terrain/Textures/T_SnowSparkleCircles_N
  }
  n40 = TextureSample(UVs=@n32) {
   samplerSource=SSM_Wrap_WorldGroupSettings
   samplerType=SAMPLERTYPE_Normal
   texture=&T_SnowSparkleCircles_N
  }
  n41 = ComponentMask(None=LinearInterpolate(A=@n39.RGB, B=@n40.RGB, Alpha=NamedRerouteUsage())) {b=true, g=true, r=true}
  n42 = Constant3Vector() {constant=<rgba 0 0 -0.9 1>}
  n43 = Transform(None=Normalize(VectorInput=Add(A=@n41, B=@n42)))
  n44 = DotProduct(A=ReflectionVectorWS(CustomWorldNormal=@n43), B=CameraVectorWS())
  n45 = ComponentMask(None=Saturate(None=Multiply(A=@n34) {constB=3})) {r=true}
  NamedRerouteDeclaration(None=Power(Base=Multiply(A=@n44, B=@n45), Exp=ScalarParameter() {defaultValue=20, group=PostRVTSnow, parameterName=SnowSparkleMaskExponent})) {name=HighendSparkleMask, nodeColor=<rgba 0 0.1094 1 1>}
  NamedRerouteDeclaration(None=@n36.Roughness) {name=Originalroughness, nodeColor=<rgba 0 1 0.625 1>}
  n46 = NamedRerouteUsage()
  n47 = OneMinus(None=NamedRerouteUsage())
  n48 = ScalarParameter() {defaultValue=4, group=SlipperySnow, parameterName=SL_SnowSpecularSelfShadowing}
  n49 = LinearInterpolate(
   A=ScalarParameter() {defaultValue=0.3, group=SlipperySnow, parameterName=SL_SnowSpecularMinEpic}
   B=ScalarParameter() {defaultValue=0.5, group=SlipperySnow, parameterName=SL_SnowSpecularMaxEpic}
   Alpha=Saturate(None=Power(Base=@n47, Exp=@n48) {constExponent=3})
  ) {constA=1}
  NamedRerouteDeclaration(None=QualitySwitch(Default=NamedRerouteUsage(), Epic=Multiply(A=@n46, B=@n49))) {name=SL_SnowSpecular, nodeColor=<rgba 0 1 0.5547 1>}
  n50 = NamedRerouteUsage()
  n51 = LinearInterpolate(A=@n50, B=ScalarParameter() {defaultValue=0.2, group=SlipperySnow, parameterName=SL_MinROughness}, Alpha=NamedRerouteUsage())
  n52 = NamedRerouteUsage()
  NamedRerouteDeclaration(None=ShadingPathSwitch(Default=QualitySwitch(Default=@n50, Epic=@n51), Forward=@n52, Mobile=@n52)) {name=SL_SnowRoughness, nodeColor=<rgba 1 0.9844 0 1>}
  NamedRerouteDeclaration(None=Max(A=@n17.G, B=@n20)) {name=Sparkle-Roughness_1, nodeColor=<rgba 1 0 1 1>}
  NamedRerouteDeclaration(None=@n36.BaseColor) {name=OriginalSnowColor, nodeColor=<rgba 1 0.3516 0 1>}
  NamedRerouteDeclaration(None=TextureObject() {texture=/Game/Environments/Asteria/Terrain/Textures/T_Asteria_WindBlownSnow_01_D}) {name=Macro1NormalHeightMap, nodeColor=<rgba 0 0.0156 1 1>}
  NamedRerouteDeclaration(None=@n43) {name=sparklenormal, nodeColor=<rgba 1 0.1875 0 1>}
  NamedRerouteDeclaration(None=@n36.Normal) {name=originalNormal, nodeColor=<rgba 1 0.6094 0 1>}
  NamedRerouteDeclaration(None=@n25) {name=FuzzSelfShadowing, nodeColor=<rgba 1 0 0.6562 1>}
  n53 = Abs(None=Add(A=Frac(None=@n28)) {constB=-0.5})
  NamedRerouteDeclaration(None=Saturate(None=Multiply(A=@n53) {constB=2})) {name=pingPongMask, nodeColor=<rgba 0 1 0.2969 1>}
  n54 = ReflectionVectorWS(CustomWorldNormal=NamedRerouteUsage())
  n55 = Multiply(
   A=Multiply(A=SkyLightEnvMapSample(Direction=@n54), B=NamedRerouteUsage())
   B=ScalarParameter() {defaultValue=8, group=PostRVTSnow, parameterName=SnowSparkleEmissiveBrightness}
  )
  n56 = NamedRerouteUsage()
  NamedRerouteDeclaration(None=QualitySwitch(Default=Constant(), Epic=Multiply(A=@n55, B=@n56))) {name=SnowSparkleEmissive, nodeColor=<rgba 0 0.2734 1 1>}
  n57 = TextureCoordinate()
  n58 = ScalarParameter() {defaultValue=4, parameterName=MeshSnowSparkleSize}
  n59 = ScalarParameter() {defaultValue=200, group=PostRVTSnow, parameterName=SnowSparkleSize}
  NamedRerouteDeclaration(None=StaticSwitchParameter(
    True=Multiply(A=@n57, B=@n58)
    False=MaterialFunctionCall(Scale=@n59) {materialFunction=&MF_SimpleWorldUV}.Result
   ) {parameterName=UseOnMesh}) {name=SparkleCoords, nodeColor=<rgba 1 0 0.2812 1>}
  Power(Base=@n17.R, Exp=Constant() {r=5})
  n60 = VectorParameter() {parameterName=Mobile-SnowColor1, defaultValue=<rgba 0.35 0.36 0.35 1>}
  QualitySwitch(Default=VectorParameter() {parameterName=PreRVTSnowAlbedo, defaultValue=<rgba 0.6437 0.6862 0.75 1>}.RGB, High=@n60.RGB, Medium=@n60.RGB)
  ScalarParameter() {defaultValue=4, group=SnowProgression, parameterName=SPK-mod-day}
  ScalarParameter() {defaultValue=0.5, group=SnowProgression, parameterName=SPK-modroughx}
  ScalarParameter() {defaultValue=5, group=SnowProgression, parameterName=SPK-FresPow2}
  ScalarParameter() {defaultValue=2, group=SnowProgression, parameterName=SPK-mod-morn}
  ScalarParameter() {defaultValue=800, group=SnowProgression, parameterName=SnowSparkleMaskScale}
  ScalarParameter() {defaultValue=10000, group=SnowProgression, parameterName=SPK-Div111}
  n61 = Constant()
  n62 = LinearInterpolate(A=Constant() {r=2}, B=Constant() {r=4}, Alpha=NamedRerouteUsage())
  n63 = NamedRerouteUsage()
  n64 = Constant() {r=0.55}
  n65 = Saturate(None=Subtract(A=OneMinus(None=@n63), B=Multiply(A=@n17.R, B=@n64) {constB=0.75}))
  n66 = Constant()
  ShadingPathSwitch(Default=QualitySwitch(Default=@n61, Low=@n61, High=@n61, Medium=@n61, Epic=Multiply(A=@n62, B=@n65)), Forward=@n66, Mobile=@n66)
  TextureObject() {samplerType=SAMPLERTYPE_LinearColor, texture=&T_Athena_Snow_01_MOBILE_M}
  n67 = ScalarParameter() {defaultValue=1600, parameterName=NoiseScale2}
  TextureSample(
   UVs=NamedRerouteDeclaration(None=MaterialFunctionCall(Scale=@n67) {materialFunction=&MF_SimpleWorldUV}.Result) {name=NoiseScale2, nodeColor=<rgba 1 0 1 1>}
   Tex=NamedRerouteUsage()
  ) {texture=&T_Asteria_WindBlownSnow_01_D}
 ]}
