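# Material function for Asteria landscape snow, height-blending a snow layer over an input material and outputting blended, base and pre-RVT variants plus height, mask and emissive.
# World-space noise, slope, cliff-crevice and variation masks accompany hex-grid-randomized wind-blown normal and height sampling and time-of-day shadow tinting.
# MF_SnowPostRVT supplies albedo and sparkle, while quality and shading-path switches simplify roughness and specular on lower tiers and mobile.
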MaterialFunction /Game/Athena/Asteria/Environments/Landscape/Materials/MaterialFunctions/MF_Asteria_Snow
graph {nodes [
  CollectionParameter() {collection=&FortniteMaterialParameters, parameterName=TimeOfDay_0-1}
  n1 = MaterialFunctionCall!(Scale=ScalarParameter() {defaultValue=200, parameterName=Snow_HighFrequency}) {materialFunction=/Game/Athena/Apollo/Environments/Landscape/Materials/Biome/CommonMaterials/MF_SimpleWorldUV}
  ComponentMask(None=MaterialFunctionCall(Vec2=@n1.Result) {materialFunction=&MF_Hash}.Vec2) {r=true}
  Constant3Vector() {constant=<rgba 0.4881 0.7045 0.901 0>}
  Constant3Vector() {constant=<rgba 0.2 0.3078 0.4 1>}
  Constant3Vector() {constant=<rgba 0.285 0.4009 0.5 1>}
  Constant() {r=8}
  Constant() {r=800}
  Constant() {r=10000}
  Constant() {r=3.5}
  Constant() {r=5}
  Constant() {r=2}
  Constant() {r=-0.15}
  Constant() {r=-1}
  Constant() {r=5}
  Constant()
  Constant()
  Constant() {r=2}
  Constant() {r=4}
  Constant() {r=1}
  Constant() {r=0.5}
  Constant()
  Constant() {r=0.7}
  Constant() {r=1200}
  Constant() {r=3000}
  n2 = NamedRerouteUsage()
  n3 = SetMaterialAttributes!(
   MaterialAttributes=@n2
   BaseColor=NamedRerouteUsage()
   Roughness=NamedRerouteUsage()
   Specular=NamedRerouteUsage()
   Normal=NamedRerouteUsage()
  ) {attributeSetTypes [
    #1
    #2
    #3
    #4
   ]}
  n4 = MaterialFunctionCall!(
   Material 1=@n2
   Height Texture 1=NamedRerouteUsage()
   Material 2=@n3
   Height Texture 2=NamedRerouteUsage()
   Transistion Phase=NamedRerouteUsage()
   UseContrast
   Contrast
   AddLayerMask
  ) {materialFunction=/Game/Athena/Apollo/Environments/Landscape/Materials/Biome/CommonMaterials/MF_BlendLandscapeLayers}
  FunctionOutput(None=@n4.BlendedMaterial)
  FunctionOutput(None=@n4.Resulting Height Map) {outputName=HeightMap}
  FunctionOutput(None=@n4.Alpha) {outputName=SnowScreeMask}
  n5 = MaterialFunctionCall!(Mask, SlipperySnow_Mask, "Material;") {materialFunction=&/MF_SnowPostRVT}
  FunctionOutput(None=NamedRerouteDeclaration(None=@n5.SparkleEmissive) {name=Sparkle-Emissive, nodeColor=<rgba 1 0 1 1>}) {outputName=SnowEmissive}
  n6 = SetMaterialAttributes!(
   MaterialAttributes=@n2
   BaseColor=NamedRerouteUsage()
   Roughness=NamedRerouteUsage()
   Specular=NamedRerouteUsage()
   Normal=NamedRerouteUsage()
   EmissiveColor=NamedRerouteUsage()
  ) {attributeSetTypes [
    #1
    #2
    #3
    #4
    #5
   ]}
  FunctionOutput(None=@n6) {outputName=Base_SnowMaterial}
  FunctionOutput(None=@n3) {outputName=PreRVT_SnowMaterial}
  LandscapeLayerCoords()
  n7 = LinearInterpolate(A=PixelNormalWS(), B=VertexNormalWS(), Alpha=ScalarParameter() {group=SnowProgression, parameterName=FSS-BlendNormalType})
  n8 = DotProduct(A=@n7, B=SkyAtmosphereLightDirection())
  LinearInterpolate(A=DotProduct(A=@n7, B=VectorParameter() {parameterName=FakeVector, defaultValue=<rgba 0 0 1 1>}.RGB), B=@n8, Alpha=NamedRerouteUsage())
  n9 = NamedRerouteUsage()
  MakeMaterialAttributes!(
   BaseColor=@n9
   Metallic
   Specular
   Roughness
   Anisotropy
   EmissiveColor=@n9
   Opacity
   OpacityMask
   Normal=NamedRerouteUsage()
   Tangent
   WorldPositionOffset
   SubsurfaceColor
   ClearCoat
   ClearCoatRoughness
   AmbientOcclusion
   Refraction
   CustomizedUVs_0
   CustomizedUVs_1
   CustomizedUVs_2
   CustomizedUVs_3
   CustomizedUVs_4
   CustomizedUVs_5
   CustomizedUVs_6
   CustomizedUVs_7
   PixelDepthOffset
   ShadingModel
   Displacement
  )
  n10 = ScalarParameter() {defaultValue=400, group=SnowProgression, parameterName=SnowSparkleMaskScale}
  n11 = Divide(A=CameraPositionWS(), B=ScalarParameter() {defaultValue=10000, group=SnowProgression, parameterName=SPK-Div111})
  n12 = TextureSample(
   UVs=Add(
    A=MaterialFunctionCall(Scale=@n10) {materialFunction=&MF_SimpleWorldUV}.Result
    B=ComponentMask(None=@n11) {g=true, r=true}
   )
   Tex=NamedRerouteUsage()
  ) {
   desc=T_Athena_Snow_01_MOBILE_M
   samplerSource=SSM_Wrap_WorldGroupSettings
   samplerType=SAMPLERTYPE_Normal
   texture=&FlatNormalBlue
  }
  n13 = Multiply(A=@n12.R, B=@n12.R)
  n14 = Multiply(A=@n12.R, B=Multiply(A=@n13, B=@n13))
  Multiply(A=@n14, B=ScalarParameter() {defaultValue=1, parameterName=SAPK_modSpecx})
  n15 = OneMinus(None=DotProduct(A=CameraVectorWS(), B=VertexNormalWS()))
  Multiply(A=@n15) {constB=10}
  n16 = NamedRerouteUsage()
  n17 = LinearInterpolate(
   A=ScalarParameter() {defaultValue=1, parameterName=SnowMinRoughness}
   B=ScalarParameter() {defaultValue=1, parameterName=SnowMaxRoughness}
   Alpha=@n16
  )
  n18 = LinearInterpolate(A=ScalarParameter() {defaultValue=-1, parameterName=SparkleRoughness}, B=@n17, Alpha=Max(A=NamedRerouteUsage(), B=@n16))
  n19 = QualitySwitch(Default=@n17, Low=@n17, High=@n18, Medium=@n17, Epic=@n18)
  n20 = NamedRerouteUsage()
  n21 = ScalarParameter() {defaultValue=0.5, group=SnowProgression, parameterName=SnowRoughMobile}
  NamedRerouteDeclaration(None=ShadingPathSwitch(Default=StaticSwitch(True=@n19, False=@n17, Value=@n20), Forward=@n21, Mobile=@n21)) {name=SnowRoughness, nodeColor=<rgba 1 0.0938 0 1>}
  n22 = NamedRerouteDeclaration(None=ScalarParameter() {defaultValue=0.5, parameterName=SnowBaseSpec}) {name=Pre_RVT_Spec, nodeColor=<rgba 0 0.9766 1 1>}
  n23 = LinearInterpolate(A=ScalarParameter() {defaultValue=2, parameterName=SpeckleSpecValue}, B=@n22, Alpha=NamedRerouteUsage())
  NamedRerouteDeclaration(None=ShadingPathSwitch(Default=QualitySwitch(Default=@n22, Low=@n22, High=@n23, Medium=@n22, Epic=@n23), Forward=@n22, Mobile=@n22)) {name=SnowSpecular, nodeColor=<rgba 0 1 0.9766 1>}
  n24 = MaterialFunctionCall!(CustomBiomeMask, Slope) {materialFunction=&/MF_AsteriaCommon_Cliff}
  n25 = Multiply(
   A=Subtract(A=OneMinus(None=@n24.CliffFaceHeightMap), B=ScalarParameter() {defaultValue=0.1, parameterName=CliffCreviceSnowSpread})
   B=ScalarParameter() {defaultValue=2, parameterName=CreviceSnowStrength}
  ) {constB=4}
  NamedRerouteDeclaration(None=Multiply(
    A=MaterialFunctionCall() {materialFunction=&/MF_AltitudeMask}.Result
    B=Clamp(None=@n25)
   )) {name=Cliff_CreviceMask, nodeColor=<rgba 0.3359 1 0 1>}
  NamedRerouteDeclaration(None=@n24.CliffFaceHeightMap) {name=CliffHeightMap, nodeColor=<rgba 1 0 0.2578 1>}
  n26 = TextureSample(
   UVs=MaterialFunctionCall(Scale=Constant() {r=600}) {materialFunction=&MF_SimpleWorldUV}.Result
   Tex=NamedRerouteUsage()
  ) {
   desc=T_Athena_Snow_01_MOBILE_M
   samplerSource=SSM_Wrap_WorldGroupSettings
   samplerType=SAMPLERTYPE_Normal
   texture=&FlatNormalBlue
  }
  NamedRerouteDeclaration(None=@n26.R) {name=SnowDetailMask-R, nodeColor=<rgba 1 0 1 1>}
  n27 = ScalarParameter() {defaultValue=1800, parameterName=NoiseScale1}
  n28 = TextureSample(
   UVs=NamedRerouteDeclaration(None=MaterialFunctionCall(Scale=@n27) {materialFunction=&MF_SimpleWorldUV}.Result) {name=NoiseScale1, nodeColor=<rgba 1 0 1 1>}
   Tex=NamedRerouteUsage()
  ) {
   desc=T_Athena_Snow_01_MOBILE_M
   samplerSource=SSM_Wrap_WorldGroupSettings
   samplerType=SAMPLERTYPE_Normal
   texture=&FlatNormalBlue
  }
  NamedRerouteDeclaration(None=@n28.G) {name=SnowDetailMask-G, nodeColor=<rgba 1 0 1 1>}
  NamedRerouteDeclaration(None=@n5.OutputSnowAlbedo) {name=OutputSnowAlbedo, nodeColor=<rgba 1 0 1 1>}
  n29 = CurveAtlasRowParameter(CurveTime=CollectionParameter() {collection=/Game/TimeOfDay/TODM_Sequence/ParameterCollection/MPC_FN_TimeOfDay, parameterName=NormalizedTimeOfDay}) {
   atlas=/Game/Athena/Apollo/Environments/Landscape/Materials/Biome/UniqueBiomeMaterialFunctions/LandscapeCurveAtlas
   curve=/Game/Athena/Apollo/Environments/Landscape/Materials/Biome/UniqueBiomeMaterialFunctions/LandscapeCurve-TOD
   desc="0=morn 1=day 0=eve 1=night"
  }
  NamedRerouteDeclaration(None=@n29.R) {name=TOD-Lerp, nodeColor=<rgba 1 0 1 1>}
  n30 = SetMaterialAttributes!(MaterialAttributes)
  n31 = GetMaterialAttributes(None=NamedRerouteDeclaration(None=FunctionInput(Preview=@n30) {bUsePreviewValueAsDefault=true, inputName=Material, inputType=FunctionInput_MaterialAttributes, sortPriority=500}) {name=InputMaterial, nodeColor=<rgba 1 0 1 1>}) {attributeGetTypes [
    #1
    #4
    #2
    #3
   ]}
  NamedRerouteDeclaration(None=@n31.Normal) {name=InputMaterial-Normals, nodeColor=<rgba 1 0 1 1>}
  NamedRerouteDeclaration(None=MaterialFunctionCall(Scale=ScalarParameter() {defaultValue=400, parameterName=MicroNoiseScale2}) {materialFunction=&MF_SimpleWorldUV}.Result) {name=MicroNoiseScale2, nodeColor=<rgba 1 0 1 1>}
  n32 = VertexNormalWS()
  n33 = Power(Base=Abs(None=ComponentMask(None=@n32) {b=true})) {constExponent=32}
  NamedRerouteDeclaration(None=FunctionInput(Preview=Clamp(None=@n33)) {bUsePreviewValueAsDefault=true, inputName=LayerMask, inputType=FunctionInput_Scalar}) {name=Snow_LayerMask, nodeColor=<rgba 0 0.5781 1 1>}
  NamedRerouteDeclaration(None=@n31.Roughness) {name=InputMaterial-Roughness, nodeColor=<rgba 1 0 1 1>}
  NamedRerouteDeclaration(None=@n31.Specular) {name=InputMaterial-Spec, nodeColor=<rgba 1 0 1 1>}
  NamedRerouteDeclaration(None=@n31.BaseColor) {name=InputMaterial-Basecolor, nodeColor=<rgba 1 0 1 1>}
  NamedRerouteDeclaration(None=@n31.MaterialAttributes) {name=InputMaterial-Attributes, nodeColor=<rgba 1 0 1 1>}
  n34 = TextureObject() {
   desc="T_Athena_Snow_01_MOBILE_M\r\nUse this texture on mobile/switch or it will cause a blocky mess!\r\nThere was a reason we set it to a 512 uncompressed, but I can't remember why exactly.\r\nIt needs to be kept at this size for default now because the other one becomes pixellated\r\nat this larger scale.\r\nNote: Previously, the blue channel was being used as a general-area snow mask. It\r\nwas divided into 9 sections for 9 different snow amount levels."
   samplerType=SAMPLERTYPE_LinearColor
   texture=/Game/Athena/Environments/Landscape/Textures/T_Athena_Snow_01_MOBILE_M
  }
  NamedRerouteDeclaration(None=@n34) {name=SnowChannelPack, nodeColor=<rgba 1 0 1 1>}
  NamedRerouteDeclaration(None=StaticSwitch(
    True=Constant() {r=1}
    False=TextureObject() {desc=T_SnowMask_Progression, samplerType=SAMPLERTYPE_Normal, texture=&FlatNormalBlue}
    Value=StaticBool() {value=true}
   )) {name=SnowAreaDetail, nodeColor=<rgba 1 0 1 1>}
  NamedRerouteDeclaration(None=MaterialFunctionCall(Scale=ScalarParameter() {defaultValue=800, parameterName=MicroNoiseScale1}) {materialFunction=&MF_SimpleWorldUV}.Result) {name=MicroNoiseScale1, nodeColor=<rgba 1 0 1 1>}
  n35 = FunctionInput(Preview=StaticBool() {value=true}) {bUsePreviewValueAsDefault=true, description="Default = True", inputName=UseSparkle, inputType=FunctionInput_StaticBool, sortPriority=4}
  NamedRerouteDeclaration(None=@n35) {name=UseSparkle, nodeColor=<rgba 1 0 1 1>}
  n36 = MaterialFunctionCall(Scale=ScalarParameter() {defaultValue=1600, parameterName=NoiseScale2}) {materialFunction=&MF_SimpleWorldUV}
  n37 = TextureSample(UVs=@n36.Result, Tex=NamedRerouteUsage()) {
   desc=T_Athena_Snow_01_MOBILE_M
   samplerSource=SSM_Wrap_WorldGroupSettings
   samplerType=SAMPLERTYPE_Normal
   texture=&FlatNormalBlue
  }
  NamedRerouteDeclaration(None=@n37.G) {name=SnowDetailMask-G2, nodeColor=<rgba 1 0 1 1>}
  n38 = NamedRerouteUsage()
  n39 = FunctionInput(Preview=StaticBool()) {
   bUsePreviewValueAsDefault=true
   description="Default = False"
   inputName=IsTangentSpaceInput
   inputType=FunctionInput_StaticBool
   sortPriority=4
  }
  NamedRerouteDeclaration(None=StaticSwitch(True=Transform(None=@n38), False=@n38, Value=NamedRerouteDeclaration(None=@n39) {name=IsTangentSpaceInput, nodeColor=<rgba 1 0 1 1>})) {name=InputNormalsForMask, nodeColor=<rgba 1 0 1 1>}
  NamedRerouteDeclaration(None=FunctionInput(Preview=Constant()) {inputName=InputMaterialHeightmap, inputType=FunctionInput_Scalar}) {name=InputMaterial_HeightMap, nodeColor=<rgba 0 1 0.0156 1>}
  n40 = TextureSample(UVs=@n1.Result, Tex=NamedRerouteUsage()) {
   desc=T_Athena_Snow_01_MOBILE_M
   samplerSource=SSM_Wrap_WorldGroupSettings
   samplerType=SAMPLERTYPE_Normal
   texture=&FlatNormalBlue
  }
  n41 = ComponentMask(None=@n40.RGB) {r=true}
  NamedRerouteDeclaration(None=Power(Base=OneMinus(None=@n41))) {name=HighFrequencyNoise, nodeColor=<rgba 0.1953 0 1 1>}
  NamedRerouteDeclaration(None=TextureObject() {texture=/Game/Environments/Asteria/Terrain/Textures/T_Asteria_WindBlownSnow_01_D}) {name=Macro1NormalHeightMap, nodeColor=<rgba 0 0.0156 1 1>}
  NamedRerouteDeclaration(None=LinearInterpolate(
    A=VectorParameter() {group=SnowProgression, parameterName=SnowShadowColor-Morning, defaultValue=<rgba 0.0784 0.251 0.4 1>}.RGB
    B=VectorParameter() {group=SnowProgression, parameterName=SnowShadowColor-Day, defaultValue=<rgba 0.08 0.2478 0.4 1>}.RGB
    Alpha=NamedRerouteUsage()
   )) {name=ShadowColor, nodeColor=<rgba 1 0 1 1>}
  n42 = LinearInterpolate(A=NamedRerouteUsage(), B=NamedRerouteUsage(), Alpha=Constant() {r=0.5})
  NamedRerouteDeclaration(None=LinearInterpolate(A=SquareRoot(None=@n42), B=Constant() {r=1}, Alpha=Constant() {r=0.25})) {name=AdditionalNoise, nodeColor=<rgba 1 0 1 1>}
  n43 = FunctionInput(Preview=StaticBool()) {
   bUsePreviewValueAsDefault=true
   description="Default = False"
   inputName=UseVertexNormalsForBasicSnow
   inputType=FunctionInput_StaticBool
   sortPriority=4
  }
  n44 = StaticSwitch(
   True=VertexNormalWS()
   False=MaterialFunctionCall!(Normals=NamedRerouteUsage(), TransformToTangentSpace, Normal Strength=Constant() {r=1.5}) {materialFunction=/Game/Athena/Apollo/Environments/Landscape/Materials/Biome/CommonMaterials/MF_HQ_WorldAlignedNormal}.Result
   Value=@n43
  )
  n45 = FunctionInput(Preview=StaticBool() {value=true}) {
   bUsePreviewValueAsDefault=true
   description="Default = True"
   inputName=UseTangentSnowNormals
   inputType=FunctionInput_StaticBool
   sortPriority=4
  }
  NamedRerouteDeclaration(None=StaticSwitch(True=Transform(None=@n44) {transformSourceType=TRANSFORMSOURCE_World, transformType=TRANSFORM_Tangent}, False=@n44, Value=@n45)) {name=SnowNormals, nodeColor=<rgba 1 0 1 1>}
  n46 = TextureObject() {
   desc=T_Athena_Snow_01_MOBILE_N
   samplerType=SAMPLERTYPE_Normal
   texture=/Game/Athena/Environments/Landscape/Textures/T_Athena_Snow_01_MOBILE_N
  }
  NamedRerouteDeclaration(None=@n46) {name=MicroNormalTextureObj, nodeColor=<rgba 1 0 1 1>}
  n47 = TextureObject() {desc=T_Artemis_Snow_N_01, samplerType=SAMPLERTYPE_Normal, texture=/Game/Environments/Artemis/Sets/Terrain/Textures/T_Artemis_Snow_N_01}
  NamedRerouteDeclaration(None=@n47) {name=Macro1NormalTextureObj, nodeColor=<rgba 1 0 1 1>}
  n48 = ComponentMask(None=WorldPosition().XYZ) {g=true, r=true}
  n49 = TextureSample(UVs=Divide(A=@n48, B=ScalarParameter() {defaultValue=20000, parameterName=SnowVarMaskBaseScale})) {
   samplerSource=SSM_Wrap_WorldGroupSettings
   samplerType=SAMPLERTYPE_Grayscale
   texture=/Game/Athena/Apollo/Environments/Landscape/Materials/Biome/CommonMaterials/T_ApolloGrass_ColorMask
  }
  n50 = Multiply(A=ComponentMask(None=@n49.RGB) {r=true}, B=ScalarParameter() {defaultValue=1, parameterName=SnowVarMaskStrength})
  n51 = ScalarParameter() {defaultValue=1, parameterName=Snow-VariationMaskPow}
  NamedRerouteDeclaration(None=Clamp(None=Power(Base=@n50, Exp=@n51))) {name=VariationMask_01, nodeColor=<rgba 0 0.0625 1 1>}
  NamedRerouteDeclaration(None=TextureObject() {samplerType=SAMPLERTYPE_Normal, texture=/Game/Environments/Asteria/Terrain/Textures/T_Asteria_WindBlownSnow_01_N}) {name=Macro2NormalTextureObj, nodeColor=<rgba 1 0 1 1>}
  NamedRerouteDeclaration(None=@n36.Result) {name=NoiseScale2, nodeColor=<rgba 1 0 1 1>}
  n52 = LinearInterpolate(A=NamedRerouteUsage(), B=VertexNormalWS(), Alpha=ScalarParameter() {group=SnowProgression, parameterName=Noise-SlopeMask-UseVertexNormals})
  n53 = Multiply(
   A=Saturate(None=ComponentMask(None=@n52) {b=true})
   B=ScalarParameter() {defaultValue=3, group=SnowProgression, parameterName=SnowHeightMask-Mod}
  )
  n54 = ScalarParameter() {defaultValue=-0.15, group=SnowProgression, parameterName=SnowHeightMask-Add}
  NamedRerouteDeclaration(None=Saturate(None=Add(A=@n53, B=@n54))) {name=SlopeMask, nodeColor=<rgba 1 0 1 1>}
  n55 = Power(Base=Saturate(None=@n8), Exp=ScalarParameter() {defaultValue=7, group=SnowProgression, parameterName=FSS-BlendPow})
  NamedRerouteDeclaration(None=@n55) {name=DebugShadowMask, nodeColor=<rgba 1 0 0.3984 1>}
  n56 = LinearInterpolate(
   A=ScalarParameter() {group=SnowProgression, parameterName=FSS-A-Morning}
   B=ScalarParameter() {group=SnowProgression, parameterName=FSS-A-Day}
   Alpha=NamedRerouteUsage()
  )
  n57 = LinearInterpolate(
   A=ScalarParameter() {defaultValue=1, group=SnowProgression, parameterName=FSS-B-Morning}
   B=ScalarParameter() {defaultValue=1, group=SnowProgression, parameterName=FSS-B-Day}
  )
  NamedRerouteDeclaration(None=Saturate(None=LinearInterpolate(A=@n56, B=@n57, Alpha=@n55))) {name=ShadowMask, nodeColor=<rgba 1 0 1 1>}
  NamedRerouteDeclaration(None=@n5.Sparkle-RoughnessDEPRICATED) {name=Sparkle-Roughness, nodeColor=<rgba 1 0 1 1>}
  n58 = NamedRerouteUsage()
  n59 = NamedRerouteUsage()
  n60 = Multiply(A=@n59, B=ScalarParameter() {defaultValue=0.5, parameterName=HexGridTile})
  n61 = ScalarParameter() {defaultValue=2, parameterName=MaxScale}
  n62 = ScalarParameter() {defaultValue=1, parameterName=MinScale}
  n63 = ScalarParameter() {defaultValue=90, parameterName=MaxRotation}
  n64 = ScalarParameter() {defaultValue=-90, parameterName=MinRotation}
  n65 = MaterialFunctionCall(
   Normal=ComponentMask(None=MaterialFunctionCall!(
     Texture=@n58
     UseMipBias
     MipBias
     LowEndTiling
     LandscapeCoordinates=@n59
     HexGridSize
     HexGridUVInput=@n60
     HexGridGradientSharpness
     MaxScale=@n61
     MinScale=@n62
     MaxOffset
     MinOffset
     "MaxRotation(Degrees)"=@n63
     "MinRotation(Degrees)"=@n64
     forceLowQuality
    ) {materialFunction=/Game/Athena/Asteria/Environments/Landscape/Materials/MaterialFunctions/UtilityFunctions/Landscape_NormalTextureSample}.Result) {b=true, g=true, r=true}
   Flatness=ScalarParameter() {defaultValue=0.5, parameterName=WindsweptFlatness}
  ) {materialFunction=&FlattenNormal}
  NamedRerouteDeclaration(None=MaterialFunctionCall(
    Normal=NamedRerouteDeclaration(None=@n65.Result) {name=MedScaleNormal_02, nodeColor=<rgba 0.7813 0 1 1>}
    Flatness=ScalarParameter() {defaultValue=0.5, parameterName=Med3_Flatten}
   ) {materialFunction=&FlattenNormal}.Result) {name=MedScale_03, nodeColor=<rgba 0 1 0.1797 1>}
  NamedRerouteDeclaration(None=@n5.SparkleMaskDEPRICATED) {name=SparkleMask, nodeColor=<rgba 0 1 0.625 1>}
  NamedRerouteDeclaration(None=LinearInterpolate(
    A=VectorParameter() {parameterName=SnowAlbedo2, defaultValue=<rgba 0.3035 0.4709 0.702 1>}.RGB
    B=VectorParameter() {parameterName=SnowAlbedo, defaultValue=<rgba 0.8 0.8 1 1>}.RGB
    Alpha=NamedRerouteUsage()
   )) {name=AlbedoColor, nodeColor=<rgba 0 0.4844 1 1>}
  NamedRerouteDeclaration(None=@n5.PreRVT_SnowAlbedo) {name=PreRVT_SnowAlbedo, nodeColor=<rgba 0.6641 0 1 1>}
  NamedRerouteDeclaration(None=Constant3Vector() {constant=<rgba 0.26 0.3355 0.4 1>}) {name=OutputSnowAlbedoForRVT, nodeColor=<rgba 1 0 1 1>}
  NamedRerouteDeclaration(None=ShadingPathSwitch(Default=@n17, Forward=@n21, Mobile=@n21)) {name=PreRVT_SnowRoughness, nodeColor=<rgba 1 0.4453 0 1>}
  n66 = NamedRerouteUsage()
  n67 = ScalarParameter() {defaultValue=1, parameterName=HeightScale}
  n68 = Constant()
  n69 = Constant()
  n70 = MaterialFunctionCall(
   Texture=NamedRerouteUsage()
   LandscapeCoordinates=@n66
   HexGridUVInput=Multiply(A=@n66) {constB=0.5}
   MaxScale=@n67
   MinScale=@n67
   MaxOffset=@n68
   MinOffset=@n68
   "MaxRotation(Degrees)"=@n69
   "MinRotation(Degrees)"=@n69
  ) {materialFunction=&Landscape_NormalTextureSample}
  NamedRerouteDeclaration(None=Multiply(A=ComponentMask(None=@n70.Result) {a=true}, B=ScalarParameter() {defaultValue=0.5, parameterName=SnowHeightMapBias})) {name=HeightMap, nodeColor=<rgba 0 0.9062 1 1>}
  NamedRerouteDeclaration(None=ComponentMask(None=@n70.Result) {b=true, g=true, r=true}) {name=TextureDiffuse, nodeColor=<rgba 1 0.3984 0 1>}
  n71 = TextureSample(UVs=Divide(A=@n48, B=ScalarParameter() {defaultValue=30000, parameterName=SnowMicroVariationMaskScale2})) {
   samplerSource=SSM_Wrap_WorldGroupSettings
   samplerType=SAMPLERTYPE_Grayscale
   texture=&T_ApolloGrass_ColorMask
  }
  n72 = Multiply(A=ComponentMask(None=@n71.RGB) {g=true}, B=ScalarParameter() {defaultValue=3, parameterName=SnowMicroVarMaskStrength})
  n73 = ScalarParameter() {defaultValue=1, parameterName=Snow-MicroVariationMaskPow}
  NamedRerouteDeclaration(None=Clamp(None=Power(Base=@n72, Exp=@n73))) {name=VariationMask_02, nodeColor=<rgba 0.125 0 1 1>}
  n74 = MaterialFunctionCall(
   Texture=NamedRerouteUsage()
   LandscapeCoordinates=@n59
   HexGridUVInput=@n60
   MaxScale=@n61
   MinScale=@n62
   "MaxRotation(Degrees)"=@n63
   "MinRotation(Degrees)"=@n64
  ) {materialFunction=&Landscape_NormalTextureSample}
  NamedRerouteDeclaration(None=Multiply(A=ComponentMask(None=@n74.Result) {b=true, g=true, r=true}, B=ScalarParameter() {defaultValue=1, parameterName=MedNormal1_Power})) {name=MedScaleNormal_01, nodeColor=<rgba 1 0.0075 0 1>}
  n75 = NamedRerouteUsage()
  n76 = NamedRerouteUsage()
  n77 = NamedRerouteUsage()
  NamedRerouteDeclaration(None=LinearInterpolate(A=LinearInterpolate(A=@n75, B=@n76, Alpha=@n77), B=NamedRerouteUsage(), Alpha=NamedRerouteUsage())) {name=NormalResultNew, nodeColor=<rgba 0 1 0.1094 1>}
  NamedRerouteUsage()
  NamedRerouteUsage()
  NamedRerouteUsage()
  NamedRerouteUsage()
  NamedRerouteUsage()
  NamedRerouteUsage()
  NamedRerouteUsage()
  NamedRerouteUsage()
  NamedRerouteUsage()
  NamedRerouteUsage()
  NamedRerouteUsage()
  NamedRerouteUsage()
  NamedRerouteUsage()
  NamedRerouteUsage()
  NamedRerouteUsage()
  NamedRerouteUsage()
  Power(Base=@n12.R, Exp=Constant() {r=5})
  Power(Base=@n15)
  Saturate(None=@n29.R)
  Saturate(None=Multiply(A=@n14, B=ScalarParameter() {defaultValue=0.5, group=SnowProgression, parameterName=SPK-modroughx}))
  ScalarParameter() {defaultValue=0.7, group=SnowProgression, parameterName=SnowAlbedo}
  ScalarParameter() {defaultValue=600, group=SnowProgression, parameterName=SnowRoughScale}
  ScalarParameter() {defaultValue=3000, desc="Shared!", group=SnowProgression, parameterName=SnowNoiseScale2}
  ScalarParameter() {defaultValue=1200, group=SnowProgression, parameterName=SnowNoiseScale}
  ScalarParameter() {defaultValue=2, parameterName=Snow-VariationMask_01_Tiling}
  ScalarParameter() {defaultValue=0.25, group=SnowProgression, parameterName=DGDG-AddNoise-Soft}
  ScalarParameter() {defaultValue=5, group=SnowProgression, parameterName=SPK-FresPow2}
  ScalarParameter() {defaultValue=0.35, group=SnowProgression, parameterName=SnowAlbedo-Mobile}
  ShadingModel() {shadingModel=MSM_Subsurface}
  ShadingPathSwitch(
   Default=QualitySwitch(Default=Reroute(), Low=Reroute(), High=Reroute(), Medium=Reroute(), Epic=Reroute())
   Mobile=QualitySwitch(Default=Reroute(), Low=Reroute(), Medium=Reroute())
  )
  n78 = Constant()
  n79 = OneMinus(None=NamedRerouteUsage())
  n80 = Multiply(A=@n12.R, B=ScalarParameter() {defaultValue=0.55, parameterName=Sparkle_Emissive_Pow}) {constB=0.75}
  n81 = Multiply(
   A=LinearInterpolate(
    A=ScalarParameter() {defaultValue=100, group=SnowProgression, parameterName=SPK-mod-morn}
    B=ScalarParameter() {defaultValue=100, group=SnowProgression, parameterName=SPK-mod-day}
    Alpha=NamedRerouteUsage()
   )
   B=Saturate(None=Subtract(A=@n79, B=@n80))
  )
  n82 = Constant()
  ShadingPathSwitch(Default=QualitySwitch(Default=@n78, Low=@n78, High=@n81, Medium=@n78, Epic=@n81), Forward=@n82, Mobile=@n82)
  n83 = LinearInterpolate(A=NamedRerouteUsage(), B=NamedRerouteUsage(), Alpha=NamedRerouteUsage())
  n84 = NamedRerouteUsage()
  n85 = LinearInterpolate(
   A=ScalarParameter() {defaultValue=1.25, parameterName=FresnelDarken}
   B=ScalarParameter() {defaultValue=2, parameterName=FresnelDarken2}
   Alpha=@n84
  )
  n86 = LinearInterpolate(
   A=ScalarParameter() {defaultValue=0.75, parameterName=FresnelBrighten}
   B=ScalarParameter() {defaultValue=1, parameterName=FresnelBrighten2}
   Alpha=@n84
  )
  n87 = DotProduct(A=CameraVectorWS(), B=LinearInterpolate(A=PixelNormalWS(), B=VertexNormalWS(), Alpha=ScalarParameter() {parameterName=BlendPixelNormal}))
  n88 = Power(
   Base=Multiply(A=OneMinus(None=@n87), B=ScalarParameter() {defaultValue=3, parameterName=SnowFresnelMult})
   Exp=ScalarParameter() {defaultValue=2, parameterName=SnowFresnelPower}
  )
  n89 = Constant() {r=0.35}
  ShadingPathSwitch(Default=LinearInterpolate(A=Multiply(A=@n83, B=@n85), B=Multiply(A=@n83, B=@n86), Alpha=Clamp(None=@n88)), Forward=@n89, Mobile=@n89)
  n90 = VertexNormalWS()
  SphereMask(
   A=Abs(None=ComponentMask(None=@n90) {b=true})
   B=Constant() {r=1}
   Radius=ScalarParameter() {defaultValue=0.1, parameterName=FlatSnowSampleDistance}
   Hardness=ScalarParameter() {defaultValue=0.5, parameterName=FlatSnowSampleHardness}
  )
  TextureObject() {
   desc="Set this to the new, static snow texture mask"
   samplerType=SAMPLERTYPE_LinearColor
   texture=/Game/Athena/Apollo/Environments/Landscape/Materials/Biome/CommonMaterials/T_SnowMask_Static
  }
  TextureObject() {samplerType=SAMPLERTYPE_Normal, texture=/Game/Environments/Asteria/Terrain/Textures/T_Asteria_WindBlownSnow_02_N}
  TextureObject() {desc=T_Artemis_Snow_N_03, samplerType=SAMPLERTYPE_Normal, texture=/Game/Environments/Artemis/Sets/Terrain/Textures/T_Artemis_Snow_N_03}
  TextureSample() {
   desc=FlatNormalBlue
   samplerSource=SSM_Wrap_WorldGroupSettings
   samplerType=SAMPLERTYPE_Normal
   texture=&FlatNormalBlue
  }
  VertexNormalWS()
 ]}
