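# Material function for a generic landscape biome layer in Asteria, returning material attributes plus a Height output from the diffuse alpha.
# Diffuse, S mask and normal inputs go through shared landscape sampling functions using scaled coordinates and hex-grid tiling settings with scale, offset and rotation ranges.
# The S mask's red channel drives both Specular and Roughness.
# Only normal sampling takes mip bias; the UseHexGridonLow input is unconnected.
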MaterialFunction /Game/Athena/Asteria/Environments/Landscape/Materials/MaterialFunctions/MF_Biome_Generic_v2
graph {nodes [
  FunctionInput(Preview=StaticBool()) {bUsePreviewValueAsDefault=true, inputName=UseHexGridonLow, inputType=FunctionInput_StaticBool}
  n1 = Multiply(A=NamedRerouteUsage(), B=NamedRerouteUsage())
  n2 = FunctionInput(Preview=Constant() {r=1}) {bUsePreviewValueAsDefault=true, inputName=LowEndTiling, inputType=FunctionInput_Scalar}
  n3 = MaterialFunctionCall!(
   Texture=FunctionInput(Preview=TextureObject() {texture=/Engine/EngineMaterials/DefaultDiffuse}) {inputName=Diffuse, inputType=FunctionInput_Texture2D, sortPriority=1}
   LandscapeCoordinates=@n1
   LowEndTiling=@n2
   HexGridGradientSharpness=NamedRerouteUsage()
   HexGridSize=NamedRerouteUsage()
   HexGridUVInput
   MinScale=NamedRerouteUsage()
   MaxScale=NamedRerouteUsage()
   MinOffset=NamedRerouteUsage()
   MaxOffset=NamedRerouteUsage()
   "MaxRotation(Degrees)"=NamedRerouteUsage()
   "MinRotation(Degrees)"=NamedRerouteUsage()
   forceLowQuality=NamedRerouteUsage()
  ) {materialFunction=/Game/Athena/Asteria/Environments/Landscape/Materials/MaterialFunctions/UtilityFunctions/Landscape_TextureSample}
  n4 = TextureObject() {samplerType=SAMPLERTYPE_Masks, texture=&T_Fortnite_Default_S}
  n5 = NamedRerouteUsage()
  n6 = NamedRerouteUsage()
  n7 = MaterialFunctionCall(Float3=MaterialFunctionCall(
    Texture=FunctionInput(Preview=@n4) {inputName=S_Mask, inputType=FunctionInput_Texture2D, sortPriority=3}
    LandscapeCoordinates=Multiply(A=@n5, B=@n6)
    LowEndTiling=@n2
    HexGridGradientSharpness=NamedRerouteUsage()
    HexGridSize=NamedRerouteUsage()
    MinScale=NamedRerouteUsage()
    MaxScale=NamedRerouteUsage()
    MinOffset=NamedRerouteUsage()
    MaxOffset=NamedRerouteUsage()
    "MaxRotation(Degrees)"=NamedRerouteUsage()
    "MinRotation(Degrees)"=NamedRerouteUsage()
    forceLowQuality=NamedRerouteUsage()
   ) {materialFunction=&Landscape_TextureSample}.Result) {materialFunction=&BreakOutFloat3Components}
  n8 = TextureObject() {samplerType=SAMPLERTYPE_Normal, texture=&FlatNormalBlue}
  n9 = NamedRerouteUsage()
  n10 = NamedRerouteUsage()
  n11 = SetMaterialAttributes!(
   MaterialAttributes
   BaseColor=@n3.Result
   Specular=@n7.R
   Roughness=@n7.R
   Normal=MaterialFunctionCall!(
    Texture=FunctionInput(Preview=@n8) {inputName=Normal, inputType=FunctionInput_Texture2D, sortPriority=2}
    UseMipBias=NamedRerouteUsage()
    MipBias=NamedRerouteUsage()
    LowEndTiling
    LandscapeCoordinates=Multiply(A=@n9, B=@n10)
    HexGridSize=NamedRerouteUsage()
    HexGridUVInput
    HexGridGradientSharpness=NamedRerouteUsage()
    MaxScale=NamedRerouteUsage()
    MinScale=NamedRerouteUsage()
    MaxOffset=NamedRerouteUsage()
    MinOffset=NamedRerouteUsage()
    "MaxRotation(Degrees)"=NamedRerouteUsage()
    "MinRotation(Degrees)"=NamedRerouteUsage()
    forceLowQuality=NamedRerouteUsage()
   ) {materialFunction=/Game/Athena/Asteria/Environments/Landscape/Materials/MaterialFunctions/UtilityFunctions/Landscape_NormalTextureSample}.Result
  ) {attributeSetTypes [
    #1
    #2
    #3
    #4
   ]}
  FunctionOutput(None=@n11)
  FunctionOutput(None=ComponentMask(None=@n3.Result) {a=true}) {outputName=Height, sortPriority=1}
  LandscapeLayerCoords()
  LandscapeLayerCoords()
  LandscapeLayerCoords()
  Multiply(A=@n1)
  NamedRerouteDeclaration(None=FunctionInput(Preview=ScalarParameter() {defaultValue=1, parameterName=HexGridSize}) {bUsePreviewValueAsDefault=true, inputName=HexGridSize, inputType=FunctionInput_Scalar, sortPriority=3}) {name=HexGridSize, nodeColor=<rgba 1 0.5391 0 1>}
  NamedRerouteDeclaration(None=FunctionInput(Preview=ScalarParameter() {defaultValue=2, parameterName=HexGridSharpness}) {bUsePreviewValueAsDefault=true, inputName=HexGridGradientShaprmness, inputType=FunctionInput_Scalar, sortPriority=4}) {name=HexGridGradientSharpness, nodeColor=<rgba 0 0.7422 1 1>}
  NamedRerouteDeclaration(None=FunctionInput(Preview=Constant() {r=1}) {bUsePreviewValueAsDefault=true, inputName=MacroScale, inputType=FunctionInput_Scalar}) {name=MacroScale, nodeColor=<rgba 0 1 0.1328 1>}
  NamedRerouteDeclaration(None=FunctionInput(Preview=Constant() {r=3}) {bUsePreviewValueAsDefault=true, inputName=MipBiasValue, inputType=FunctionInput_Scalar, sortPriority=10}) {name=MipBiasValue, nodeColor=<rgba 0.6641 1 0 1>}
  NamedRerouteDeclaration(None=FunctionInput(Preview=LandscapeLayerCoords()) {bUsePreviewValueAsDefault=true, inputName=BaseCoordinates, inputType=FunctionInput_Vector2}) {name=BaseCoordinates, nodeColor=<rgba 1 0.3047 0 1>}
  NamedRerouteDeclaration(None=FunctionInput(Preview=ScalarParameter() {defaultValue=4, parameterName=MinScale}) {bUsePreviewValueAsDefault=true, inputName=MinScale, inputType=FunctionInput_Scalar, sortPriority=1}) {name=MinScale, nodeColor=<rgba 1 0.8906 0 1>}
  NamedRerouteDeclaration(None=FunctionInput(Preview=ScalarParameter() {defaultValue=8, parameterName=MaxScale}) {bUsePreviewValueAsDefault=true, inputName=MaxScale, inputType=FunctionInput_Scalar, sortPriority=2}) {name=MaxScale, nodeColor=<rgba 1 0.8906 0 1>}
  NamedRerouteDeclaration(None=FunctionInput(Preview=Constant() {r=-1}) {bUsePreviewValueAsDefault=true, inputName=MinOffset, inputType=FunctionInput_Scalar, sortPriority=5}) {name=MinOffset, nodeColor=<rgba 0.7812 1 0 1>}
  NamedRerouteDeclaration(None=FunctionInput(Preview=Constant() {r=1}) {bUsePreviewValueAsDefault=true, inputName=MaxOffset, inputType=FunctionInput_Scalar, sortPriority=6}) {name=MaxOffset, nodeColor=<rgba 0.7812 1 0 1>}
  NamedRerouteDeclaration(None=FunctionInput(Preview=StaticBool()) {bUsePreviewValueAsDefault=true, inputName=UseRegularTiling, inputType=FunctionInput_StaticBool, sortPriority=9}) {name=ForceLow, nodeColor=<rgba 0 1 0.3203 1>}
  NamedRerouteDeclaration(None=FunctionInput(Preview=Constant()) {bUsePreviewValueAsDefault=true, inputName=MinRotation, inputType=FunctionInput_Scalar, sortPriority=7}) {name=MinRotation, nodeColor=<rgba 0.5937 1 0 1>}
  NamedRerouteDeclaration(None=FunctionInput(Preview=StaticBool()) {bUsePreviewValueAsDefault=true, inputName=UseMipBias, inputType=FunctionInput_StaticBool, sortPriority=11}) {name=useMipBias, nodeColor=<rgba 0 1 0.9531 1>}
  NamedRerouteDeclaration(None=FunctionInput(Preview=Constant()) {bUsePreviewValueAsDefault=true, inputName=MaxRotation, inputType=FunctionInput_Scalar, sortPriority=8}) {name=MaxRotation, nodeColor=<rgba 0.5938 1 0 1>}
 ]}
