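# Material function for the Asteria landscape that outputs grass, snow, cliff, ice, gravel, mud, sand, jungle floor and asphalt layers, most with heights.
# MF_AsteriaTerrainMasks supplies biome masks to the per-layer functions.
# Virtual texture feature switches pick full functions or mobile variants for most layers.
# Mobile mud modifies mobile gravel with constant height, while jungle floor always uses the generic biome function with hex-grid or regular tiling options.
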MaterialFunction /Game/Athena/Asteria/Environments/Landscape/Materials/MaterialFunctions/MF_AsteriaLandscapeLayers
graph {nodes [
  n1 = MaterialFunctionCall!(UseLandscapeCoords) {materialFunction=&/MF_AsteriaTerrainMasks}
  n2 = MaterialFunctionCall!(BiomeMask-RGBA=@n1.BiomeMaskRGBA, Grass Localised Tint Mask, CombinedLandscapeLayersMask, BiomeColors-UseDefault) {materialFunction=&/MF_Asteria_Common_Grass}
  FunctionOutput(None=VirtualTextureFeatureSwitch!(
    No=MaterialFunctionCall!(BiomeMask-RGBA=@n1.BiomeMaskRGBA, CombinedLandscapeLayersMask, BypassUseRVT, BypassBiomeColors) {materialFunction=/Game/Athena/Asteria/Environments/Landscape/Materials/MaterialFunctions/MobileLayers/MF_Mobile_Asteria_Grass}.ForStaticMesh
    Yes=@n2.Default-LandscapeMaterial
   )) {outputName=CommonGrass}
  n3 = MaterialFunctionCall() {materialFunction=/Game/Athena/Asteria/Environments/Landscape/Materials/MaterialFunctions/MobileLayers/MF_Mobile_Asteria_Snow}
  n4 = MaterialFunctionCall!(LayerMask, InputMaterialHeightmap=Constant(), UseVertexNormalsForBasicSnow, UseSparkle, UseTangentSnowNormals, IsTangentSpaceInput, Material) {materialFunction=&/MF_Asteria_Snow}
  FunctionOutput(None=VirtualTextureFeatureSwitch(No=@n3.Result, Yes=@n4.PreRVT_SnowMaterial)) {outputName=Snow, sortPriority=1}
  n5 = MaterialFunctionCall!(
   LowEndTiling=ScalarParameter() {defaultValue=0.15, parameterName=JungleLowEndTiling}
   UseHexGridonLow
   BaseCoordinates=MaterialFunctionCall!(WorldPosition) {materialFunction=&/MF_Asteria_Terrain_Coords}.Adjustable
   MacroScale=ScalarParameter() {defaultValue=1750, parameterName=JungleFloor_Scale}
   MinScale
   Diffuse=TextureObject() {texture=/Game/Environments/Asteria/Terrain/Textures/T_Asteria_JungleFloor_D}
   MaxScale
   Normal=TextureObject() {samplerType=SAMPLERTYPE_Normal, texture=/Game/Environments/Asteria/Terrain/Textures/T_Asteria_JungleFloor_N}
   S_Mask=TextureObject() {samplerType=SAMPLERTYPE_Masks, texture=/Game/Environments/Asteria/Terrain/Textures/T_Asteria_JungleFloor_S}
   HexGridSize
   HexGridGradientShaprmness
   MinOffset
   MaxOffset
   MinRotation
   MaxRotation
   UseRegularTiling=StaticBoolParameter() {parameterName=UseRegularTiling}
   MipBiasValue
   UseMipBias
  ) {materialFunction=&/MF_Biome_Generic_v2}
  FunctionOutput(None=@n5.Result) {outputName=JungleFloor, sortPriority=7}
  n6 = NamedRerouteUsage()
  n7 = MaterialFunctionCall!(CustomSeaLevel, BiomeMask-RGBA=@n6, BypassBiomeColors) {materialFunction=/Game/Athena/Asteria/Environments/Landscape/Materials/MaterialFunctions/MobileLayers/MF_Mobile_Asteria_Sand}
  n8 = MaterialFunctionCall!(BiomeMask-RGBA=@n6, BiomeColors-UseDefault, UseUnderOceanSandMaterial) {materialFunction=&/MF_Asteria_Sand}
  FunctionOutput(None=VirtualTextureFeatureSwitch(No=@n7.Height, Yes=@n8.Height)) {outputName=Sand_H, sortPriority=6}
  n9 = NamedRerouteUsage()
  n10 = MaterialFunctionCall!(Spec, Diffuse, Height, BiomeColors-UseDefault, Rough, Normals, VarianceMask, BiomeMask-RGBA=@n9, BiomeRoughMod) {materialFunction=&/MF_Asteria_Mud}
  FunctionOutput(None=VirtualTextureFeatureSwitch(No=Constant() {r=0.5}, Yes=@n10.Height)) {outputName=Mud_H, sortPriority=5}
  n11 = NamedRerouteUsage()
  n12 = MaterialFunctionCall!(OverrideEdgeColor, BiomeMask-RGBA=@n11, EdgeMask, BypassBiomeColors) {materialFunction=/Game/Athena/Asteria/Environments/Landscape/Materials/MaterialFunctions/MobileLayers/MF_Mobile_Asteria_Gravel}
  n13 = MaterialFunctionCall!(LayerWeight=Constant() {r=1}, VarianceMask, BiomeMask-RGBA=@n11, BiomeColors-UseDefault) {materialFunction=&/MF_Asteria_Gravel}
  FunctionOutput(None=VirtualTextureFeatureSwitch(No=@n12.Height, Yes=@n13.Height)) {outputName=Gravel_H, sortPriority=4}
  n14 = MaterialFunctionCall!(MudMaterial-Input) {materialFunction=/Game/Athena/Asteria/Environments/Landscape/Materials/MaterialFunctions/MobileLayers/MF_Mobile_Asteria_Ice}
  n15 = MaterialFunctionCall!(
   LayerMask
   InputHeightMap
   inputMaterial=MakeMaterialAttributes!(
    BaseColor
    Metallic
    Specular
    Roughness
    Anisotropy
    EmissiveColor
    Opacity
    OpacityMask
    Normal
    Tangent
    WorldPositionOffset
    SubsurfaceColor
    ClearCoat
    ClearCoatRoughness
    AmbientOcclusion
    Refraction
    CustomizedUVs_0
    CustomizedUVs_1
    CustomizedUVs_2
    CustomizedUVs_3
    CustomizedUVs_4
    CustomizedUVs_5
    CustomizedUVs_6
    CustomizedUVs_7
    PixelDepthOffset
    ShadingModel
    Displacement
   )
  ) {materialFunction=&/MF_Asteria_Ice}
  FunctionOutput(None=VirtualTextureFeatureSwitch(No=@n14.Height, Yes=@n15.Heightmap)) {outputName=Ice_H, sortPriority=3}
  n16 = MaterialFunctionCall!(BiomeMask-RGBA=NamedRerouteUsage(), UseParameters) {materialFunction=/Game/Athena/Asteria/Environments/Landscape/Materials/MaterialFunctions/MobileLayers/MF_Mobile_Asteria_Cliff}
  n17 = MaterialFunctionCall!(CustomBiomeMask, Slope) {materialFunction=&/MF_AsteriaCommon_Cliff}
  FunctionOutput(None=VirtualTextureFeatureSwitch(No=@n16.Height, Yes=@n17.Height)) {outputName=CommonCliff_H, sortPriority=2}
  FunctionOutput(None=@n5.Height) {outputName=JungleFloor_H, sortPriority=7}
  FunctionOutput(None=VirtualTextureFeatureSwitch(No=@n16.LandscapeMaterial, Yes=@n17.Material)) {outputName=CommonCliff, sortPriority=2}
  FunctionOutput(None=VirtualTextureFeatureSwitch(No=@n14.Result, Yes=@n15.Result)) {outputName=Ice, sortPriority=3}
  FunctionOutput(None=VirtualTextureFeatureSwitch(No=@n12.ForStaticMesh, Yes=@n13.Material)) {outputName=Gravel, sortPriority=4}
  FunctionOutput(None=VirtualTextureFeatureSwitch(
    No=MaterialFunctionCall!(
     BiomeMask-RGBA=@n9
     GravelMaterialInput=@n12.Material
     Heightmap=@n12.Material
     UseUnmodifiedGravelDiffuse
     UnmodifiedGravelDiffuse
     BypassBiomeColors
    ) {materialFunction=/Game/Athena/Asteria/Environments/Landscape/Materials/MaterialFunctions/MobileLayers/MF_Asteria_DirtMudModifier}.MudMaterial
    Yes=@n10.Material
   )) {outputName=Mud, sortPriority=5}
  FunctionOutput(None=VirtualTextureFeatureSwitch(No=@n7.ForStaticmesh, Yes=@n8.Material)) {outputName=Sand, sortPriority=6}
  FunctionOutput(None=VirtualTextureFeatureSwitch(
    No=MaterialFunctionCall!(TilingSize) {materialFunction=/Game/Athena/Asteria/Environments/Landscape/Materials/MaterialFunctions/MobileLayers/MF_Mobile_Asteria_Asphault}.ForStaticmesh
    Yes=MaterialFunctionCall!(TilingSize) {materialFunction=&/MF_Asteria_Asphault}.ForStaticmesh
   )) {outputName=Asphault, sortPriority=8}
  FunctionOutput(None=@n2.Mobile-Height) {outputName=Grass H}
  FunctionOutput(None=VirtualTextureFeatureSwitch(No=@n3.Height, Yes=@n4.HeightMap)) {outputName=Snow H, sortPriority=1}
  NamedRerouteDeclaration(None=@n1.BiomeMaskRGBA) {name=BiomeMask-RGBA, nodeColor=<rgba 1 0 0.4922 1>}
 ]}
