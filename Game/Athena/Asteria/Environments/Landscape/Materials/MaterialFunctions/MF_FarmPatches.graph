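# Material function that tints Asteria landscape material with farm patch colors and outputs material attributes plus a Height value.
# The painted BiomeC landscape layer weight serves as UVs into the FarmColorizeGrid texture, and its ceiling forms a hard patch mask.
# The brightness-scaled, saturated color blends into the incoming BaseColor; the mobile shading path bypasses this blend.
# A detail-color multiply branch exists but feeds no output.
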MaterialFunction /Game/Athena/Asteria/Environments/Landscape/Materials/MaterialFunctions/MF_FarmPatches
graph {nodes [
  n1 = NamedRerouteUsage()
  n2 = LandscapeLayerSample() {desc=ForestFloor, parameterName=BiomeC}
  n3 = TextureSampleParameter2D(UVs=@n2) {
   group=FarmPatches
   parameterName=FarmColorizeGrid
   samplerSource=SSM_Wrap_WorldGroupSettings
   texture=/Game/Athena/Asteria/Environments/Landscape/Materials/Textures/T_AsteriaFarm_Patches
  }
  n4 = Multiply(A=@n3.RGB, B=ScalarParameter() {defaultValue=1, parameterName=ColorBrightness})
  n5 = Multiply(A=NamedRerouteUsage(), B=@n4) {desc="Multiply is being used because dividing\r\nit would require numbers greater than 1\r\nfrom the texture"}
  Add(A=Multiply(A=Constant() {r=0.01}, B=@n1), B=Multiply(A=@n1, B=Saturate(None=@n5) {desc=new}))
  n6 = GetMaterialAttributes(None=FunctionInput() {inputName=Material, inputType=FunctionInput_MaterialAttributes}) {attributeGetTypes [#1]}
  n7 = SetMaterialAttributes!(MaterialAttributes=@n6.MaterialAttributes, BaseColor=LinearInterpolate(A=@n6.BaseColor, B=Saturate(None=@n4), Alpha=NamedRerouteUsage())) {attributeSetTypes [#1]}
  n8 = NamedRerouteUsage()
  FunctionOutput(None=ShadingPathSwitch(Default=QualitySwitch(Default=@n7), Mobile=QualitySwitch(Default=@n8, Low=@n8, Medium=@n8)))
  FunctionOutput(None=NamedRerouteUsage()) {outputName=Height}
  NamedRerouteDeclaration(None=FunctionInput(Preview=Constant() {r=1}) {bUsePreviewValueAsDefault=true, inputName=Grassmask, inputType=FunctionInput_Scalar}) {name=Grass_Mask, nodeColor=<rgba 0.7109 0 1 1>}
  NamedRerouteDeclaration(None=FunctionInput(Preview=Constant()) {bUsePreviewValueAsDefault=true, inputName=InputMaterialHeight}) {name=InputHeight, nodeColor=<rgba 0 0.5547 1 1>}
  NamedRerouteDeclaration(None=@n6.MaterialAttributes) {name=InputMaterial, nodeColor=<rgba 0 1 0.9531 1>}
  NamedRerouteDeclaration(None=Ceil(None=@n2)) {name=PatchSphereMask, nodeColor=<rgba 0 0.6484 1 1>}
  n9 = Constant()
  NamedRerouteDeclaration(None=FunctionInput(Preview=@n9) {bUsePreviewValueAsDefault=true, inputName=Gravel-Height, inputType=FunctionInput_Scalar}) {name=Gravel-Height, nodeColor=<rgba 1 0.2578 0 1>}
  NamedRerouteDeclaration(None=FunctionInput(Preview=@n9) {bUsePreviewValueAsDefault=true, inputName=CommonLayers-Height, inputType=FunctionInput_Scalar}) {name=CommonLayers-Height, nodeColor=<rgba 0 0.8125 1 1>}
  NamedRerouteDeclaration(None=FunctionInput() {bUsePreviewValueAsDefault=true, inputName=GrassColorInput}) {name=BaseColorDetail, nodeColor=<rgba 0 1 0.4141 1>}
  NamedRerouteDeclaration(None=FunctionInput() {bUsePreviewValueAsDefault=true, inputName=DetailMasking}) {name=DetailMasking, nodeColor=<rgba 1 0 0.5859 1>}
  NamedRerouteUsage()
 ]}
