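# Material function that layers animated, glowing Fortanium veins onto incoming landscape-environment material attributes, using a supplied vein mask.
# Vein color comes from a curve atlas indexed by a view-angle term with a wrapped offset, over darkened base color.
# World-aligned cloud and rock-detail textures pan over time to supply noise, and emissive adds to the input.
# Shading-path switches pass the input through on mobile and provide cheaper color branches, unless a static switch bypasses them.
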MaterialFunction /Game/Athena/Asteria/Environments/Landscape/Materials/MaterialFunctions/MF_FortaniumVeins
graph {nodes [
  n1 = GetMaterialAttributes(None=FunctionInput(Preview=MakeMaterialAttributes!(
     BaseColor
     Metallic
     Specular
     Roughness
     Anisotropy
     EmissiveColor
     Opacity
     OpacityMask
     Normal
     Tangent
     WorldPositionOffset
     SubsurfaceColor
     ClearCoat
     ClearCoatRoughness
     AmbientOcclusion
     Refraction
     CustomizedUVs_0
     CustomizedUVs_1
     CustomizedUVs_2
     CustomizedUVs_3
     CustomizedUVs_4
     CustomizedUVs_5
     CustomizedUVs_6
     CustomizedUVs_7
     PixelDepthOffset
     ShadingModel
     Displacement
    )) {bUsePreviewValueAsDefault=true, inputName=Material, inputType=FunctionInput_MaterialAttributes}) {attributeGetTypes [#1, #2]}
  n2 = Multiply(
   A=NamedRerouteDeclaration(None=@n1.BaseColor) {name=Input_BaseColor, nodeColor=<rgba 0.9688 1 0 1>}
   B=ScalarParameter() {defaultValue=0.2, parameterName=BaseColorEdgeDarken}
  ) {constB=0.5}
  n3 = NamedRerouteUsage()
  n4 = FunctionInput(Preview=Constant() {r=1}) {inputName=VeinMask, inputType=FunctionInput_Scalar}
  n5 = ScalarParameter() {defaultValue=1.2, parameterName=VeinMaskStrength}
  n6 = NamedRerouteDeclaration(None=Power(Base=Multiply(A=@n4, B=@n5), Exp=ScalarParameter() {defaultValue=2, parameterName=VeinMaskPow})) {name=VeinMask, nodeColor=<rgba 0 1 0.6953 1>}
  n7 = SetMaterialAttributes!(
   MaterialAttributes=@n1.MaterialAttributes
   BaseColor=NamedRerouteDeclaration(None=LinearInterpolate(A=@n2, B=@n3, Alpha=@n6)) {name=BaseColor_result, nodeColor=<rgba 1 0 0.7031 1>}
   EmissiveColor=Add(A=NamedRerouteUsage(), B=NamedRerouteUsage())
   Specular=Constant() {r=1}
   Roughness=Constant() {r=1}
  ) {attributeSetTypes [
    #1
    #2
    #3
    #4
   ]}
  n8 = QualitySwitch(Default=@n7)
  n9 = QualitySwitch(Default=@n1.MaterialAttributes)
  FunctionOutput(None=StaticSwitchParameter(True=@n7, False=ShadingPathSwitch(Default=@n8, Mobile=@n9)) {parameterName=UseBypassQualitySwitch})
  Multiply(A=Time())
  n10 = PixelNormalWS()
  n11 = VertexNormalWS()
  n12 = ScalarParameter() {parameterName=FT_BlendVertexNormal}
  n13 = NamedRerouteDeclaration(None=DotProduct(A=CameraVectorWS(), B=LinearInterpolate(A=@n10, B=@n11, Alpha=@n12))) {name=Fresnel, nodeColor=<rgba 0.5234 0 1 1>}
  n14 = Frac(None=Add(A=ConstantBiasScale(None=@n13), B=NamedRerouteUsage()))
  n15 = CurveAtlasRowParameter(CurveTime=Clamp(None=Power(Base=@n14))) {
   atlas=/Game/Athena/Asteria/Environments/Landscape/Materials/Textures/LS_FortaniumCurveAtlas
   curve=/Game/Athena/Asteria/Environments/Landscape/Materials/Textures/LS_FortaniumColor_01
   parameterName=FortaniumCurve
  }
  n16 = QualitySwitch(Default=@n15)
  n17 = VectorParameter() {parameterName=LowEndVeinsColor, defaultValue=<rgba 0.008 0.4852 0.5029 1>}
  n18 = QualitySwitch(Default=@n17.RGB, Medium=@n17.RGB)
  NamedRerouteDeclaration(None=StaticSwitchParameter(True=@n16, False=ShadingPathSwitch(Default=@n16, Mobile=@n18)) {parameterName=UseBypassQualitySwitch}) {name=BaseColor, nodeColor=<rgba 1 0 0.1875 1>}
  n19 = FunctionInput(Preview=WorldPosition().XYZ) {bUsePreviewValueAsDefault=true, inputName=Position, sortPriority=10}
  n20 = Multiply(A=Time(), B=VectorParameter() {parameterName=NoisePanSpeed, defaultValue=<rgba 6 -6 -6 1>}.RGB)
  n21 = MaterialFunctionCall(
   TextureObject=TextureObject() {texture=/Game/Global/Textures/Default/Noise/T_Noise_Cloud}
   TextureSize=ScalarParameter() {defaultValue=256, group=Detail Texture, parameterName=DetailTile_03}
   WorldPosition=Add(A=@n19, B=@n20)
  ) {materialFunction=/Game/Weapons/WeaponSkins/Wraps/MaterialFunctions/WorldAlignedTexture}
  NamedRerouteDeclaration(None=Clamp(
    None=MaterialFunctionCall(Float2=@n21.XYZ Texture) {materialFunction=&BreakOutFloat2Components}.R
    Min=ScalarParameter() {parameterName=MinGlow}
   )) {name=PanningNoise, nodeColor=<rgba 0 0.6016 1 1>}
  n22 = NamedRerouteUsage()
  n23 = ScalarParameter() {defaultValue=2, parameterName=RockDetailTexturePow}
  n24 = Subtract(A=Multiply(A=Power(Base=@n22, Exp=@n23) {constExponent=4}) {constB=2})
  n25 = ScalarParameter() {defaultValue=0.5, parameterName=DetailNoiseInfluence}
  NamedRerouteDeclaration(None=Clamp(None=Multiply(A=@n24, B=@n25)) {minDefault=-1}) {name=Noise_UVOffset, nodeColor=<rgba 0.8516 0 1 1>}
  n26 = TextureObjectParameter() {
   group=Detail Texture
   parameterName=Detail D
   sortPriority=1
   texture=/Game/Environments/Apollo/Rocks/Natural_Rocks/Textures/T_RockDetails_02D
  }
  n27 = ScalarParameter() {defaultValue=256, group=Detail Texture, parameterName=DetailTile}
  n28 = Add(A=@n19, B=@n20)
  NamedRerouteDeclaration(None=MaterialFunctionCall(Float3=MaterialFunctionCall(TextureObject=@n26, TextureSize=@n27, WorldPosition=@n28) {materialFunction=&WorldAlignedTexture}.XYZ Texture) {materialFunction=&BreakOutFloat3Components}.R) {name=DetailNoise_R, nodeColor=<rgba 1 0.0516 0 1>}
  n29 = NamedRerouteUsage()
  n30 = QualitySwitch(Default=NamedRerouteUsage())
  n31 = QualitySwitch(Default=Constant() {r=1})
  n32 = Multiply(
   A=NamedRerouteUsage()
   B=StaticSwitchParameter(True=@n30, False=ShadingPathSwitch(Default=@n30, Mobile=@n31)) {parameterName=UseBypassQualitySwitch}
  )
  NamedRerouteDeclaration(None=Multiply(A=Multiply(A=@n29, B=@n32), B=ScalarParameter() {defaultValue=10, parameterName=EmissiveStrength})) {name=Emissive, nodeColor=<rgba 0 1 0.8594 1>}
  NamedRerouteDeclaration(None=@n1.EmissiveColor) {name=Input_EmissiveColor, nodeColor=<rgba 1 0 0.2344 1>}
  NamedRerouteUsage()
  VectorParameter() {parameterName=FortaniumVeinsEmissive, defaultValue=<rgba 1 1 1 1>}
 ]}
