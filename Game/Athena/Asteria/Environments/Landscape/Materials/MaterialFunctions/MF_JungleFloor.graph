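# Landscape material function for the Asteria jungle floor ground layer.
# It feeds jungle floor diffuse, normal and mask textures, scale parameters and a regular-tiling switch into the shared MF_Biome_Generic_v2 function, which has hex-grid tiling inputs.
# It outputs that function's material and height, plus a simplified mobile material keeping only base color, with constant specular, roughness and metallic and the vertex normal replacing the normal map.
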MaterialFunction /Game/Athena/Asteria/Environments/Landscape/Materials/MaterialFunctions/MF_JungleFloor
graph {nodes [
  n1 = MaterialFunctionCall!(
   LowEndTiling=ScalarParameter() {defaultValue=0.15, parameterName=JungleFloorLowEndTiling}
   UseHexGridonLow=StaticBool()
   BaseCoordinates
   MacroScale=ScalarParameter() {defaultValue=1, parameterName=JungleFloor_Scale}
   MinScale
   Diffuse=TextureObject() {texture=/Game/Environments/Asteria/Terrain/Textures/T_Asteria_JungleFloor_D}
   MaxScale
   Normal=TextureObject() {samplerType=SAMPLERTYPE_Normal, texture=/Game/Environments/Asteria/Terrain/Textures/T_Asteria_JungleFloor_N}
   S_Mask=TextureObject() {samplerType=SAMPLERTYPE_Masks, texture=/Game/Environments/Asteria/Terrain/Textures/T_Asteria_JungleFloor_S}
   HexGridSize
   HexGridGradientShaprmness
   MinOffset
   MaxOffset
   MinRotation
   MaxRotation
   UseRegularTiling=StaticBoolParameter() {parameterName=UseRegularTiling}
   MipBiasValue
   UseMipBias
  ) {materialFunction=&/MF_Biome_Generic_v2}
  FunctionOutput(None=@n1.Result) {outputName=Material}
  n2 = GetMaterialAttributes(None=@n1.Result) {attributeGetTypes [#1]}
  n3 = SetMaterialAttributes!(
   MaterialAttributes=@n2.MaterialAttributes
   BaseColor=@n2.MaterialAttributes
   Specular=Constant() {r=0.5}
   Roughness=Constant() {r=0.75}
   Metallic=Constant()
   Normal=VertexNormalWS()
  ) {attributeSetTypes [
    #1
    #2
    #3
    #4
    #5
   ]}
  FunctionOutput(None=@n3) {outputName=MobileMaterial, sortPriority=2}
  FunctionOutput(None=@n1.Height) {outputName=Height, sortPriority=1}
 ]}
