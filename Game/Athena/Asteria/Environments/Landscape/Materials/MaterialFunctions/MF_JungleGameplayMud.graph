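# Landscape material function for jungle gameplay mud with puddles, outputting Material, Height from base-color alpha, and MobileMaterial.
# Macro and detail mud textures use hex-grid randomized landscape sampling, with height driving color variation.
# A puddle mask from height and slope is built for blending toward tinted, optionally swirled puddles with altered specular, roughness and flattened normals.
# The mobile output skips textures, lerping color, specular and roughness by a remapped grayscale input.
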MaterialFunction /Game/Athena/Asteria/Environments/Landscape/Materials/MaterialFunctions/MF_JungleGameplayMud
graph {nodes [
  n1 = NamedRerouteUsage()
  n2 = ScalarParameter() {defaultValue=0.5, group=GamePlay_Mud, parameterName=Mud_SwirlScale}
  n3 = TextureSampleParameter2D(UVs=Divide(A=@n1, B=@n2)) {group=GamePlay_Mud, parameterName=Mud_Swirl_Texture, texture=/Game/Weapons/WeaponSkins/Textures/T_Swirly_Noise_D1}
  n4 = MaterialFunctionCall(
   Input=@n3.R
   Target Low=ScalarParameter() {defaultValue=0.666, group=GamePlay_Mud, parameterName=Mud_SwirlMin}
   Target High=ScalarParameter() {defaultValue=1, group=GamePlay_Mud, parameterName=Mud_SwirlMax}
  ) {materialFunction=&RemapValueRange}
  n5 = NamedRerouteUsage()
  n6 = Multiply(A=@n4.Result, B=@n5)
  n7 = TextureObjectParameter() {
   group=GamePlay_Mud
   parameterName=GameplayMud_S
   samplerType=SAMPLERTYPE_Masks
   texture=/Game/Environments/Asteria/Terrain/Textures/T_Asteria_SmoothMud_S
  }
  n8 = LandscapeLayerCoords()
  n9 = NamedRerouteUsage()
  n10 = MaterialFunctionCall(Float3=MaterialFunctionCall!(
    Texture=@n7
    LandscapeCoordinates=Multiply(A=@n8, B=@n9)
    LowEndTiling=NamedRerouteUsage()
    HexGridGradientSharpness=NamedRerouteUsage()
    HexGridSize=NamedRerouteUsage()
    HexGridUVInput
    MinScale=NamedRerouteUsage()
    MaxScale=NamedRerouteUsage()
    MinOffset=NamedRerouteUsage()
    MaxOffset=NamedRerouteUsage()
    "MaxRotation(Degrees)"=NamedRerouteUsage()
    "MinRotation(Degrees)"=NamedRerouteUsage()
    forceLowQuality=NamedRerouteUsage()
   ) {materialFunction=/Game/Athena/Asteria/Environments/Landscape/Materials/MaterialFunctions/UtilityFunctions/Landscape_TextureSample}.Result) {materialFunction=&BreakOutFloat3Components}
  n11 = ScalarParameter() {defaultValue=0.1, group=GamePlay_Mud, parameterName=PuddleRoughness}
  n12 = Multiply(A=@n3.R, B=ScalarParameter() {defaultValue=0.1, group=GamePlay_Mud, parameterName=PuddleAdditionalSwirlRoughness}) {constB=0.1}
  n13 = TextureObjectParameter() {
   group=GamePlay_Mud
   parameterName=GamePlayMud_Normal
   samplerType=SAMPLERTYPE_Normal
   texture=/Game/Environments/Asteria/Terrain/Textures/T_Asteria_SmoothMud_N
  }
  n14 = LandscapeLayerCoords()
  n15 = NamedRerouteUsage()
  n16 = MaterialFunctionCall(
   BaseNormal=MaterialFunctionCall!(
    Texture=@n13
    UseMipBias=NamedRerouteUsage()
    MipBias=NamedRerouteUsage()
    LowEndTiling=NamedRerouteUsage()
    LandscapeCoordinates=Multiply(A=@n14, B=@n15)
    HexGridSize=NamedRerouteUsage()
    HexGridUVInput
    HexGridGradientSharpness=NamedRerouteUsage()
    MaxScale=NamedRerouteUsage()
    MinScale=NamedRerouteUsage()
    MaxOffset=NamedRerouteUsage()
    MinOffset=NamedRerouteUsage()
    "MaxRotation(Degrees)"=NamedRerouteUsage()
    "MinRotation(Degrees)"=NamedRerouteUsage()
    forceLowQuality=NamedRerouteUsage()
   ) {materialFunction=/Game/Athena/Asteria/Environments/Landscape/Materials/MaterialFunctions/UtilityFunctions/Landscape_NormalTextureSample}.Result
   AdditionalNormal=NamedRerouteUsage()
  ) {materialFunction=&BlendAngleCorrectedNormals}
  n17 = MaterialFunctionCall(Normal=@n16.Result, Flatness=Constant() {r=1}) {materialFunction=&FlattenNormal}
  n18 = Saturate(None=NamedRerouteUsage())
  n19 = ScalarParameter() {defaultValue=1, group=GamePlay_Mud, parameterName=MudNormalStrength}
  n20 = Constant() {r=1}
  n21 = SetMaterialAttributes!(
   MaterialAttributes
   BaseColor=LinearInterpolate(A=NamedRerouteUsage(), B=StaticSwitchParameter(True=@n6, False=@n5) {parameterName=Use_Mud_Swirl}, Alpha=NamedRerouteUsage())
   Specular=LinearInterpolate(A=@n10.R, B=ScalarParameter() {defaultValue=1, group=GamePlay_Mud, parameterName=PuddleSpec}, Alpha=NamedRerouteUsage())
   Roughness=LinearInterpolate(A=@n10.B, B=Add(A=@n11, B=@n12), Alpha=NamedRerouteUsage())
   Normal=Multiply(
    A=LinearInterpolate(A=@n16.Result, B=@n17.Result, Alpha=@n18)
    B=MaterialFunctionCall(X=@n19, Y=@n19, Z=@n20) {materialFunction=&MakeFloat3}.Result
   )
  ) {attributeSetTypes [
    #1
    #2
    #3
    #4
   ]}
  FunctionOutput(None=@n21) {outputName=Material}
  n22 = TextureObjectParameter() {group=GamePlay_Mud, parameterName=GameplayMud_BaseColor, texture=/Game/Environments/Asteria/Terrain/Textures/T_Asteria_SmoothMud_D}
  n23 = Multiply(A=LandscapeLayerCoords(), B=NamedRerouteUsage())
  n24 = MaterialFunctionCall(
   Texture=@n22
   LandscapeCoordinates=@n23
   LowEndTiling=NamedRerouteUsage()
   HexGridGradientSharpness=NamedRerouteUsage()
   HexGridSize=NamedRerouteUsage()
   MinScale=NamedRerouteUsage()
   MaxScale=NamedRerouteUsage()
   MinOffset=NamedRerouteUsage()
   MaxOffset=NamedRerouteUsage()
   "MaxRotation(Degrees)"=NamedRerouteUsage()
   "MinRotation(Degrees)"=NamedRerouteUsage()
   forceLowQuality=NamedRerouteUsage()
  ) {materialFunction=&Landscape_TextureSample}
  n25 = ComponentMask(None=@n24.Result) {a=true}
  FunctionOutput(None=@n25) {outputName=Height, sortPriority=1}
  n26 = FunctionInput(Preview=Constant() {r=0.5}) {bUsePreviewValueAsDefault=true, inputName=FallBackMobileMudGrayscale, inputType=FunctionInput_Scalar}
  n27 = ScalarParameter() {parameterName=GamePlayMudMobileColorMin}
  n28 = ScalarParameter() {defaultValue=1, parameterName=GamePlayMudMobileColorMax}
  n29 = Saturate(None=MaterialFunctionCall(Input=@n26, Input Low=@n27, Input High=@n28) {materialFunction=&RemapValueRange}.Result)
  n30 = SetMaterialAttributes!(
   MaterialAttributes
   BaseColor=LinearInterpolate(
    A=VectorParameter() {group=GamePlay_Mud, parameterName=MobileColorVariationA, defaultValue=<rgba 0.0417 0.0164 0.0074 1>}.RGB
    B=VectorParameter() {group=GamePlay_Mud, parameterName=MobileColorVariationB, defaultValue=<rgba 0.1198 0.0564 0.0301 1>}.RGB
    Alpha=@n29
   )
   Metallic=Constant()
   Specular=LinearInterpolate(
    A=ScalarParameter() {defaultValue=0.5, group=GamePlay_Mud, parameterName=JungleMud_MinSpec}
    B=ScalarParameter() {defaultValue=1, group=GamePlay_Mud, parameterName=JungleMud_MaxSpec}
    Alpha=@n29
   )
   Roughness=LinearInterpolate(
    A=ScalarParameter() {defaultValue=0.1, group=GamePlay_Mud, parameterName=JungleMud_MinRoughness}
    B=ScalarParameter() {defaultValue=0.25, group=GamePlay_Mud, parameterName=JungleMud_MaxRoughness}
    Alpha=@n29
   )
  ) {attributeSetTypes [
    #1
    #5
    #2
    #3
   ]}
  FunctionOutput(None=@n30) {outputName=MobileMaterial, sortPriority=2}
  Length(None=@n3.RGB) {is=class MaterialExpressionLength}
  MaterialFunctionCall(In=ComponentMask(None=Multiply(A=@n26, B=@n27) {constB=2.4}) {r=true}, Contrast=@n28) {materialFunction=&CheapContrast}
  Multiply(A=@n1, B=@n2)
  n31 = Multiply(A=LandscapeLayerCoords(), B=NamedRerouteUsage())
  Multiply(A=@n31)
  NamedRerouteDeclaration(None=ScalarParameter() {defaultValue=1, group=GamePlay_Mud, parameterName=HexGridSize}) {name=HexGridSize, nodeColor=<rgba 1 0.5391 0 1>}
  NamedRerouteDeclaration(None=ScalarParameter() {defaultValue=2, group=GamePlay_Mud, parameterName=HexGridSharpness}) {name=HexGridGradientSharpness, nodeColor=<rgba 0 0.7422 1 1>}
  NamedRerouteDeclaration(None=Constant() {r=1}) {name=MacroScale, nodeColor=<rgba 0 1 0.1328 1>}
  NamedRerouteDeclaration(None=Constant() {r=3}) {name=MipBiasValue, nodeColor=<rgba 0.6641 1 0 1>}
  n32 = TextureObjectParameter() {group=GamePlay_Mud, parameterName=GamePlayMud_DetailHeight, texture=/Game/Environments/Asteria/Terrain/Textures/T_Asteria_MudGround_A_D}
  n33 = NamedRerouteUsage()
  n34 = NamedRerouteUsage()
  n35 = NamedRerouteUsage()
  n36 = NamedRerouteUsage()
  n37 = NamedRerouteUsage()
  n38 = NamedRerouteUsage()
  n39 = NamedRerouteUsage()
  n40 = NamedRerouteUsage()
  n41 = NamedRerouteUsage()
  n42 = NamedRerouteUsage()
  NamedRerouteDeclaration(None=ComponentMask(None=MaterialFunctionCall(
     Texture=@n32
     LandscapeCoordinates=@n31
     LowEndTiling=@n33
     HexGridGradientSharpness=@n34
     HexGridSize=@n35
     MinScale=@n36
     MaxScale=@n37
     MinOffset=@n38
     MaxOffset=@n39
     "MaxRotation(Degrees)"=@n40
     "MinRotation(Degrees)"=@n41
     forceLowQuality=@n42
    ) {materialFunction=&Landscape_TextureSample}.Result) {a=true}) {name=MicroDetailHeight, nodeColor=<rgba 1 0.5156 0 1>}
  NamedRerouteDeclaration(None=@n25) {name=MacroDetailHeight, nodeColor=<rgba 0.4531 1 0 1>}
  NamedRerouteDeclaration(None=ScalarParameter() {defaultValue=2, group=GamePlay_Mud, parameterName=GM_MicroDetail_MinScale}) {name=MicroMinScale, nodeColor=<rgba 1 0.8906 0 1>}
  NamedRerouteDeclaration(None=ScalarParameter() {defaultValue=4, group=GamePlay_Mud, parameterName=GM_MicroDetail_MaxScale}) {name=MicroDetailMaxScale, nodeColor=<rgba 1 0.8906 0 1>}
  n43 = TextureObjectParameter() {
   group=GamePlay_Mud
   parameterName=GamePlayMud_DetailNormal
   samplerType=SAMPLERTYPE_Normal
   texture=/Game/Environments/Asteria/Terrain/Textures/T_Asteria_MudGround_A_N
  }
  n44 = LandscapeLayerCoords()
  n45 = NamedRerouteUsage()
  NamedRerouteDeclaration(None=MaterialFunctionCall(
    Texture=@n43
    UseMipBias=NamedRerouteUsage()
    MipBias=NamedRerouteUsage()
    LowEndTiling=NamedRerouteUsage()
    LandscapeCoordinates=Multiply(A=@n44, B=@n45)
    HexGridSize=NamedRerouteUsage()
    HexGridGradientSharpness=NamedRerouteUsage()
    MaxScale=NamedRerouteUsage()
    MinScale=NamedRerouteUsage()
    MaxOffset=NamedRerouteUsage()
    MinOffset=NamedRerouteUsage()
    "MaxRotation(Degrees)"=NamedRerouteUsage()
    "MinRotation(Degrees)"=NamedRerouteUsage()
    forceLowQuality=NamedRerouteUsage()
   ) {materialFunction=&Landscape_NormalTextureSample}.Result) {name=MicroNormal, nodeColor=<rgba 0 0.8594 1 1>}
  n46 = ScalarParameter() {group=GamePlay_Mud, parameterName=Puddle_Contrast_Low}
  n47 = ScalarParameter() {defaultValue=1, group=GamePlay_Mud, parameterName=Puddle_Contrast_High}
  n48 = SphereMask(
   A=NamedRerouteUsage()
   B=ScalarParameter() {group=GamePlay_Mud, parameterName=Puddle_HeightSample}
   Radius=ScalarParameter() {defaultValue=0.5, group=GamePlay_Mud, parameterName=Puddle_HeightFill}
   Hardness=ScalarParameter() {defaultValue=0.75, group=GamePlay_Mud, parameterName=Puddle_EdgeFalloff}
  )
  n49 = Multiply(A=NamedRerouteUsage(), B=ScalarParameter() {defaultValue=1, group=GamePlay_Mud, parameterName=Puddle_HeightMap_influence})
  n50 = ScalarParameter() {defaultValue=-78.5354, group=GamePlay_Mud, parameterName=PuddleSlopeMask_Low}
  n51 = ScalarParameter() {defaultValue=1.2891, group=GamePlay_Mud, parameterName=PuddleSlopeMask_High}
  n52 = ComponentMask(None=VertexNormalWS()) {b=true}
  n53 = Multiply(A=Saturate(None=Subtract(A=@n48, B=@n49)), B=Saturate(None=LinearInterpolate(A=@n50, B=@n51, Alpha=@n52)))
  NamedRerouteDeclaration(None=Saturate(None=LinearInterpolate(A=@n46, B=@n47, Alpha=@n53))) {name=PuddleMask, nodeColor=<rgba 1 0.8438 0 1>}
  NamedRerouteDeclaration(None=NamedRerouteUsage()) {name=Height, nodeColor=<rgba 1 0 0.1875 1>}
  n54 = ComponentMask(None=@n24.Result) {b=true, g=true, r=true}
  n55 = Desaturation(None=@n54, Fraction=ScalarParameter() {defaultValue=1, parameterName=DesaturationAmt})
  n56 = VectorParameter() {parameterName=MudColor, defaultValue=<rgba 1 1 1 1>}
  n57 = Multiply(
   A=LinearInterpolate(
    A=VectorParameter() {group=GamePlay_Mud, parameterName=BaseColorMult, defaultValue=<rgba 1 1 1 1>}.RGB
    B=VectorParameter() {group=GamePlay_Mud, parameterName=BaseColorMult2, defaultValue=<rgba 1 1 1 1>}.RGB
    Alpha=NamedRerouteUsage()
   )
   B=LinearInterpolate(A=@n54, B=Multiply(A=@n55, B=@n56.RGB), Alpha=ScalarParameter() {parameterName=UseDesaturation})
  )
  n58 = VectorParameter() {group=GamePlay_Mud, parameterName=HeightColorVariation_01, defaultValue=<rgba 0.4287 0.1683 0.0762 1>}
  NamedRerouteDeclaration(None=LinearInterpolate(A=@n57, B=Multiply(A=@n57, B=@n58.RGB), Alpha=NamedRerouteUsage())) {name=BaseColor, nodeColor=<rgba 0.9922 0 1 1>}
  NamedRerouteDeclaration(None=ScalarParameter() {defaultValue=4, group=GamePlay_Mud, parameterName=GM_MinScale}) {name=MinScale, nodeColor=<rgba 1 0.8906 0 1>}
  NamedRerouteDeclaration(None=SphereMask(
    A=NamedRerouteUsage()
    B=ScalarParameter() {defaultValue=1, group=GamePlay_Mud, parameterName=HeightSample}
    Radius=ScalarParameter() {defaultValue=0.25, group=GamePlay_Mud, parameterName=HeightSampleRadius}
    Hardness=ScalarParameter() {defaultValue=0.2, group=GamePlay_Mud, parameterName=HeightSampleHardness}
   )) {name=HeightColorVariation, nodeColor=<rgba 0 1 0.3672 1>}
  NamedRerouteDeclaration(None=@n23) {name=LandscapeUV, nodeColor=<rgba 0.7812 1 0 1>}
  n59 = OneMinus(None=MaterialFunctionCall(Input=NamedRerouteUsage(), Input High=ScalarParameter() {defaultValue=0.5, group=GamePlay_Mud, parameterName=Puddle_HeightFill}) {materialFunction=&RemapValueRange}.Result)
  n60 = ScalarParameter() {defaultValue=2, parameterName=MudPuddleBlendPower}
  NamedRerouteDeclaration(None=LinearInterpolate(
    A=VectorParameter() {group=GamePlay_Mud, parameterName=PuddleColor, defaultValue=<rgba 0.4072 0.1529 0.0595 1>}.RGB
    B=VectorParameter() {group=GamePlay_Mud, parameterName=PuddleColor2, defaultValue=<rgba 0.0283 0.0521 0.0283 1>}.RGB
    Alpha=Power(Base=@n59, Exp=@n60)
   )) {name=PuddleColor, nodeColor=<rgba 1 0 0.6562 1>}
  n61 = MaterialFunctionCall(
   Input=OneMinus(None=NamedRerouteUsage())
   Input High=OneMinus(None=ScalarParameter() {defaultValue=0.5, group=GamePlay_Mud, parameterName=Puddle_HeightFill})
  ) {materialFunction=&RemapValueRange}
  NamedRerouteDeclaration(None=Power(Base=OneMinus(None=@n61.Result), Exp=ScalarParameter() {defaultValue=1, parameterName=MudColorBlendPower})) {name=MudColorLerp, nodeColor=<rgba 0.4531 0 1 1>}
  NamedRerouteDeclaration(None=ScalarParameter() {defaultValue=0.15, parameterName=JungleMud_LowEndTiling}) {name=LowEndTiling, nodeColor=<rgba 0 0.0156 1 1>}
  NamedRerouteDeclaration(None=ScalarParameter() {defaultValue=8, group=GamePlay_Mud, parameterName=GM_MaxScale}) {name=MaxScale, nodeColor=<rgba 1 0.8906 0 1>}
  NamedRerouteDeclaration(None=Constant() {r=-1}) {name=MinOffset, nodeColor=<rgba 0.7812 1 0 1>}
  NamedRerouteDeclaration(None=Constant() {r=1}) {name=MaxOffset, nodeColor=<rgba 0.7812 1 0 1>}
  NamedRerouteDeclaration(None=StaticBool()) {name=ForceLow, nodeColor=<rgba 0 1 0.3203 1>}
  NamedRerouteDeclaration(None=Constant()) {name=MinRotation, nodeColor=<rgba 0.5938 1 0 1>}
  NamedRerouteDeclaration(None=StaticBool()) {name=useMipBias, nodeColor=<rgba 0 1 0.9531 1>}
  NamedRerouteDeclaration(None=Constant()) {name=MaxRotation, nodeColor=<rgba 0.5938 1 0 1>}
  NamedRerouteUsage()
  NamedRerouteUsage()
  OneMinus(None=@n4.Result)
  OneMinus(None=Power(Base=@n29))
  Power(Base=OneMinus(None=NamedRerouteUsage()), Exp=ScalarParameter() {defaultValue=2, parameterName=MudPuddleBlendPower})
  n62 = Round(None=NamedRerouteUsage())
  n63 = ScalarParameter() {group=GamePlay_Mud, parameterName=Offset PuddleByHeight}
  n64 = ComponentMask(None=WorldPosition().XYZ) {g=true, r=true}
  n65 = ScalarParameter() {defaultValue=2000, group=GamePlay_Mud, parameterName=Puddle_Size}
  TextureSample(UVs=Add(A=Multiply(A=@n62, B=@n63), B=Divide(A=@n64, B=@n65))) {
   samplerSource=SSM_Wrap_WorldGroupSettings
   samplerType=SAMPLERTYPE_LinearColor
   texture=/Game/Environments/Asteria/Sets/NeonCity/Textures/WaterPuddles_M
  }
  TextureSample()
 ]}
