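# Material function for Asteria landscape terrain that mutes base color in the ray-traced version of the material, named as a Lumen modification.
# It takes material attributes as input and pulls out BaseColor, then partially desaturates a copy by a fixed fraction.
# A RayTracingQualitySwitch keeps the original color for normal rendering and uses the desaturated one for ray tracing.
# The result is written back into BaseColor, leaving all other attributes unchanged.
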MaterialFunction /Game/Athena/Asteria/Environments/Landscape/Materials/MaterialFunctions/MF_Terrain_LumenModification
graph {nodes [
  n1 = GetMaterialAttributes(None=FunctionInput() {inputType=FunctionInput_MaterialAttributes}) {attributeGetTypes [#1]}
  n2 = Desaturation(None=@n1.BaseColor, Fraction=Constant() {r=0.35})
  FunctionOutput(None=SetMaterialAttributes!(MaterialAttributes=@n1.MaterialAttributes, BaseColor=RayTracingQualitySwitch(Normal=@n1.BaseColor, RayTraced=@n2)) {attributeSetTypes [#1]})
 ]}
